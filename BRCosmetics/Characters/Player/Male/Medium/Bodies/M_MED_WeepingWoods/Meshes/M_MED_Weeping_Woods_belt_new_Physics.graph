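# Physics asset for the Weeping Woods medium male player body's belt mesh, setting up a dynamic right-side belt piece at the pelvis.
# Three capsule bodies form a chain from the pelvis through a near-zero-size master bone to the belt bone.
# The master rotates freely against the pelvis, while the belt joint's tight twist and first swing limits leave it swinging mostly on one axis, like a limited hinge.
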PhysicsAsset /BRCosmetics/Characters/Player/Male/Medium/Bodies/M_MED_WeepingWoods/Meshes/M_MED_Weeping_Woods_belt_new_Physics
bodies {
 dyn_belt_NEW_01_r [{shapeType=Capsule, length=12.9029, radius=4.61, center=<10.8215 -0.0001 0>, rotation=<rot 90 0 -89.9998>}]
 dyn_belt_NEW_01_r_MSTR [{shapeType=Capsule, length=0.01, radius=0.01}]
 pelvis [{shapeType=Capsule, length=11.6988, radius=13.0546}]
}
constraints [
 {
  bone1Name=dyn_belt_NEW_01_r_MSTR
  bone2Name=pelvis
  swing1LimitDegrees=45
  swing1Motion=Free
  swing2LimitDegrees=45
  swing2Motion=Free
  twistLimitDegrees=45
  twistMotion=Free
 }
 {bone1Name=dyn_belt_NEW_01_r, bone2Name=dyn_belt_NEW_01_r_MSTR, swing1LimitDegrees=1, swing2LimitDegrees=25, twistLimitDegrees=1}
]
