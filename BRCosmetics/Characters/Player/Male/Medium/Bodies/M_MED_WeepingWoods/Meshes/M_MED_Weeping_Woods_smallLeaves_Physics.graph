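# Physics asset for the small leaves on the Weeping Woods male medium player body cosmetic, defining bodies for its dynamic leaf bones.
# Five left-side leaves each use an identical box body, constrained to a matching MSTR bone with a negligible-size capsule as anchor, under tight swing and twist limits.
# Six right-side MSTR bones also carry tiny capsules, but have no leaf bodies or constraints here.
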PhysicsAsset /BRCosmetics/Characters/Player/Male/Medium/Bodies/M_MED_WeepingWoods/Meshes/M_MED_Weeping_Woods_smallLeaves_Physics
bodies {
 dyn_leafbody1_01_l [{shapeType=Box, extentX=9.1, extentY=9.1, extentZ=3.1, center=<4.6773 0 -0.0001>}]
 dyn_leafbody1_01_l_MSTR [{shapeType=Capsule, length=0.01, radius=0.01}]
 dyn_leafbody1_01_r_MSTR [{shapeType=Capsule, length=0.01, radius=0.01}]
 dyn_leafbody2_01_l [{shapeType=Box, extentX=9.1, extentY=9.1, extentZ=3.1, center=<4.6773 0 -0.0001>}]
 dyn_leafbody2_01_l_MSTR [{shapeType=Capsule, length=0.01, radius=0.01}]
 dyn_leafbody2_01_r_MSTR [{shapeType=Capsule, length=0.01, radius=0.01}]
 dyn_leafbody3_01_l [{shapeType=Box, extentX=9.1, extentY=9.1, extentZ=3.1, center=<4.6773 0 -0.0001>}]
 dyn_leafbody3_01_l_MSTR [{shapeType=Capsule, length=0.01, radius=0.01}]
 dyn_leafbody3_01_r_MSTR [{shapeType=Capsule, length=0.01, radius=0.01}]
 dyn_leafbody4_01_l [{shapeType=Box, extentX=9.1, extentY=9.1, extentZ=3.1, center=<4.6773 0 -0.0001>}]
 dyn_leafbody4_01_l_MSTR [{shapeType=Capsule, length=0.01, radius=0.01}]
 dyn_leafbody4_01_r_MSTR [{shapeType=Capsule, length=0.01, radius=0.01}]
 dyn_leafbody5_01_l [{shapeType=Box, extentX=9.1, extentY=9.1, extentZ=3.1, center=<4.6773 0 -0.0001>}]
 dyn_leafbody5_01_l_MSTR [{shapeType=Capsule, length=0.01, radius=0.01}]
 dyn_leafbody7_01_r_MSTR [{shapeType=Capsule, length=0.01, radius=0.01}]
 dyn_leafbody8_01_r_MSTR [{shapeType=Capsule, length=0.01, radius=0.01}]
}
constraints [
 {bone1Name=dyn_leafbody1_01_l, bone2Name=dyn_leafbody1_01_l_MSTR, swing1LimitDegrees=5, swing2LimitDegrees=15, twistLimitDegrees=5}
 {bone1Name=dyn_leafbody2_01_l, bone2Name=dyn_leafbody2_01_l_MSTR, swing1LimitDegrees=5, swing2LimitDegrees=15, twistLimitDegrees=5}
 {bone1Name=dyn_leafbody3_01_l, bone2Name=dyn_leafbody3_01_l_MSTR, swing1LimitDegrees=15, swing2LimitDegrees=15, twistLimitDegrees=5}
 {bone1Name=dyn_leafbody4_01_l, bone2Name=dyn_leafbody4_01_l_MSTR, swing1LimitDegrees=5, swing2LimitDegrees=15, twistLimitDegrees=5}
 {bone1Name=dyn_leafbody5_01_l, bone2Name=dyn_leafbody5_01_l_MSTR, swing1LimitDegrees=5, swing2LimitDegrees=15, twistLimitDegrees=5}
]
