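# Anim Blueprint for the Weeping Woods face accessory, driving its leaf and branch bones from the attached parent mesh's copied pose.
# Weighted constraints bind them to face, head and neck bones, then Look At nodes aim leaf, neck and shoulder bones.
# Head swing, twist and crouch-targeting alphas add corrective rotations before sprung AnimDynamics adds leaf sway.
# Separate head-leaf constraint and back-neck Look At chains are left unconnected.
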AnimBlueprint /BRCosmetics/Characters/Player/Male/Medium/Bodies/M_MED_WeepingWoods/Meshes/Parts/M_MED_Weeping_Woods_FaceAcc_AnimBP
target_skeleton=&/M_MED_Weeping_Woods_FaceAcc_Skeleton
graphs [
 AnimGraph {nodes [
   n1 = "Constraint - dyn_head_leaf_028_MSTR (P)"!(ComponentPose, Alpha=1, ConstraintWeights_0=1, ConstraintWeights_1:float=0) -> (Pose) {
    alphaBoolBlend {blendOption=Linear}
    boneToModify {boneName=dyn_head_leaf_028_MSTR}
    constraintSetup [{targetBone {boneName=head}}]
    constraintWeights [1,0]
   }
   n2 = "Constraint - dyn_head_leaf_08_MSTR (PPPPP)"!(
    ComponentPose=@n1
    Alpha=1
    ConstraintWeights_0=0
    ConstraintWeights_1:float=0.0183
    ConstraintWeights_2:float=0.0008
    ConstraintWeights_3:float=0.9806
    ConstraintWeights_4:float=0.0003
   ) -> (Pose) {
    alphaBoolBlend {blendOption=Linear}
    boneToModify {boneName=dyn_head_leaf_08_MSTR}
    constraintSetup [
     {targetBone {boneName=clavicle_r}}
     {targetBone {boneName=neck_02}}
     {targetBone {boneName=neck_01}}
     {targetBone {boneName=head}}
     {targetBone {boneName=spine_05}}
    ]
    constraintWeights [1,1,1,1,1]
   }
   n3 = "Constraint - dyn_head_leaf_029_MSTR (PP)"!(ComponentPose=@n2, Alpha=1, ConstraintWeights_0=0.0286, ConstraintWeights_1:float=0.9714) -> (Pose) {
    alphaBoolBlend {blendOption=Linear}
    boneToModify {boneName=dyn_head_leaf_029_MSTR}
    constraintSetup [{targetBone {boneName=neck_02}}, {targetBone {boneName=head}}]
    constraintWeights [1,1]
   }
   n4 = "Constraint - dyn_head_leaf_017_MSTR (PPPPPPPPPPPPPPPPPPPP)"!(
    ComponentPose=@n3
    Alpha=1
    ConstraintWeights_0=0
    ConstraintWeights_1:float=0.1494
    ConstraintWeights_2:float=0
    ConstraintWeights_3:float=0
    ConstraintWeights_4:float=0
    ConstraintWeights_5:float=0.0023
    ConstraintWeights_6:float=0
    ConstraintWeights_7:float=0.2656
    ConstraintWeights_8:float=0.0649
    ConstraintWeights_9:float=0.025
    ConstraintWeights_10:float=0.003
    ConstraintWeights_11:float=0
    ConstraintWeights_12:float=0.1225
    ConstraintWeights_13:float=0.1619
    ConstraintWeights_14:float=0.0213
    ConstraintWeights_15:float=0.0005
    ConstraintWeights_16:float=0.1077
    ConstraintWeights_17:float=0.0187
    ConstraintWeights_18:float=0
    ConstraintWeights_19:float=0.057
   ) -> (Pose) {
    alphaBoolBlend {blendOption=Linear}
    boneToModify {boneName=dyn_head_leaf_017_MSTR}
    constraintSetup [
     {targetBone {boneName=dyn_neck_bk_01_l}}
     {targetBone {boneName=head}}
     {targetBone {boneName=dyn_neck_fr_02}}
     {targetBone {boneName=dyn_neck_fr_02_l}}
     {targetBone {boneName=dyn_shoulder_01_l}}
     {targetBone {boneName=clavicle_l}}
     {targetBone {boneName=dyn_neck_bk_02_l}}
     {targetBone {boneName=dyn_neck_bk_01_r}}
     {targetBone {boneName=dyn_neck_bk_02_r}}
     {targetBone {boneName=dyn_shoulder_01_r}}
     {targetBone {boneName=spine_03}}
     {targetBone {boneName=spine_04}}
     {targetBone {boneName=spine_05}}
     {targetBone {boneName=neck_02}}
     {targetBone {boneName=clavicle_r}}
     {targetBone {boneName=C_jaw}}
     {targetBone {boneName=neck_01}}
     {targetBone {boneName=dyn_neck_bk_02}}
     {targetBone {boneName=dyn_neck_fr_02_r}}
     {targetBone {boneName=dyn_neck_bk_01}}
    ]
    constraintWeights [1,1,1,1,1,1,1,1,1,1,1,1,1,1,1,1,1,1,1,1]
   }
   n5 = "Constraint - dyn_head_leaf_037_MSTR (PPPPPPPPPPPPPPPPPPPP)"!(
    ComponentPose=@n4
    Alpha=1
    ConstraintWeights_0=0
    ConstraintWeights_1:float=0.363
    ConstraintWeights_2:float=0
    ConstraintWeights_3:float=0
    ConstraintWeights_4:float=0
    ConstraintWeights_5:float=0.0001
    ConstraintWeights_6:float=0
    ConstraintWeights_7:float=0.0709
    ConstraintWeights_8:float=0.0055
    ConstraintWeights_9:float=0.0177
    ConstraintWeights_10:float=0.0004
    ConstraintWeights_11:float=0
    ConstraintWeights_12:float=0.1033
    ConstraintWeights_13:float=0.3021
    ConstraintWeights_14:float=0.0116
    ConstraintWeights_15:float=0.0043
    ConstraintWeights_16:float=0.1178
    ConstraintWeights_17:float=0.0006
    ConstraintWeights_18:float=0
    ConstraintWeights_19:float=0.0026
   ) -> (Pose) {
    alphaBoolBlend {blendOption=Linear}
    boneToModify {boneName=dyn_head_leaf_037_MSTR}
    constraintSetup [
     {targetBone {boneName=dyn_neck_bk_01_l}}
     {targetBone {boneName=head}}
     {targetBone {boneName=dyn_neck_fr_02}}
     {targetBone {boneName=dyn_neck_fr_02_l}}
     {targetBone {boneName=dyn_shoulder_01_l}}
     {targetBone {boneName=clavicle_l}}
     {targetBone {boneName=dyn_neck_bk_02_l}}
     {targetBone {boneName=dyn_neck_bk_01_r}}
     {targetBone {boneName=dyn_neck_bk_02_r}}
     {targetBone {boneName=dyn_shoulder_01_r}}
     {targetBone {boneName=spine_03}}
     {targetBone {boneName=spine_04}}
     {targetBone {boneName=spine_05}}
     {targetBone {boneName=neck_02}}
     {targetBone {boneName=clavicle_r}}
     {targetBone {boneName=C_jaw}}
     {targetBone {boneName=neck_01}}
     {targetBone {boneName=dyn_neck_bk_02}}
     {targetBone {boneName=dyn_neck_fr_02_r}}
     {targetBone {boneName=dyn_neck_bk_01}}
    ]
    constraintWeights [1,1,1,1,1,1,1,1,1,1,1,1,1,1,1,1,1,1,1,1]
   }
   n6 = "Constraint - dyn_head_leaf_016_MSTR (P)"!(ComponentPose=@n5, Alpha=1, ConstraintWeights_0=1, ConstraintWeights_1:float=0) -> (Pose) {
    alphaBoolBlend {blendOption=Linear}
    boneToModify {boneName=dyn_head_leaf_016_MSTR}
    constraintSetup [{targetBone {boneName=head}}]
    constraintWeights [1,0]
   }
   n7 = "Constraint - dyn_head_leaf_032_MSTR (PPPPPPPPPPPPPPPPPPPP)"!(
    ComponentPose=@n6
    Alpha=1
    ConstraintWeights_0=0.0791
    ConstraintWeights_1:float=0.1349
    ConstraintWeights_2:float=0
    ConstraintWeights_3:float=0
    ConstraintWeights_4:float=0
    ConstraintWeights_5:float=0.008
    ConstraintWeights_6:float=0.0255
    ConstraintWeights_7:float=0.0005
    ConstraintWeights_8:float=0.0003
    ConstraintWeights_9:float=0
    ConstraintWeights_10:float=0.0017
    ConstraintWeights_11:float=0
    ConstraintWeights_12:float=0.0737
    ConstraintWeights_13:float=0.1364
    ConstraintWeights_14:float=0.0029
    ConstraintWeights_15:float=0
    ConstraintWeights_16:float=0.088
    ConstraintWeights_17:float=0.1024
    ConstraintWeights_18:float=0
    ConstraintWeights_19:float=0.3467
   ) -> (Pose) {
    alphaBoolBlend {blendOption=Linear}
    boneToModify {boneName=dyn_head_leaf_032_MSTR}
    constraintSetup [
     {targetBone {boneName=dyn_neck_bk_01_l}}
     {targetBone {boneName=head}}
     {targetBone {boneName=dyn_neck_fr_02}}
     {targetBone {boneName=dyn_neck_fr_02_l}}
     {targetBone {boneName=dyn_shoulder_01_l}}
     {targetBone {boneName=clavicle_l}}
     {targetBone {boneName=dyn_neck_bk_02_l}}
     {targetBone {boneName=dyn_neck_bk_01_r}}
     {targetBone {boneName=dyn_neck_bk_02_r}}
     {targetBone {boneName=dyn_shoulder_01_r}}
     {targetBone {boneName=spine_03}}
     {targetBone {boneName=spine_04}}
     {targetBone {boneName=spine_05}}
     {targetBone {boneName=neck_02}}
     {targetBone {boneName=clavicle_r}}
     {targetBone {boneName=C_jaw}}
     {targetBone {boneName=neck_01}}
     {targetBone {boneName=dyn_neck_bk_02}}
     {targetBone {boneName=dyn_neck_fr_02_r}}
     {targetBone {boneName=dyn_neck_bk_01}}
    ]
    constraintWeights [1,1,1,1,1,1,1,1,1,1,1,1,1,1,1,1,1,1,1,1]
   }
   n8 = "Constraint - dyn_head_leaf_02_MSTR (PPPPP)"!(
    ComponentPose=@n7
    Alpha=1
    ConstraintWeights_0=0.3564
    ConstraintWeights_1:float=0.0569
    ConstraintWeights_2:float=0.5571
    ConstraintWeights_3:float=0.029
    ConstraintWeights_4:float=0.0007
   ) -> (Pose) {
    alphaBoolBlend {blendOption=Linear}
    boneToModify {boneName=dyn_head_leaf_02_MSTR}
    constraintSetup [
     {targetBone {boneName=neck_02}}
     {targetBone {boneName=neck_01}}
     {targetBone {boneName=head}}
     {targetBone {boneName=spine_05}}
     {targetBone {boneName=clavicle_l}}
    ]
    constraintWeights [1,1,1,1,1]
   }
   n9 = "Constraint - dyn_head_leaf_026_MSTR (PPPPPPPPPPPPPPPPPPPP)"!(
    ComponentPose=@n8
    Alpha=1
    ConstraintWeights_0=0.0835
    ConstraintWeights_1:float=0.2837
    ConstraintWeights_2:float=0
    ConstraintWeights_3:float=0
    ConstraintWeights_4:float=0.001
    ConstraintWeights_5:float=0.0152
    ConstraintWeights_6:float=0.026
    ConstraintWeights_7:float=0
    ConstraintWeights_8:float=0
    ConstraintWeights_9:float=0
    ConstraintWeights_10:float=0.0004
    ConstraintWeights_11:float=0
    ConstraintWeights_12:float=0.132
    ConstraintWeights_13:float=0.2773
    ConstraintWeights_14:float=0.0002
    ConstraintWeights_15:float=0.0002
    ConstraintWeights_16:float=0.1702
    ConstraintWeights_17:float=0.0038
    ConstraintWeights_18:float=0
    ConstraintWeights_19:float=0.0064
   ) -> (Pose) {
    alphaBoolBlend {blendOption=Linear}
    boneToModify {boneName=dyn_head_leaf_026_MSTR}
    constraintSetup [
     {targetBone {boneName=dyn_neck_bk_01_l}}
     {targetBone {boneName=head}}
     {targetBone {boneName=dyn_neck_fr_02}}
     {targetBone {boneName=dyn_neck_fr_02_l}}
     {targetBone {boneName=dyn_shoulder_01_l}}
     {targetBone {boneName=clavicle_l}}
     {targetBone {boneName=dyn_neck_bk_02_l}}
     {targetBone {boneName=dyn_neck_bk_01_r}}
     {targetBone {boneName=dyn_neck_bk_02_r}}
     {targetBone {boneName=dyn_shoulder_01_r}}
     {targetBone {boneName=spine_03}}
     {targetBone {boneName=spine_04}}
     {targetBone {boneName=spine_05}}
     {targetBone {boneName=neck_02}}
     {targetBone {boneName=clavicle_r}}
     {targetBone {boneName=C_jaw}}
     {targetBone {boneName=neck_01}}
     {targetBone {boneName=dyn_neck_bk_02}}
     {targetBone {boneName=dyn_neck_fr_02_r}}
     {targetBone {boneName=dyn_neck_bk_01}}
    ]
    constraintWeights [1,1,1,1,1,1,1,1,1,1,1,1,1,1,1,1,1,1,1,1]
   }
   n10 = "Constraint - dyn_head_leaf_039_MSTR (P)"!(ComponentPose=@n9, Alpha=1, ConstraintWeights_0=1, ConstraintWeights_1:float=0) -> (Pose) {
    alphaBoolBlend {blendOption=Linear}
    boneToModify {boneName=dyn_head_leaf_039_MSTR}
    constraintSetup [{targetBone {boneName=head}}]
    constraintWeights [1,0]
   }
   n11 = "Constraint - dyn_head_leaf_027_MSTR (P)"!(ComponentPose=@n10, Alpha=1, ConstraintWeights_0=1, ConstraintWeights_1:float=0) -> (Pose) {
    alphaBoolBlend {blendOption=Linear}
    boneToModify {boneName=dyn_head_leaf_027_MSTR}
    constraintSetup [{targetBone {boneName=head}}]
    constraintWeights [1,0]
   }
   n12 = "Constraint - dyn_head_leaf_07_MSTR (P)"!(ComponentPose=@n11, Alpha=1, ConstraintWeights_0=1, ConstraintWeights_1:float=0) -> (Pose) {
    alphaBoolBlend {blendOption=Linear}
    boneToModify {boneName=dyn_head_leaf_07_MSTR}
    constraintSetup [{targetBone {boneName=head}}]
    constraintWeights [1,0]
   }
   "Constraint - dyn_head_leaf_01_MSTR (P)"!(ComponentPose=@n12, Alpha=1, ConstraintWeights_0=1, ConstraintWeights_1:float=0) -> (Pose) {
    alphaBoolBlend {blendOption=Linear}
    boneToModify {boneName=dyn_head_leaf_01_MSTR}
    constraintSetup [{targetBone {boneName=head}}]
    constraintWeights [1,0]
   }
   n13 = Look At - Bone: dyn_neck_bk_01_MSTR(Alpha=1) {
    alphaBoolBlend {blendOption=Linear}
    boneToModify {boneName=dyn_neck_bk_01_MSTR}
    lookAtLocation {x=0}
    lookAtTarget {boneReference {boneName=dyn_neck_bk_aim}}
    lookAt_Axis {axis=<1 0>}
    lookUp_Axis {axis {x=0, z=1}}
   }
   n14 = Look At - Bone: dyn_neck_bk_01_l_MSTR(ComponentPose=@n13, Alpha=1) {
    alphaBoolBlend {blendOption=Linear}
    boneToModify {boneName=dyn_neck_bk_01_l_MSTR}
    lookAtLocation {x=0}
    lookAtTarget {boneReference {boneName=dyn_neck_bk_aim_l}}
    lookAt_Axis {axis=<1 0>}
    lookUp_Axis {axis {x=0, z=1}}
   }
   Look At - Bone: dyn_neck_bk_01_r_MSTR(ComponentPose=@n14, Alpha=1) {
    alphaBoolBlend {blendOption=Linear}
    boneToModify {boneName=dyn_neck_bk_01_r_MSTR}
    lookAtLocation {x=0}
    lookAtTarget {boneReference {boneName=dyn_neck_bk_aim_r}}
    lookAt_Axis {axis=<1 0>}
    lookUp_Axis {axis {x=0, z=1}}
   }
   n15 = "Constraint - dyn_branch_01_l_MSTR (P)"!(
    ComponentPose=Local To Component(LocalPose=Copy Pose From Mesh() {bUseAttachedParent=true})
    Alpha=1
    ConstraintWeights_0=1
    ConstraintWeights_1:float=0
   ) -> (Pose) {
    alphaBoolBlend {blendOption=Linear}
    boneToModify {boneName=dyn_branch_01_l_MSTR}
    constraintSetup [{targetBone {boneName=head}}]
    constraintWeights [1,0]
   }
   n16 = "Constraint - dyn_branch_01_MSTR (PPPP)"!(
    ComponentPose=@n15
    Alpha=1
    ConstraintWeights_0=0.0479
    ConstraintWeights_1:float=0.921
    ConstraintWeights_2:float=0.0276
    ConstraintWeights_3:float=0.0035
   ) -> (Pose) {
    alphaBoolBlend {blendOption=Linear}
    boneToModify {boneName=dyn_branch_01_MSTR}
    constraintSetup [{targetBone {boneName=C_brow_mid}}, {targetBone {boneName=head}}, {targetBone {boneName=L_brow_mid}}, {targetBone {boneName=R_brow_mid}}]
    constraintWeights [1,1,1,1]
   }
   n17 = "Constraint - dyn_branch_01_r_MSTR (PPPPP)"!(
    ComponentPose=@n16
    Alpha=1
    ConstraintWeights_0=0.0013
    ConstraintWeights_1:float=0.0001
    ConstraintWeights_2:float=0.9226
    ConstraintWeights_3:float=0.0527
    ConstraintWeights_4:float=0.0233
   ) -> (Pose) {
    alphaBoolBlend {blendOption=Linear}
    boneToModify {boneName=dyn_branch_01_r_MSTR}
    constraintSetup [
     {targetBone {boneName=R_eye_lid_lower_mid}}
     {targetBone {boneName=C_brow_mid}}
     {targetBone {boneName=head}}
     {targetBone {boneName=R_brow_mid}}
     {targetBone {boneName=R_brow_outer}}
    ]
    constraintWeights [1,1,1,1,1]
   }
   n18 = "Constraint - dyn_headleaf7_01_l_MSTR (PPPP)"!(
    ComponentPose=@n17
    Alpha=1
    ConstraintWeights_0=0.0706
    ConstraintWeights_1:float=0.7084
    ConstraintWeights_2:float=0.1731
    ConstraintWeights_3:float=0.048
   ) -> (Pose) {
    alphaBoolBlend {blendOption=Linear}
    boneToModify {boneName=dyn_headleaf7_01_l_MSTR}
    constraintSetup [{targetBone {boneName=C_jaw}}, {targetBone {boneName=head}}, {targetBone {boneName=L_cheek_inner}}, {targetBone {boneName=L_brow_outer}}]
    constraintWeights [1,1,1,1]
   }
   n19 = "Constraint - dyn_headleaf8_01_l_MSTR (PPPPP)"!(
    ComponentPose=@n18
    Alpha=1
    ConstraintWeights_0=0.0069
    ConstraintWeights_1:float=0.0048
    ConstraintWeights_2:float=0.7238
    ConstraintWeights_3:float=0.2604
    ConstraintWeights_4:float=0.0042
   ) -> (Pose) {
    alphaBoolBlend {blendOption=Linear}
    boneToModify {boneName=dyn_headleaf8_01_l_MSTR}
    constraintSetup [
     {targetBone {boneName=L_eye_lid_lower_mid}}
     {targetBone {boneName=L_cheek_inner}}
     {targetBone {boneName=head}}
     {targetBone {boneName=L_brow_outer}}
     {targetBone {boneName=L_eye_lid_upper_mid}}
    ]
    constraintWeights [1,1,1,1,1]
   }
   n20 = "Constraint - dyn_headleaf9_01_l_MSTR (PPPPP)"!(
    ComponentPose=@n19
    Alpha=1
    ConstraintWeights_0=0.0012
    ConstraintWeights_1:float=0.1837
    ConstraintWeights_2:float=0.389
    ConstraintWeights_3:float=0.4249
    ConstraintWeights_4:float=0.0013
   ) -> (Pose) {
    alphaBoolBlend {blendOption=Linear}
    boneToModify {boneName=dyn_headleaf9_01_l_MSTR}
    constraintSetup [
     {targetBone {boneName=L_eye_lid_lower_mid}}
     {targetBone {boneName=L_brow_mid}}
     {targetBone {boneName=L_brow_outer}}
     {targetBone {boneName=head}}
     {targetBone {boneName=L_eye_lid_upper_mid}}
    ]
    constraintWeights [1,1,1,1,1]
   }
   n21 = "Constraint - dyn_headleaf11_01_l_MSTR (PPPP)"!(
    ComponentPose=@n20
    Alpha=1
    ConstraintWeights_0=0.4269
    ConstraintWeights_1:float=0.5196
    ConstraintWeights_2:float=0.0134
    ConstraintWeights_3:float=0.0401
   ) -> (Pose) {
    alphaBoolBlend {blendOption=Linear}
    boneToModify {boneName=dyn_headleaf11_01_l_MSTR}
    constraintSetup [{targetBone {boneName=C_brow_mid}}, {targetBone {boneName=head}}, {targetBone {boneName=L_brow_mid}}, {targetBone {boneName=R_brow_mid}}]
    constraintWeights [1,1,1,1]
   }
   n22 = "Constraint - dyn_headleaf10_01_l_MSTR (PPPPP)"!(
    ComponentPose=@n21
    Alpha=1
    ConstraintWeights_0=0.1453
    ConstraintWeights_1:float=0.0072
    ConstraintWeights_2:float=0.602
    ConstraintWeights_3:float=0.0026
    ConstraintWeights_4:float=0.243
   ) -> (Pose) {
    alphaBoolBlend {blendOption=Linear}
    boneToModify {boneName=dyn_headleaf10_01_l_MSTR}
    constraintSetup [
     {targetBone {boneName=C_brow_mid}}
     {targetBone {boneName=L_eye_lid_lower_mid}}
     {targetBone {boneName=L_brow_mid}}
     {targetBone {boneName=L_eye_lid_upper_mid}}
     {targetBone {boneName=head}}
    ]
    constraintWeights [1,1,1,1,1]
   }
   n23 = "Constraint - dyn_headleaf6_01_l_MSTR (PP)"!(ComponentPose=@n22, Alpha=1, ConstraintWeights_0=0.6598, ConstraintWeights_1:float=0.3402) -> (Pose) {
    alphaBoolBlend {blendOption=Linear}
    boneToModify {boneName=dyn_headleaf6_01_l_MSTR}
    constraintSetup [{targetBone {boneName=C_jaw}}, {targetBone {boneName=head}}]
    constraintWeights [1,1]
   }
   n24 = "Constraint - dyn_headleaf5_01_l_MSTR (PP)"!(ComponentPose=@n23, Alpha=1, ConstraintWeights_0=0.6598, ConstraintWeights_1:float=0.3402) -> (Pose) {
    alphaBoolBlend {blendOption=Linear}
    boneToModify {boneName=dyn_headleaf5_01_l_MSTR}
    constraintSetup [{targetBone {boneName=C_jaw}}, {targetBone {boneName=head}}]
    constraintWeights [1,1]
   }
   n25 = "Constraint - dyn_headleaf2_01_l_MSTR (PPPPP)"!(
    ComponentPose=@n24
    Alpha=1
    ConstraintWeights_0=0.0391
    ConstraintWeights_1:float=0.8198
    ConstraintWeights_2:float=0.0639
    ConstraintWeights_3:float=0.0482
    ConstraintWeights_4:float=0.029
   ) -> (Pose) {
    alphaBoolBlend {blendOption=Linear}
    boneToModify {boneName=dyn_headleaf2_01_l_MSTR}
    constraintSetup [
     {targetBone {boneName=L_lip_corner}}
     {targetBone {boneName=C_jaw}}
     {targetBone {boneName=head}}
     {targetBone {boneName=L_lip_lower_outer}}
     {targetBone {boneName=C_lip_lower_mid}}
    ]
    constraintWeights [1,1,1,1,1]
   }
   n26 = "Constraint - dyn_headleaf1_01_l_MSTR (PPPPP)"!(
    ComponentPose=@n25
    Alpha=1
    ConstraintWeights_0=0.0039
    ConstraintWeights_1:float=0.9606
    ConstraintWeights_2:float=0.0087
    ConstraintWeights_3:float=0.0109
    ConstraintWeights_4:float=0.0159
   ) -> (Pose) {
    alphaBoolBlend {blendOption=Linear}
    boneToModify {boneName=dyn_headleaf1_01_l_MSTR}
    constraintSetup [
     {targetBone {boneName=R_lip_corner}}
     {targetBone {boneName=C_jaw}}
     {targetBone {boneName=R_lip_lower_outer}}
     {targetBone {boneName=L_lip_lower_outer}}
     {targetBone {boneName=C_lip_lower_mid}}
    ]
    constraintWeights [1,1,1,1,1]
   }
   n27 = "Constraint - dyn_headleaf3_01_l_MSTR (PPPPP)"!(
    ComponentPose=@n26
    Alpha=1
    ConstraintWeights_0=0.8124
    ConstraintWeights_1:float=0.0004
    ConstraintWeights_2:float=0.1702
    ConstraintWeights_3:float=0.0006
    ConstraintWeights_4:float=0.0163
   ) -> (Pose) {
    alphaBoolBlend {blendOption=Linear}
    boneToModify {boneName=dyn_headleaf3_01_l_MSTR}
    constraintSetup [
     {targetBone {boneName=C_jaw}}
     {targetBone {boneName=neck_01}}
     {targetBone {boneName=head}}
     {targetBone {boneName=spine_05}}
     {targetBone {boneName=neck_02}}
    ]
    constraintWeights [1,1,1,1,1]
   }
   n28 = "Constraint - dyn_headleaf4_01_l_MSTR (PPP)"!(ComponentPose=@n27, Alpha=1, ConstraintWeights_0=0.7349, ConstraintWeights_1:float=0.2594, ConstraintWeights_2:float=0.0058) -> (Pose) {
    alphaBoolBlend {blendOption=Linear}
    boneToModify {boneName=dyn_headleaf4_01_l_MSTR}
    constraintSetup [{targetBone {boneName=C_jaw}}, {targetBone {boneName=head}}, {targetBone {boneName=neck_02}}]
    constraintWeights [1,1,1]
   }
   n29 = "Constraint - dyn_headleaf2_01_r_MSTR (PPPPP)"!(
    ComponentPose=@n28
    Alpha=1
    ConstraintWeights_0=0.0627
    ConstraintWeights_1:float=0.7674
    ConstraintWeights_2:float=0.0986
    ConstraintWeights_3:float=0.0525
    ConstraintWeights_4:float=0.0189
   ) -> (Pose) {
    alphaBoolBlend {blendOption=Linear}
    boneToModify {boneName=dyn_headleaf2_01_r_MSTR}
    constraintSetup [
     {targetBone {boneName=R_lip_corner}}
     {targetBone {boneName=C_jaw}}
     {targetBone {boneName=head}}
     {targetBone {boneName=R_lip_lower_outer}}
     {targetBone {boneName=C_lip_lower_mid}}
    ]
    constraintWeights [1,1,1,1,1]
   }
   n30 = "Constraint - dyn_headleaf3_01_r_MSTR (PPP)"!(ComponentPose=@n29, Alpha=1, ConstraintWeights_0=0.7507, ConstraintWeights_1:float=0.2181, ConstraintWeights_2:float=0.0312) -> (Pose) {
    alphaBoolBlend {blendOption=Linear}
    boneToModify {boneName=dyn_headleaf3_01_r_MSTR}
    constraintSetup [{targetBone {boneName=C_jaw}}, {targetBone {boneName=head}}, {targetBone {boneName=neck_02}}]
    constraintWeights [1,1,1]
   }
   n31 = "Constraint - dyn_headleaf4_01_r_MSTR (PPP)"!(ComponentPose=@n30, Alpha=1, ConstraintWeights_0=0.6809, ConstraintWeights_1:float=0.2891, ConstraintWeights_2:float=0.03) -> (Pose) {
    alphaBoolBlend {blendOption=Linear}
    boneToModify {boneName=dyn_headleaf4_01_r_MSTR}
    constraintSetup [{targetBone {boneName=C_jaw}}, {targetBone {boneName=head}}, {targetBone {boneName=neck_02}}]
    constraintWeights [1,1,1]
   }
   n32 = "Constraint - dyn_headleaf5_01_r_MSTR (PP)"!(ComponentPose=@n31, Alpha=1, ConstraintWeights_0=0.3005, ConstraintWeights_1:float=0.6995) -> (Pose) {
    alphaBoolBlend {blendOption=Linear}
    boneToModify {boneName=dyn_headleaf5_01_r_MSTR}
    constraintSetup [{targetBone {boneName=C_jaw}}, {targetBone {boneName=head}}]
    constraintWeights [1,1]
   }
   n33 = "Constraint - dyn_headleaf7_01_r_MSTR (PPPP)"!(
    ComponentPose=@n32
    Alpha=1
    ConstraintWeights_0=0.0257
    ConstraintWeights_1:float=0.8359
    ConstraintWeights_2:float=0.0825
    ConstraintWeights_3:float=0.0559
   ) -> (Pose) {
    alphaBoolBlend {blendOption=Linear}
    boneToModify {boneName=dyn_headleaf7_01_r_MSTR}
    constraintSetup [{targetBone {boneName=C_jaw}}, {targetBone {boneName=head}}, {targetBone {boneName=R_cheek_inner}}, {targetBone {boneName=R_brow_outer}}]
    constraintWeights [1,1,1,1]
   }
   n34 = "Constraint - dyn_headleaf8_01_r_MSTR (PPPPP)"!(
    ComponentPose=@n33
    Alpha=1
    ConstraintWeights_0=0.0069
    ConstraintWeights_1:float=0.7039
    ConstraintWeights_2:float=0.0067
    ConstraintWeights_3:float=0.0038
    ConstraintWeights_4:float=0.2787
   ) -> (Pose) {
    alphaBoolBlend {blendOption=Linear}
    boneToModify {boneName=dyn_headleaf8_01_r_MSTR}
    constraintSetup [
     {targetBone {boneName=R_eye_lid_lower_mid}}
     {targetBone {boneName=head}}
     {targetBone {boneName=R_eye_lid_upper_mid}}
     {targetBone {boneName=R_cheek_inner}}
     {targetBone {boneName=R_brow_outer}}
    ]
    constraintWeights [1,1,1,1,1]
   }
   n35 = "Constraint - dyn_headleaf9_01_r_MSTR (PPP)"!(ComponentPose=@n34, Alpha=1, ConstraintWeights_0=0.835, ConstraintWeights_1:float=0.0407, ConstraintWeights_2:float=0.1243) -> (Pose) {
    alphaBoolBlend {blendOption=Linear}
    boneToModify {boneName=dyn_headleaf9_01_r_MSTR}
    constraintSetup [{targetBone {boneName=head}}, {targetBone {boneName=R_brow_mid}}, {targetBone {boneName=R_brow_outer}}]
    constraintWeights [1,1,1]
   }
   n36 = "Constraint - dyn_headleaf10_01_r_MSTR (PPPPP)"!(
    ComponentPose=@n35
    Alpha=1
    ConstraintWeights_0=0.0126
    ConstraintWeights_1:float=0.0669
    ConstraintWeights_2:float=0.2788
    ConstraintWeights_3:float=0.0017
    ConstraintWeights_4:float=0.6401
   ) -> (Pose) {
    alphaBoolBlend {blendOption=Linear}
    boneToModify {boneName=dyn_headleaf10_01_r_MSTR}
    constraintSetup [
     {targetBone {boneName=R_eye_lid_lower_mid}}
     {targetBone {boneName=C_brow_mid}}
     {targetBone {boneName=head}}
     {targetBone {boneName=R_eye_lid_upper_mid}}
     {targetBone {boneName=R_brow_mid}}
    ]
    constraintWeights [1,1,1,1,1]
   }
   n37 = "Constraint - dyn_neck_fr_01_r_MSTR (P)"(ComponentPose=@n36, Alpha=1, ConstraintWeights_0=1) {
    alphaBoolBlend {blendOption=Linear}
    boneToModify {boneName=dyn_neck_fr_01_r_MSTR}
    constraintSetup [{targetBone {boneName=head}}]
    constraintWeights [1]
   }
   n38 = "Constraint - dyn_neck_fr_01_l_MSTR (P)"(ComponentPose=@n37, Alpha=1, ConstraintWeights_0=1) {
    alphaBoolBlend {blendOption=Linear}
    boneToModify {boneName=dyn_neck_fr_01_l_MSTR}
    constraintSetup [{targetBone {boneName=head}}]
    constraintWeights [1]
   }
   n39 = "Constraint - dyn_neck_fr_01_MSTR (P)"(ComponentPose=@n38, Alpha=1, ConstraintWeights_0=1) {
    alphaBoolBlend {blendOption=Linear}
    boneToModify {boneName=dyn_neck_fr_01_MSTR}
    constraintSetup [{targetBone {boneName=head}}]
    constraintWeights [1]
   }
   n40 = "Constraint - dyn_neck_bk_01_r_MSTR (P)"(ComponentPose=@n39, Alpha=0, ConstraintWeights_0=1) {
    alphaBoolBlend {blendOption=Linear}
    boneToModify {boneName=dyn_neck_bk_01_r_MSTR}
    constraintSetup [{targetBone {boneName=head}}]
    constraintWeights [1]
   }
   n41 = "Constraint - dyn_neck_bk_01_l_MSTR (P)"(ComponentPose=@n40, Alpha=0, ConstraintWeights_0=1) {
    alphaBoolBlend {blendOption=Linear}
    boneToModify {boneName=dyn_neck_bk_01_l_MSTR}
    constraintSetup [{targetBone {boneName=head}}]
    constraintWeights [1]
   }
   n42 = "Constraint - dyn_neck_bk_01_MSTR (P)"(ComponentPose=@n41, Alpha=0, ConstraintWeights_0=1) {
    alphaBoolBlend {blendOption=Linear}
    boneToModify {boneName=dyn_neck_bk_01_MSTR}
    constraintSetup [{targetBone {boneName=head}}]
    constraintWeights [1]
   }
   n43 = "Constraint - dyn_neck_fr_01_r_MSTR (P)"(ComponentPose=@n42, Alpha=1, ConstraintWeights_0=1) {
    alphaBoolBlend {blendOption=Linear}
    boneToModify {boneName=dyn_neck_fr_01_r_MSTR}
    constraintSetup [{targetBone {boneName=neck_02}}]
    constraintWeights [1]
   }
   n44 = "Constraint - dyn_neck_fr_01_MSTR (P)"(ComponentPose=@n43, Alpha=1, ConstraintWeights_0=1) {
    alphaBoolBlend {blendOption=Linear}
    boneToModify {boneName=dyn_neck_fr_01_MSTR}
    constraintSetup [{targetBone {boneName=neck_02}}]
    constraintWeights [1]
   }
   n45 = "Constraint - dyn_neck_fr_01_l_MSTR (P)"(ComponentPose=@n44, Alpha=1, ConstraintWeights_0=1) {
    alphaBoolBlend {blendOption=Linear}
    boneToModify {boneName=dyn_neck_fr_01_l_MSTR}
    constraintSetup [{targetBone {boneName=neck_02}}]
    constraintWeights [1]
   }
   n46 = Look At - Bone: dyn_neck_fr_01_MSTR(ComponentPose=@n45, Alpha=1) {
    alphaBoolBlend {blendOption=Linear}
    boneToModify {boneName=dyn_neck_fr_01_MSTR}
    lookAtLocation {x=0}
    lookAtTarget {boneReference {boneName=dyn_neck_fr_aim}}
    lookAt_Axis {axis=<1 0>}
    lookUp_Axis {axis {x=0, z=1}}
   }
   n47 = Look At - Bone: dyn_neck_fr_01_l_MSTR(ComponentPose=@n46, Alpha=1) {
    alphaBoolBlend {blendOption=Linear}
    boneToModify {boneName=dyn_neck_fr_01_l_MSTR}
    lookAtLocation {x=0}
    lookAtTarget {boneReference {boneName=dyn_neck_fr_aim_l}}
    lookAt_Axis {axis=<1 0>}
    lookUp_Axis {axis {x=0, z=1}}
   }
   n48 = Look At - Bone: dyn_neck_fr_01_r_MSTR(ComponentPose=@n47, Alpha=1) {
    alphaBoolBlend {blendOption=Linear}
    boneToModify {boneName=dyn_neck_fr_01_r_MSTR}
    lookAtLocation {x=0}
    lookAtTarget {boneReference {boneName=dyn_neck_fr_aim_r}}
    lookAt_Axis {axis=<1 0>}
    lookUp_Axis {axis {x=0, z=1}}
   }
   n49 = Look At - Bone: dyn_shoulder_01_l_MSTR(ComponentPose=@n48, Alpha=1) {
    alphaBoolBlend {blendOption=Linear}
    boneToModify {boneName=dyn_shoulder_01_l_MSTR}
    lookAtLocation {x=0}
    lookAtTarget {boneReference {boneName=dyn_shoulder_aim_l}}
    lookAt_Axis {axis=<1 0>}
    lookUp_Axis {axis {x=0, z=1}}
   }
   n50 = Look At - Bone: dyn_shoulder_01_r_MSTR(ComponentPose=@n49, Alpha=1) {
    alphaBoolBlend {blendOption=Linear}
    boneToModify {boneName=dyn_shoulder_01_r_MSTR}
    lookAtLocation {x=0}
    lookAtTarget {boneReference {boneName=dyn_shoulder_aim_r}}
    lookAt_Axis {axis=<-1 0>}
    lookUp_Axis {axis {x=0, z=1}}
   }
   n51 = Get headSwingAngleAlphaFrontBeardLeaves() -> (headSwingAngleAlphaFrontBeardLeaves:double)
   n52 = Look At - Bone: dyn_headleaf1_01_l_MSTR(ComponentPose=@n50, Alpha=@n51) {
    alphaBoolBlend {blendOption=Linear}
    boneToModify {boneName=dyn_headleaf1_01_l_MSTR}
    lookAtLocation {x=0}
    lookAtTarget {boneReference {boneName=dyn_headleaf2_aim}}
    lookAt_Axis {axis=<1 0>}
    lookUp_Axis {axis {x=0, z=1}}
   }
   n53 = Look At - Bone: dyn_headleaf2_01_l_MSTR(ComponentPose=@n52, Alpha=@n51) {
    alphaBoolBlend {blendOption=Linear}
    boneToModify {boneName=dyn_headleaf2_01_l_MSTR}
    lookAtLocation {x=0}
    lookAtTarget {boneReference {boneName=dyn_headleaf1_aim_l}}
    lookAt_Axis {axis=<1 0>}
    lookUp_Axis {axis {x=0, z=1}}
   }
   n54 = Look At - Bone: dyn_headleaf2_01_r_MSTR(ComponentPose=@n53, Alpha=@n51) {
    alphaBoolBlend {blendOption=Linear}
    boneToModify {boneName=dyn_headleaf2_01_r_MSTR}
    lookAtLocation {x=0}
    lookAtTarget {boneReference {boneName=dyn_headleaf1_aim_r}}
    lookAt_Axis {axis=<1 0>}
    lookUp_Axis {axis {x=0, z=1}}
   }
   n55 = Look At - Bone: dyn_headleaf3_01_l_MSTR(ComponentPose=@n54, Alpha=@n51) {
    alphaBoolBlend {blendOption=Linear}
    boneToModify {boneName=dyn_headleaf3_01_l_MSTR}
    lookAtLocation {x=0}
    lookAtTarget {boneReference {boneName=dyn_headleaf3_aim_l}}
    lookAt_Axis {axis=<1 0>}
    lookUp_Axis {axis {x=0, z=1}}
   }
   n56 = Look At - Bone: dyn_headleaf3_01_r_MSTR(ComponentPose=@n55, Alpha=@n51) {
    alphaBoolBlend {blendOption=Linear}
    boneToModify {boneName=dyn_headleaf3_01_r_MSTR}
    lookAtLocation {x=0}
    lookAtTarget {boneReference {boneName=dyn_headleaf3_aim_r}}
    lookAt_Axis {axis=<1 0>}
    lookUp_Axis {axis {x=0, z=1}}
   }
   n57 = Look At - Bone: dyn_headleaf4_01_l_MSTR(ComponentPose=@n56, Alpha=@n51) {
    alphaBoolBlend {blendOption=Linear}
    boneToModify {boneName=dyn_headleaf4_01_l_MSTR}
    lookAtLocation {x=0}
    lookAtTarget {boneReference {boneName=dyn_headleaf4_aim_l}}
    lookAt_Axis {axis=<1 0>}
    lookUp_Axis {axis {x=0, z=1}}
   }
   n58 = "Transform (Modify) Bone - Bone: dyn_headleaf1_01_l_MSTR"(ComponentPose=@n57, Alpha=@n51, Translation="0,0,0", Rotation="35,0,0.000", Scale="1,1,1") {
    rotationMode=BMM_Additive
    rotationSpace=BCS_BoneSpace
    alphaBoolBlend {blendOption=Linear}
    boneToModify {boneName=dyn_headleaf1_01_l_MSTR}
    rotation {pitch=35}
   }
   n59 = "Transform (Modify) Bone - Bone: dyn_headleaf2_01_l_MSTR"(ComponentPose=@n58, Alpha=@n51, Translation="0,0,0", Rotation="35,0,0.000", Scale="1,1,1") {
    rotationMode=BMM_Additive
    rotationSpace=BCS_BoneSpace
    alphaBoolBlend {blendOption=Linear}
    boneToModify {boneName=dyn_headleaf2_01_l_MSTR}
    rotation {pitch=35}
   }
   n60 = "Transform (Modify) Bone - Bone: dyn_headleaf2_01_r_MSTR"(ComponentPose=@n59, Alpha=@n51, Translation="0,0,0", Rotation="35,0,0.000", Scale="1,1,1") {
    rotationMode=BMM_Additive
    rotationSpace=BCS_BoneSpace
    alphaBoolBlend {blendOption=Linear}
    boneToModify {boneName=dyn_headleaf2_01_r_MSTR}
    rotation {pitch=35}
   }
   n61 = "Transform (Modify) Bone - Bone: dyn_headleaf3_01_l_MSTR"(ComponentPose=@n60, Alpha=@n51, Translation="0,0,0", Rotation="35,0,0.000", Scale="1,1,1") {
    rotationMode=BMM_Additive
    rotationSpace=BCS_BoneSpace
    alphaBoolBlend {blendOption=Linear}
    boneToModify {boneName=dyn_headleaf3_01_l_MSTR}
    rotation {pitch=35}
   }
   n62 = "Transform (Modify) Bone - Bone: dyn_headleaf3_01_r_MSTR"(ComponentPose=@n61, Alpha=@n51, Translation="0,0,0", Rotation="35,0,0.000", Scale="1,1,1") {
    rotationMode=BMM_Additive
    rotationSpace=BCS_BoneSpace
    alphaBoolBlend {blendOption=Linear}
    boneToModify {boneName=dyn_headleaf3_01_r_MSTR}
    rotation {pitch=35}
   }
   n63 = "Transform (Modify) Bone - Bone: dyn_headleaf4_01_l_MSTR"(ComponentPose=@n62, Alpha=@n51, Translation="0,0,0", Rotation="35,0,0.000", Scale="1,1,1") {
    rotationMode=BMM_Additive
    rotationSpace=BCS_BoneSpace
    alphaBoolBlend {blendOption=Linear}
    boneToModify {boneName=dyn_headleaf4_01_l_MSTR}
    rotation {pitch=35}
   }
   n64 = "Constraint - dyn_head_leaf_08_MSTR (RR)"!(ComponentPose=@n63, Alpha=0, ConstraintWeights_0=1, ConstraintWeights_1:float=0) -> (Pose) {
    alphaBoolBlend {blendOption=Linear}
    boneToModify {boneName=dyn_head_leaf_08_MSTR}
    constraintSetup [{transformType=Rotation, targetBone {boneName=neck_backhelper_01}}, {transformType=Rotation, targetBone {boneName=head}}]
    constraintWeights [1,1]
   }
   n65 = "Transform (Modify) Bone - Bone: dyn_head_leaf_019_MSTR"(
    ComponentPose=@n64
    Alpha=Get headLeavesAimingAlpha() -> (headLeavesAimingAlpha:double)
    Translation="0,0,0"
    Rotation="0.000,15.000,0.000"
    Scale="1,1,1"
   ) {
    rotationMode=BMM_Additive
    rotationSpace=BCS_BoneSpace
    alphaBoolBlend {blendOption=Linear}
    boneToModify {boneName=dyn_head_leaf_019_MSTR}
    rotation {yaw=15.0003}
   }
   n66 = "Transform (Modify) Bone - Bone: neck_backhelper_01"(
    ComponentPose=@n65
    Alpha=Get headSwingAngleAlpha() -> (headSwingAngleAlpha:double)
    Translation="-6.964,-0.000,-5.255"
    Rotation="45.000,0.000,0.000"
    Scale="1,1,1"
   ) {
    rotationMode=BMM_Additive
    rotationSpace=BCS_BoneSpace
    translationMode=BMM_Additive
    translationSpace=BCS_BoneSpace
    alphaBoolBlend {blendOption=Linear}
    boneToModify {boneName=neck_backhelper_01}
    rotation {pitch=45}
    translation {x=-6.964, z=-5.255}
   }
   n67 = "Constraint - dyn_neck_bk_01_l_MSTR (P)"(ComponentPose=@n66, Alpha=0.5, ConstraintWeights_0=1) {
    alphaBoolBlend {blendOption=Linear}
    boneToModify {boneName=dyn_neck_bk_01_l_MSTR}
    constraintSetup [{targetBone {boneName=neck_backhelper_01}}]
    constraintWeights [1]
   }
   n68 = "Constraint - dyn_neck_bk_01_r_MSTR (P)"(ComponentPose=@n67, Alpha=0.5, ConstraintWeights_0=1) {
    alphaBoolBlend {blendOption=Linear}
    boneToModify {boneName=dyn_neck_bk_01_r_MSTR}
    constraintSetup [{targetBone {boneName=neck_backhelper_01}}]
    constraintWeights [1]
   }
   n69 = "Transform (Modify) Bone - Bone: neck_backhelper_01"(
    ComponentPose=@n68
    Alpha=Get isCrouchingAlphaAndTargeting() -> (isCrouchingAlphaAndTargeting:double)
    Translation="0.000,-0.000,2"
    Rotation="15.000,0.000,0.000"
    Scale="1,1,1"
   ) {
    rotationMode=BMM_Additive
    rotationSpace=BCS_BoneSpace
    scaleMode=BMM_Additive
    scaleSpace=BCS_BoneSpace
    translationMode=BMM_Additive
    translationSpace=BCS_BoneSpace
    alphaBoolBlend {blendOption=Linear}
    boneToModify {boneName=neck_backhelper_01}
    rotation {pitch=15}
    translation {z=2}
   }
   n70 = "Transform (Modify) Bone - Bone: neck_backhelper_01"(ComponentPose=@n69, Alpha=Get headTwistAngle() -> (headTwistAngle:double), Translation="0,0,0", Rotation="0.000,0.000,15", Scale="1,1,1") {
    rotationMode=BMM_Additive
    rotationSpace=BCS_BoneSpace
    scaleMode=BMM_Additive
    scaleSpace=BCS_BoneSpace
    translationMode=BMM_Additive
    translationSpace=BCS_BoneSpace
    alphaBoolBlend {blendOption=Linear}
    boneToModify {boneName=neck_backhelper_01}
    rotation {roll=15}
   }
   n71 = Anim Dynamics - Bone: dyn_headleaf4_01_r(ComponentPose=@n70, Alpha=0.75, ExternalForce="0,0,0") {
    angularDampingOverride=1
    angularSpringConstant=1500
    bAngularSpring=true
    bGravityOverrideInSimSpace=true
    bOverrideAngularDamping=true
    gravityScale=0.35
    simulationSpace=RootRelative
    alphaBoolBlend {blendOption=Linear}
    boundBone {boneName=dyn_headleaf4_01_r}
    physicsBodyDefinitions [{
      sphereCollisionRadius=0
      boundBone {boneName=dyn_headleaf4_01_r}
      boxExtents=<10 20 20>
      constraintSetup {angularLimitsMax=<0 0 5>, angularLimitsMin=<0 -30 -5>, angularTarget=<0 0 0>}
      localJointOffset=<5 0 0>
     }]
   }
   n72 = Anim Dynamics - Bone: dyn_headleaf4_02_r(ComponentPose=@n71, Alpha=0.75, ExternalForce="0,0,0") {
    angularDampingOverride=1
    angularSpringConstant=1500
    bAngularSpring=true
    bGravityOverrideInSimSpace=true
    bOverrideAngularDamping=true
    gravityScale=0.35
    simulationSpace=RootRelative
    alphaBoolBlend {blendOption=Linear}
    boundBone {boneName=dyn_headleaf4_02_r}
    physicsBodyDefinitions [{
      sphereCollisionRadius=0
      boundBone {boneName=dyn_headleaf4_02_r}
      boxExtents=<10 20 20>
      constraintSetup {angularLimitsMax=<0 0 5>, angularLimitsMin=<0 -30 -5>, angularTarget=<0 0 0>}
      localJointOffset=<5 0 0>
     }]
   }
   n73 = Anim Dynamics - Bone: dyn_headleaf3_01_r(ComponentPose=@n72, Alpha=0.75, ExternalForce="0,0,0") {
    angularDampingOverride=1
    angularSpringConstant=1500
    bAngularSpring=true
    bGravityOverrideInSimSpace=true
    bOverrideAngularDamping=true
    gravityScale=0.35
    simulationSpace=RootRelative
    alphaBoolBlend {blendOption=Linear}
    boundBone {boneName=dyn_headleaf3_01_r}
    physicsBodyDefinitions [{
      sphereCollisionRadius=0
      boundBone {boneName=dyn_headleaf3_01_r}
      boxExtents=<10 20 20>
      constraintSetup {angularLimitsMax=<0 0 5>, angularLimitsMin=<0 -30 -5>, angularTarget=<0 0 0>}
      localJointOffset=<5 0 0>
     }]
   }
   n74 = Anim Dynamics - Bone: dyn_headleaf2_01_r(ComponentPose=@n73, Alpha=0.75, ExternalForce="0,0,0") {
    angularDampingOverride=1
    angularSpringConstant=1500
    bAngularSpring=true
    bGravityOverrideInSimSpace=true
    bOverrideAngularDamping=true
    gravityScale=0.35
    simulationSpace=RootRelative
    alphaBoolBlend {blendOption=Linear}
    boundBone {boneName=dyn_headleaf2_01_r}
    physicsBodyDefinitions [{
      sphereCollisionRadius=0
      boundBone {boneName=dyn_headleaf2_01_r}
      boxExtents=<10 20 20>
      constraintSetup {angularLimitsMax=<0 0 5>, angularLimitsMin=<0 -30 -5>, angularTarget=<0 0 0>}
      localJointOffset=<5 0 0>
     }]
   }
   n75 = Anim Dynamics - Bone: dyn_headleaf1_02_r(ComponentPose=@n74, Alpha=0.75, ExternalForce="0,0,0") {
    angularDampingOverride=1
    angularSpringConstant=1500
    bAngularSpring=true
    bGravityOverrideInSimSpace=true
    bOverrideAngularDamping=true
    gravityScale=0.35
    simulationSpace=RootRelative
    alphaBoolBlend {blendOption=Linear}
    boundBone {boneName=dyn_headleaf1_02_r}
    physicsBodyDefinitions [{
      sphereCollisionRadius=0
      boundBone {boneName=dyn_headleaf1_02_r}
      boxExtents=<10 20 20>
      constraintSetup {angularLimitsMax=<0 0 5>, angularLimitsMin=<0 -30 -5>, angularTarget=<0 0 0>}
      localJointOffset=<5 0 0>
     }]
   }
   n76 = Anim Dynamics - Bone: dyn_headleaf1_01_l(ComponentPose=@n75, Alpha=0.75, ExternalForce="0,0,0") {
    angularDampingOverride=1
    angularSpringConstant=1500
    bAngularSpring=true
    bGravityOverrideInSimSpace=true
    bOverrideAngularDamping=true
    gravityScale=0.35
    simulationSpace=RootRelative
    alphaBoolBlend {blendOption=Linear}
    boundBone {boneName=dyn_headleaf1_01_l}
    physicsBodyDefinitions [{
      sphereCollisionRadius=0
      boundBone {boneName=dyn_headleaf1_01_l}
      boxExtents=<10 20 20>
      constraintSetup {angularLimitsMax=<0 0 5>, angularLimitsMin=<0 -30 -5>, angularTarget=<0 0 0>}
      localJointOffset=<5 0 0>
     }]
   }
   n77 = Anim Dynamics - Bone: dyn_headleaf2_02_l(ComponentPose=@n76, Alpha=0.75, ExternalForce="0,0,0") {
    angularDampingOverride=1
    angularSpringConstant=1500
    bAngularSpring=true
    bGravityOverrideInSimSpace=true
    bOverrideAngularDamping=true
    gravityScale=0.35
    simulationSpace=RootRelative
    alphaBoolBlend {blendOption=Linear}
    boundBone {boneName=dyn_headleaf2_02_l}
    physicsBodyDefinitions [{
      sphereCollisionRadius=0
      boundBone {boneName=dyn_headleaf2_02_l}
      boxExtents=<10 20 20>
      constraintSetup {angularLimitsMax=<0 0 5>, angularLimitsMin=<0 -30 -5>, angularTarget=<0 0 0>}
      localJointOffset=<5 0 0>
     }]
   }
   n78 = Anim Dynamics - Bone: dyn_headleaf2_01_l(ComponentPose=@n77, Alpha=0.75, ExternalForce="0,0,0") {
    angularDampingOverride=1
    angularSpringConstant=1500
    bAngularSpring=true
    bGravityOverrideInSimSpace=true
    bOverrideAngularDamping=true
    gravityScale=0.35
    simulationSpace=RootRelative
    alphaBoolBlend {blendOption=Linear}
    boundBone {boneName=dyn_headleaf2_01_l}
    physicsBodyDefinitions [{
      sphereCollisionRadius=0
      boundBone {boneName=dyn_headleaf2_01_l}
      boxExtents=<10 20 20>
      constraintSetup {angularLimitsMax=<0 0 5>, angularLimitsMin=<0 -30 -5>, angularTarget=<0 0 0>}
      localJointOffset=<5 0 0>
     }]
   }
   n79 = Anim Dynamics - Bone: dyn_headleaf1_02_l(ComponentPose=@n78, Alpha=0.75, ExternalForce="0,0,0") {
    angularDampingOverride=1
    angularSpringConstant=1500
    bAngularSpring=true
    bGravityOverrideInSimSpace=true
    bOverrideAngularDamping=true
    gravityScale=0.35
    simulationSpace=RootRelative
    alphaBoolBlend {blendOption=Linear}
    boundBone {boneName=dyn_headleaf1_02_l}
    physicsBodyDefinitions [{
      sphereCollisionRadius=0
      boundBone {boneName=dyn_headleaf1_02_l}
      boxExtents=<10 20 20>
      constraintSetup {angularLimitsMax=<0 0 5>, angularLimitsMin=<0 -30 -5>, angularTarget=<0 0 0>}
      localJointOffset=<5 0 0>
     }]
   }
   n80 = Anim Dynamics - Bone: dyn_headleaf3_01_l(ComponentPose=@n79, Alpha=0.75, ExternalForce="0,0,0") {
    angularDampingOverride=1
    angularSpringConstant=1500
    bAngularSpring=true
    bGravityOverrideInSimSpace=true
    bOverrideAngularDamping=true
    gravityScale=0.35
    simulationSpace=RootRelative
    alphaBoolBlend {blendOption=Linear}
    boundBone {boneName=dyn_headleaf3_01_l}
    physicsBodyDefinitions [{
      sphereCollisionRadius=0
      boundBone {boneName=dyn_headleaf3_01_l}
      boxExtents=<10 20 20>
      constraintSetup {angularLimitsMax=<0 0 5>, angularLimitsMin=<0 -30 -5>, angularTarget=<0 0 0>}
      localJointOffset=<5 0 0>
     }]
   }
   n81 = Anim Dynamics - Bone: dyn_neck_fr_01_l(ComponentPose=@n80, Alpha=0.75, ExternalForce="0,0,0") {
    angularDampingOverride=1
    angularSpringConstant=1500
    bAngularSpring=true
    bGravityOverrideInSimSpace=true
    bOverrideAngularDamping=true
    gravityScale=0.35
    simulationSpace=RootRelative
    alphaBoolBlend {blendOption=Linear}
    boundBone {boneName=dyn_neck_fr_01_l}
    physicsBodyDefinitions [{
      sphereCollisionRadius=0
      boundBone {boneName=dyn_neck_fr_01_l}
      boxExtents=<10 20 20>
      constraintSetup {angularLimitsMax=<0 0 5>, angularLimitsMin=<0 -30 -5>, angularTarget=<0 0 0>}
      localJointOffset=<5 0 0>
     }]
   }
   n82 = Anim Dynamics - Bone: dyn_headleaf4_01_l(ComponentPose=@n81, Alpha=0.5, ExternalForce="0,0,0") {
    angularDampingOverride=1
    angularSpringConstant=1500
    bAngularSpring=true
    bGravityOverrideInSimSpace=true
    bOverrideAngularDamping=true
    gravityScale=0.35
    simulationSpace=RootRelative
    alphaBoolBlend {blendOption=Linear}
    boundBone {boneName=dyn_headleaf4_01_l}
    physicsBodyDefinitions [{
      sphereCollisionRadius=0
      boundBone {boneName=dyn_headleaf4_01_l}
      boxExtents=<10 20 20>
      constraintSetup {angularLimitsMax=<0 0 5>, angularLimitsMin=<0 -30 -5>, angularTarget=<0 0 0>}
      localJointOffset=<5 0 0>
     }]
   }
   n83 = Anim Dynamics - Bone: dyn_headleaf4_02_l(ComponentPose=@n82, Alpha=0.5, ExternalForce="0,0,0") {
    angularDampingOverride=1
    angularSpringConstant=1500
    bAngularSpring=true
    bGravityOverrideInSimSpace=true
    bOverrideAngularDamping=true
    gravityScale=0.35
    simulationSpace=RootRelative
    alphaBoolBlend {blendOption=Linear}
    boundBone {boneName=dyn_headleaf4_02_l}
    physicsBodyDefinitions [{
      sphereCollisionRadius=0
      boundBone {boneName=dyn_headleaf4_02_l}
      boxExtents=<10 20 20>
      constraintSetup {angularLimitsMax=<0 0 5>, angularLimitsMin=<0 -30 -5>, angularTarget=<0 0 0>}
      localJointOffset=<5 0 0>
     }]
   }
   n84 = Anim Dynamics - Bone: dyn_headleaf5_01_l(ComponentPose=@n83, Alpha=0.9, ExternalForce="0,0,0") {
    angularDampingOverride=1
    angularSpringConstant=1500
    bAngularSpring=true
    bGravityOverrideInSimSpace=true
    bOverrideAngularDamping=true
    gravityScale=0.35
    simulationSpace=RootRelative
    alphaBoolBlend {blendOption=Linear}
    boundBone {boneName=dyn_headleaf5_01_l}
    physicsBodyDefinitions [{
      sphereCollisionRadius=0
      boundBone {boneName=dyn_headleaf5_01_l}
      boxExtents=<10 20 20>
      constraintSetup {angularLimitsMax=<0 0 5>, angularLimitsMin=<0 -30 -5>, angularTarget=<0 0 0>}
      localJointOffset=<5 0 0>
     }]
   }
   n85 = Anim Dynamics - Bone: dyn_headleaf6_01_l(ComponentPose=@n84, Alpha=0.9, ExternalForce="0,0,0") {
    angularDampingOverride=1
    angularSpringConstant=1500
    bAngularSpring=true
    bGravityOverrideInSimSpace=true
    bOverrideAngularDamping=true
    gravityScale=0.35
    simulationSpace=RootRelative
    alphaBoolBlend {blendOption=Linear}
    boundBone {boneName=dyn_headleaf6_01_l}
    physicsBodyDefinitions [{
      sphereCollisionRadius=0
      boundBone {boneName=dyn_headleaf6_01_l}
      boxExtents=<10 20 20>
      constraintSetup {angularLimitsMax=<0 0 5>, angularLimitsMin=<0 -30 -5>, angularTarget=<0 0 0>}
      localJointOffset=<5 0 0>
     }]
   }
   n86 = Anim Dynamics - Bone: dyn_headleaf6_02_l(ComponentPose=@n85, Alpha=0.9, ExternalForce="0,0,0") {
    angularDampingOverride=1
    angularSpringConstant=1500
    bAngularSpring=true
    bGravityOverrideInSimSpace=true
    bOverrideAngularDamping=true
    gravityScale=0.35
    simulationSpace=RootRelative
    alphaBoolBlend {blendOption=Linear}
    boundBone {boneName=dyn_headleaf6_02_l}
    physicsBodyDefinitions [{
      sphereCollisionRadius=0
      boundBone {boneName=dyn_headleaf6_02_l}
      boxExtents=<10 20 20>
      constraintSetup {angularLimitsMax=<0 0 5>, angularLimitsMin=<0 -30 -5>, angularTarget=<0 0 0>}
      localJointOffset=<5 0 0>
     }]
   }
   n87 = Anim Dynamics - Bone: dyn_headleaf7_01_l(ComponentPose=@n86, Alpha=0.9, ExternalForce="0,0,0") {
    angularDampingOverride=1
    angularSpringConstant=1500
    bAngularSpring=true
    bGravityOverrideInSimSpace=true
    bOverrideAngularDamping=true
    gravityScale=0.35
    simulationSpace=RootRelative
    alphaBoolBlend {blendOption=Linear}
    boundBone {boneName=dyn_headleaf7_01_l}
    physicsBodyDefinitions [{
      sphereCollisionRadius=0
      boundBone {boneName=dyn_headleaf7_01_l}
      boxExtents=<10 20 20>
      constraintSetup {angularLimitsMax=<0 0 5>, angularLimitsMin=<0 -30 -5>, angularTarget=<0 0 0>}
      localJointOffset=<5 0 0>
     }]
   }
   n88 = Anim Dynamics - Bone: dyn_headleaf7_02_l(ComponentPose=@n87, Alpha=0.9, ExternalForce="0,0,0") {
    angularDampingOverride=1
    angularSpringConstant=1500
    bAngularSpring=true
    bGravityOverrideInSimSpace=true
    bOverrideAngularDamping=true
    gravityScale=0.35
    simulationSpace=RootRelative
    alphaBoolBlend {blendOption=Linear}
    boundBone {boneName=dyn_headleaf7_02_l}
    physicsBodyDefinitions [{
      sphereCollisionRadius=0
      boundBone {boneName=dyn_headleaf7_02_l}
      boxExtents=<10 20 20>
      constraintSetup {angularLimitsMax=<0 0 5>, angularLimitsMin=<0 -30 -5>, angularTarget=<0 0 0>}
      localJointOffset=<5 0 0>
     }]
   }
   n89 = Anim Dynamics - Bone: dyn_headleaf8_01_l(ComponentPose=@n88, Alpha=0.9, ExternalForce="0,0,0") {
    angularDampingOverride=1
    angularSpringConstant=1500
    bAngularSpring=true
    bGravityOverrideInSimSpace=true
    bOverrideAngularDamping=true
    gravityScale=0.35
    simulationSpace=RootRelative
    alphaBoolBlend {blendOption=Linear}
    boundBone {boneName=dyn_headleaf8_01_l}
    physicsBodyDefinitions [{
      sphereCollisionRadius=0
      boundBone {boneName=dyn_headleaf8_01_l}
      boxExtents=<10 20 20>
      constraintSetup {angularLimitsMax=<0 0 5>, angularLimitsMin=<0 -30 -5>, angularTarget=<0 0 0>}
      localJointOffset=<5 0 0>
     }]
   }
   n90 = Anim Dynamics - Bone: dyn_headleaf8_02_l(ComponentPose=@n89, Alpha=0.9, ExternalForce="0,0,0") {
    angularDampingOverride=1
    angularSpringConstant=1500
    bAngularSpring=true
    bGravityOverrideInSimSpace=true
    bOverrideAngularDamping=true
    gravityScale=0.35
    simulationSpace=RootRelative
    alphaBoolBlend {blendOption=Linear}
    boundBone {boneName=dyn_headleaf8_02_l}
    physicsBodyDefinitions [{
      sphereCollisionRadius=0
      boundBone {boneName=dyn_headleaf8_02_l}
      boxExtents=<10 20 20>
      constraintSetup {angularLimitsMax=<0 0 5>, angularLimitsMin=<0 -30 -5>, angularTarget=<0 0 0>}
      localJointOffset=<5 0 0>
     }]
   }
   n91 = Anim Dynamics - Bone: dyn_headleaf9_01_l(ComponentPose=@n90, Alpha=0.9, ExternalForce="0,0,0") {
    angularDampingOverride=1
    angularSpringConstant=1500
    bAngularSpring=true
    bGravityOverrideInSimSpace=true
    bOverrideAngularDamping=true
    gravityScale=0.35
    simulationSpace=RootRelative
    alphaBoolBlend {blendOption=Linear}
    boundBone {boneName=dyn_headleaf9_01_l}
    physicsBodyDefinitions [{
      sphereCollisionRadius=0
      boundBone {boneName=dyn_headleaf9_01_l}
      boxExtents=<10 20 20>
      constraintSetup {angularLimitsMax=<0 0 5>, angularLimitsMin=<0 -30 -5>, angularTarget=<0 0 0>}
      localJointOffset=<5 0 0>
     }]
   }
   n92 = Anim Dynamics - Bone: dyn_headleaf9_02_l(ComponentPose=@n91, Alpha=0.9, ExternalForce="0,0,0") {
    angularDampingOverride=1
    angularSpringConstant=1500
    bAngularSpring=true
    bGravityOverrideInSimSpace=true
    bOverrideAngularDamping=true
    gravityScale=0.35
    simulationSpace=RootRelative
    alphaBoolBlend {blendOption=Linear}
    boundBone {boneName=dyn_headleaf9_02_l}
    physicsBodyDefinitions [{
      sphereCollisionRadius=0
      boundBone {boneName=dyn_headleaf9_02_l}
      boxExtents=<10 20 20>
      constraintSetup {angularLimitsMax=<0 0 5>, angularLimitsMin=<0 -30 -5>, angularTarget=<0 0 0>}
      localJointOffset=<5 0 0>
     }]
   }
   n93 = Anim Dynamics - Bone: dyn_headleaf10_01_l(ComponentPose=@n92, Alpha=0.9, ExternalForce="0,0,0") {
    angularDampingOverride=1
    angularSpringConstant=1500
    bAngularSpring=true
    bGravityOverrideInSimSpace=true
    bOverrideAngularDamping=true
    gravityScale=0.35
    simulationSpace=RootRelative
    alphaBoolBlend {blendOption=Linear}
    boundBone {boneName=dyn_headleaf10_01_l}
    physicsBodyDefinitions [{
      sphereCollisionRadius=0
      boundBone {boneName=dyn_headleaf10_01_l}
      boxExtents=<10 20 20>
      constraintSetup {angularLimitsMax=<0 0 5>, angularLimitsMin=<0 -30 -5>, angularTarget=<0 0 0>}
      localJointOffset=<5 0 0>
     }]
   }
   n94 = Anim Dynamics - Bone: dyn_headleaf10_02_l(ComponentPose=@n93, Alpha=0.9, ExternalForce="0,0,0") {
    angularDampingOverride=1
    angularSpringConstant=1500
    bAngularSpring=true
    bGravityOverrideInSimSpace=true
    bOverrideAngularDamping=true
    gravityScale=0.35
    simulationSpace=RootRelative
    alphaBoolBlend {blendOption=Linear}
    boundBone {boneName=dyn_headleaf10_02_l}
    physicsBodyDefinitions [{
      sphereCollisionRadius=0
      boundBone {boneName=dyn_headleaf10_02_l}
      boxExtents=<10 20 20>
      constraintSetup {angularLimitsMax=<0 0 5>, angularLimitsMin=<0 -30 -5>, angularTarget=<0 0 0>}
      localJointOffset=<5 0 0>
     }]
   }
   n95 = Anim Dynamics - Bone: dyn_headleaf11_01_l(ComponentPose=@n94, Alpha=0.9, ExternalForce="0,0,0") {
    angularDampingOverride=1
    angularSpringConstant=1500
    bAngularSpring=true
    bGravityOverrideInSimSpace=true
    bOverrideAngularDamping=true
    gravityScale=0.35
    simulationSpace=RootRelative
    alphaBoolBlend {blendOption=Linear}
    boundBone {boneName=dyn_headleaf11_01_l}
    physicsBodyDefinitions [{
      sphereCollisionRadius=0
      boundBone {boneName=dyn_headleaf11_01_l}
      boxExtents=<10 20 20>
      constraintSetup {angularLimitsMax=<0 0 5>, angularLimitsMin=<0 -30 -5>, angularTarget=<0 0 0>}
      localJointOffset=<5 0 0>
     }]
   }
   n96 = Anim Dynamics - Bone: dyn_headleaf11_02_l(ComponentPose=@n95, Alpha=0.9, ExternalForce="0,0,0") {
    angularDampingOverride=1
    angularSpringConstant=1500
    bAngularSpring=true
    bGravityOverrideInSimSpace=true
    bOverrideAngularDamping=true
    gravityScale=0.35
    simulationSpace=RootRelative
    alphaBoolBlend {blendOption=Linear}
    boundBone {boneName=dyn_headleaf11_02_l}
    physicsBodyDefinitions [{
      sphereCollisionRadius=0
      boundBone {boneName=dyn_headleaf11_02_l}
      boxExtents=<10 20 20>
      constraintSetup {angularLimitsMax=<0 0 5>, angularLimitsMin=<0 -30 -5>, angularTarget=<0 0 0>}
      localJointOffset=<5 0 0>
     }]
   }
   n97 = Anim Dynamics - Bone: dyn_headleaf10_01_r(ComponentPose=@n96, Alpha=0.9, ExternalForce="0,0,0") {
    angularDampingOverride=1
    angularSpringConstant=1500
    bAngularSpring=true
    bGravityOverrideInSimSpace=true
    bOverrideAngularDamping=true
    gravityScale=0.35
    simulationSpace=RootRelative
    alphaBoolBlend {blendOption=Linear}
    boundBone {boneName=dyn_headleaf10_01_r}
    physicsBodyDefinitions [{
      sphereCollisionRadius=0
      boundBone {boneName=dyn_headleaf10_01_r}
      boxExtents=<10 20 20>
      constraintSetup {angularLimitsMax=<0 0 5>, angularLimitsMin=<0 -30 -5>, angularTarget=<0 0 0>}
      localJointOffset=<5 0 0>
     }]
   }
   n98 = Anim Dynamics - Bone: dyn_headleaf9_01_r(ComponentPose=@n97, Alpha=0.9, ExternalForce="0,0,0") {
    angularDampingOverride=1
    angularSpringConstant=1500
    bAngularSpring=true
    bGravityOverrideInSimSpace=true
    bOverrideAngularDamping=true
    gravityScale=0.35
    simulationSpace=RootRelative
    alphaBoolBlend {blendOption=Linear}
    boundBone {boneName=dyn_headleaf9_01_r}
    physicsBodyDefinitions [{
      sphereCollisionRadius=0
      boundBone {boneName=dyn_headleaf9_01_r}
      boxExtents=<10 20 20>
      constraintSetup {angularLimitsMax=<0 0 5>, angularLimitsMin=<0 -30 -5>, angularTarget=<0 0 0>}
      localJointOffset=<5 0 0>
     }]
   }
   n99 = Anim Dynamics - Bone: dyn_headleaf9_02_r(ComponentPose=@n98, Alpha=0.9, ExternalForce="0,0,0") {
    angularDampingOverride=1
    angularSpringConstant=1500
    bAngularSpring=true
    bGravityOverrideInSimSpace=true
    bOverrideAngularDamping=true
    gravityScale=0.35
    simulationSpace=RootRelative
    alphaBoolBlend {blendOption=Linear}
    boundBone {boneName=dyn_headleaf9_02_r}
    physicsBodyDefinitions [{
      sphereCollisionRadius=0
      boundBone {boneName=dyn_headleaf9_02_r}
      boxExtents=<10 20 20>
      constraintSetup {angularLimitsMax=<0 0 5>, angularLimitsMin=<0 -30 -5>, angularTarget=<0 0 0>}
      localJointOffset=<5 0 0>
     }]
   }
   n100 = Anim Dynamics - Bone: dyn_headleaf8_01_r(ComponentPose=@n99, Alpha=0.9, ExternalForce="0,0,0") {
    angularDampingOverride=1
    angularSpringConstant=1500
    bAngularSpring=true
    bGravityOverrideInSimSpace=true
    bOverrideAngularDamping=true
    gravityScale=0.35
    simulationSpace=RootRelative
    alphaBoolBlend {blendOption=Linear}
    boundBone {boneName=dyn_headleaf8_01_r}
    physicsBodyDefinitions [{
      sphereCollisionRadius=0
      boundBone {boneName=dyn_headleaf8_01_r}
      boxExtents=<10 20 20>
      constraintSetup {angularLimitsMax=<0 0 5>, angularLimitsMin=<0 -30 -5>, angularTarget=<0 0 0>}
      localJointOffset=<5 0 0>
     }]
   }
   n101 = Anim Dynamics - Bone: dyn_headleaf8_02_r(ComponentPose=@n100, Alpha=0.9, ExternalForce="0,0,0") {
    angularDampingOverride=1
    angularSpringConstant=1500
    bAngularSpring=true
    bGravityOverrideInSimSpace=true
    bOverrideAngularDamping=true
    gravityScale=0.35
    simulationSpace=RootRelative
    alphaBoolBlend {blendOption=Linear}
    boundBone {boneName=dyn_headleaf8_02_r}
    physicsBodyDefinitions [{
      sphereCollisionRadius=0
      boundBone {boneName=dyn_headleaf8_02_r}
      boxExtents=<10 20 20>
      constraintSetup {angularLimitsMax=<0 0 5>, angularLimitsMin=<0 -30 -5>, angularTarget=<0 0 0>}
      localJointOffset=<5 0 0>
     }]
   }
   n102 = Anim Dynamics - Bone: dyn_headleaf7_01_r(ComponentPose=@n101, Alpha=0.9, ExternalForce="0,0,0") {
    angularDampingOverride=1
    angularSpringConstant=1500
    bAngularSpring=true
    bGravityOverrideInSimSpace=true
    bOverrideAngularDamping=true
    gravityScale=0.35
    simulationSpace=RootRelative
    alphaBoolBlend {blendOption=Linear}
    boundBone {boneName=dyn_headleaf7_01_r}
    physicsBodyDefinitions [{
      sphereCollisionRadius=0
      boundBone {boneName=dyn_headleaf7_01_r}
      boxExtents=<10 20 20>
      constraintSetup {angularLimitsMax=<0 0 5>, angularLimitsMin=<0 -30 -5>, angularTarget=<0 0 0>}
      localJointOffset=<5 0 0>
     }]
   }
   n103 = Anim Dynamics - Bone: dyn_headleaf5_01_r(ComponentPose=@n102, Alpha=0.9, ExternalForce="0,0,0") {
    angularDampingOverride=1
    angularSpringConstant=1500
    bAngularSpring=true
    bGravityOverrideInSimSpace=true
    bOverrideAngularDamping=true
    gravityScale=0.35
    simulationSpace=RootRelative
    alphaBoolBlend {blendOption=Linear}
    boundBone {boneName=dyn_headleaf5_01_r}
    physicsBodyDefinitions [{
      sphereCollisionRadius=0
      boundBone {boneName=dyn_headleaf5_01_r}
      boxExtents=<10 20 20>
      constraintSetup {angularLimitsMax=<0 0 5>, angularLimitsMin=<0 -30 -5>, angularTarget=<0 0 0>}
      localJointOffset=<5 0 0>
     }]
   }
   n104 = Anim Dynamics - Bone: dyn_neck_bk_01(ComponentPose=@n103, Alpha=0.75, ExternalForce="0,0,0") {
    angularDampingOverride=1
    angularSpringConstant=1500
    bAngularSpring=true
    bGravityOverrideInSimSpace=true
    bOverrideAngularDamping=true
    gravityScale=0.35
    simulationSpace=RootRelative
    alphaBoolBlend {blendOption=Linear}
    boundBone {boneName=dyn_neck_bk_01}
    physicsBodyDefinitions [{
      sphereCollisionRadius=0
      boundBone {boneName=dyn_neck_bk_01}
      boxExtents=<10 20 20>
      constraintSetup {angularLimitsMax=<0 0 2>, angularLimitsMin=<0 -15 -2>, angularTarget=<0 0 0>}
      localJointOffset=<5 0 0>
     }]
   }
   n105 = Anim Dynamics - Bone: dyn_neck_bk_02(ComponentPose=@n104, Alpha=0.75, ExternalForce="0,0,0") {
    angularDampingOverride=1
    angularSpringConstant=1500
    bAngularSpring=true
    bGravityOverrideInSimSpace=true
    bOverrideAngularDamping=true
    gravityScale=0.35
    simulationSpace=RootRelative
    alphaBoolBlend {blendOption=Linear}
    boundBone {boneName=dyn_neck_bk_02}
    physicsBodyDefinitions [{
      sphereCollisionRadius=0
      boundBone {boneName=dyn_neck_bk_02}
      boxExtents=<10 20 20>
      constraintSetup {angularLimitsMax=<0 0 2>, angularLimitsMin=<0 -15 -2>, angularTarget=<0 0 0>}
      localJointOffset=<5 0 0>
     }]
   }
   n106 = Anim Dynamics - Bone: dyn_neck_fr_01_l(ComponentPose=@n105, Alpha=0.75, ExternalForce="0,0,0") {
    angularDampingOverride=1
    angularSpringConstant=1500
    bAngularSpring=true
    bGravityOverrideInSimSpace=true
    bOverrideAngularDamping=true
    gravityScale=0.35
    simulationSpace=RootRelative
    alphaBoolBlend {blendOption=Linear}
    boundBone {boneName=dyn_neck_fr_01_l}
    physicsBodyDefinitions [{
      sphereCollisionRadius=0
      boundBone {boneName=dyn_neck_fr_01_l}
      boxExtents=<10 20 20>
      constraintSetup {angularLimitsMax=<0 0 2>, angularLimitsMin=<0 -15 -2>, angularTarget=<0 0 0>}
      localJointOffset=<5 0 0>
     }]
   }
   n107 = Anim Dynamics - Bone: dyn_neck_fr_02_l(ComponentPose=@n106, Alpha=0.75, ExternalForce="0,0,0") {
    angularDampingOverride=1
    angularSpringConstant=1500
    bAngularSpring=true
    bGravityOverrideInSimSpace=true
    bOverrideAngularDamping=true
    gravityScale=0.35
    simulationSpace=RootRelative
    alphaBoolBlend {blendOption=Linear}
    boundBone {boneName=dyn_neck_fr_02_l}
    physicsBodyDefinitions [{
      sphereCollisionRadius=0
      boundBone {boneName=dyn_neck_fr_02_l}
      boxExtents=<10 20 20>
      constraintSetup {angularLimitsMax=<0 0 2>, angularLimitsMin=<0 -15 -2>, angularTarget=<0 0 0>}
      localJointOffset=<5 0 0>
     }]
   }
   n108 = Anim Dynamics - Bone: dyn_neck_fr_01(ComponentPose=@n107, Alpha=0.6, ExternalForce="0,0,0") {
    angularDampingOverride=1
    angularSpringConstant=1500
    bAngularSpring=true
    bGravityOverrideInSimSpace=true
    bOverrideAngularDamping=true
    gravityScale=0.35
    simulationSpace=RootRelative
    alphaBoolBlend {blendOption=Linear}
    boundBone {boneName=dyn_neck_fr_01}
    physicsBodyDefinitions [{
      sphereCollisionRadius=0
      boundBone {boneName=dyn_neck_fr_01}
      boxExtents=<10 20 20>
      constraintSetup {angularLimitsMax=<0 0 2>, angularLimitsMin=<0 -15 -2>, angularTarget=<0 0 0>}
      localJointOffset=<5 0 0>
     }]
   }
   n109 = Anim Dynamics - Bone: dyn_neck_fr_02(ComponentPose=@n108, Alpha=0.75, ExternalForce="0,0,0") {
    angularDampingOverride=1
    angularSpringConstant=1500
    bAngularSpring=true
    bGravityOverrideInSimSpace=true
    bOverrideAngularDamping=true
    gravityScale=0.35
    simulationSpace=RootRelative
    alphaBoolBlend {blendOption=Linear}
    boundBone {boneName=dyn_neck_fr_02}
    physicsBodyDefinitions [{
      sphereCollisionRadius=0
      boundBone {boneName=dyn_neck_fr_02}
      boxExtents=<10 20 20>
      constraintSetup {angularLimitsMax=<0 0 2>, angularLimitsMin=<0 -15 -2>, angularTarget=<0 0 0>}
      localJointOffset=<5 0 0>
     }]
   }
   n110 = Anim Dynamics - Bone: dyn_neck_fr_01_r(ComponentPose=@n109, Alpha=0.75, ExternalForce="0,0,0") {
    angularDampingOverride=1
    angularSpringConstant=1500
    bAngularSpring=true
    bGravityOverrideInSimSpace=true
    bOverrideAngularDamping=true
    gravityScale=0.35
    simulationSpace=RootRelative
    alphaBoolBlend {blendOption=Linear}
    boundBone {boneName=dyn_neck_fr_01_r}
    physicsBodyDefinitions [{
      sphereCollisionRadius=0
      boundBone {boneName=dyn_neck_fr_01_r}
      boxExtents=<10 20 20>
      constraintSetup {angularLimitsMax=<0 0 2>, angularLimitsMin=<0 -15 -2>, angularTarget=<0 0 0>}
      localJointOffset=<5 0 0>
     }]
   }
   n111 = Anim Dynamics - Bone: dyn_neck_fr_02_r(ComponentPose=@n110, Alpha=0.75, ExternalForce="0,0,0") {
    angularDampingOverride=1
    angularSpringConstant=1500
    bAngularSpring=true
    bGravityOverrideInSimSpace=true
    bOverrideAngularDamping=true
    gravityScale=0.35
    simulationSpace=RootRelative
    alphaBoolBlend {blendOption=Linear}
    boundBone {boneName=dyn_neck_fr_02_r}
    physicsBodyDefinitions [{
      sphereCollisionRadius=0
      boundBone {boneName=dyn_neck_fr_02_r}
      boxExtents=<10 20 20>
      constraintSetup {angularLimitsMax=<0 0 2>, angularLimitsMin=<0 -15 -2>, angularTarget=<0 0 0>}
      localJointOffset=<5 0 0>
     }]
   }
   n112 = Anim Dynamics - Bone: dyn_neck_bk_01_l(ComponentPose=@n111, Alpha=0.75, ExternalForce="0,0,0") {
    angularDampingOverride=1
    angularSpringConstant=1500
    bAngularSpring=true
    bGravityOverrideInSimSpace=true
    bOverrideAngularDamping=true
    gravityScale=0.35
    simulationSpace=RootRelative
    alphaBoolBlend {blendOption=Linear}
    boundBone {boneName=dyn_neck_bk_01_l}
    physicsBodyDefinitions [{
      sphereCollisionRadius=0
      boundBone {boneName=dyn_neck_bk_01_l}
      boxExtents=<10 20 20>
      constraintSetup {angularLimitsMax=<0 0 2>, angularLimitsMin=<0 -15 -2>, angularTarget=<0 0 0>}
      localJointOffset=<5 0 0>
     }]
   }
   n113 = Anim Dynamics - Bone: dyn_neck_bk_02_l(ComponentPose=@n112, Alpha=0.75, ExternalForce="0,0,0") {
    angularDampingOverride=1
    angularSpringConstant=1500
    bAngularSpring=true
    bGravityOverrideInSimSpace=true
    bOverrideAngularDamping=true
    gravityScale=0.35
    simulationSpace=RootRelative
    alphaBoolBlend {blendOption=Linear}
    boundBone {boneName=dyn_neck_bk_02_l}
    physicsBodyDefinitions [{
      sphereCollisionRadius=0
      boundBone {boneName=dyn_neck_bk_02_l}
      boxExtents=<10 20 20>
      constraintSetup {angularLimitsMax=<0 0 2>, angularLimitsMin=<0 -15 -2>, angularTarget=<0 0 0>}
      localJointOffset=<5 0 0>
     }]
   }
   n114 = Anim Dynamics - Bone: dyn_neck_bk_01_r(ComponentPose=@n113, Alpha=0.75, ExternalForce="0,0,0") {
    angularDampingOverride=1
    angularSpringConstant=1500
    bAngularSpring=true
    bGravityOverrideInSimSpace=true
    bOverrideAngularDamping=true
    gravityScale=0.35
    simulationSpace=RootRelative
    alphaBoolBlend {blendOption=Linear}
    boundBone {boneName=dyn_neck_bk_01_r}
    physicsBodyDefinitions [{
      sphereCollisionRadius=0
      boundBone {boneName=dyn_neck_bk_01_r}
      boxExtents=<10 20 20>
      constraintSetup {angularLimitsMax=<0 0 2>, angularLimitsMin=<0 -15 -2>, angularTarget=<0 0 0>}
      localJointOffset=<5 0 0>
     }]
   }
   n115 = Anim Dynamics - Bone: dyn_neck_bk_02_r(ComponentPose=@n114, Alpha=0.75, ExternalForce="0,0,0") {
    angularDampingOverride=1
    angularSpringConstant=1500
    bAngularSpring=true
    bGravityOverrideInSimSpace=true
    bOverrideAngularDamping=true
    gravityScale=0.35
    simulationSpace=RootRelative
    alphaBoolBlend {blendOption=Linear}
    boundBone {boneName=dyn_neck_bk_02_r}
    physicsBodyDefinitions [{
      sphereCollisionRadius=0
      boundBone {boneName=dyn_neck_bk_02_r}
      boxExtents=<10 20 20>
      constraintSetup {angularLimitsMax=<0 0 2>, angularLimitsMin=<0 -15 -2>, angularTarget=<0 0 0>}
      localJointOffset=<5 0 0>
     }]
   }
   n116 = Anim Dynamics - Bone: dyn_branch_01_r(ComponentPose=@n115, Alpha=0.6, ExternalForce="0,0,0") {
    angularDampingOverride=1
    angularSpringConstant=1500
    bAngularSpring=true
    bGravityOverrideInSimSpace=true
    bOverrideAngularDamping=true
    gravityScale=0.35
    simulationSpace=RootRelative
    alphaBoolBlend {blendOption=Linear}
    boundBone {boneName=dyn_branch_01_r}
    physicsBodyDefinitions [{
      sphereCollisionRadius=0
      boundBone {boneName=dyn_branch_01_r}
      boxExtents=<10 20 20>
      constraintSetup {angularLimitsMax=<0 5 5>, angularLimitsMin=<0 -5 -5>, angularTarget=<0 0 0>}
      localJointOffset=<5 0 0>
     }]
   }
   n117 = Anim Dynamics - Bone: dyn_branch_02_r(ComponentPose=@n116, Alpha=0.6, ExternalForce="0,0,0") {
    angularDampingOverride=1
    angularSpringConstant=1500
    bAngularSpring=true
    bGravityOverrideInSimSpace=true
    bOverrideAngularDamping=true
    gravityScale=0.35
    simulationSpace=RootRelative
    alphaBoolBlend {blendOption=Linear}
    boundBone {boneName=dyn_branch_02_r}
    physicsBodyDefinitions [{
      sphereCollisionRadius=0
      boundBone {boneName=dyn_branch_02_r}
      boxExtents=<10 20 20>
      constraintSetup {angularLimitsMax=<0 5 5>, angularLimitsMin=<0 -5 -5>, angularTarget=<0 0 0>}
      localJointOffset=<5 0 0>
     }]
   }
   n118 = Anim Dynamics - Bone: dyn_branch_01(ComponentPose=@n117, Alpha=0.9, ExternalForce="0,0,0") {
    angularDampingOverride=1
    angularSpringConstant=1500
    bAngularSpring=true
    bGravityOverrideInSimSpace=true
    bOverrideAngularDamping=true
    gravityScale=0.35
    simulationSpace=RootRelative
    alphaBoolBlend {blendOption=Linear}
    boundBone {boneName=dyn_branch_01}
    physicsBodyDefinitions [{
      sphereCollisionRadius=0
      boundBone {boneName=dyn_branch_01}
      boxExtents=<10 20 20>
      constraintSetup {angularLimitsMax=<0 5 5>, angularLimitsMin=<0 -5 -5>, angularTarget=<0 0 0>}
      localJointOffset=<5 0 0>
     }]
   }
   n119 = Anim Dynamics - Bone: dyn_branch_02(ComponentPose=@n118, Alpha=0.9, ExternalForce="0,0,0") {
    angularDampingOverride=1
    angularSpringConstant=1500
    bAngularSpring=true
    bGravityOverrideInSimSpace=true
    bOverrideAngularDamping=true
    gravityScale=0.35
    simulationSpace=RootRelative
    alphaBoolBlend {blendOption=Linear}
    boundBone {boneName=dyn_branch_02}
    physicsBodyDefinitions [{
      sphereCollisionRadius=0
      boundBone {boneName=dyn_branch_02}
      boxExtents=<10 20 20>
      constraintSetup {angularLimitsMax=<0 5 5>, angularLimitsMin=<0 -5 -5>, angularTarget=<0 0 0>}
      localJointOffset=<5 0 0>
     }]
   }
   n120 = Anim Dynamics - Bone: dyn_branch_03(ComponentPose=@n119, Alpha=0.9, ExternalForce="0,0,0") {
    angularDampingOverride=1
    angularSpringConstant=1500
    bAngularSpring=true
    bGravityOverrideInSimSpace=true
    bOverrideAngularDamping=true
    gravityScale=0.35
    simulationSpace=RootRelative
    alphaBoolBlend {blendOption=Linear}
    boundBone {boneName=dyn_branch_03}
    physicsBodyDefinitions [{
      sphereCollisionRadius=0
      boundBone {boneName=dyn_branch_03}
      boxExtents=<10 20 20>
      constraintSetup {angularLimitsMax=<0 5 5>, angularLimitsMin=<0 -5 -5>, angularTarget=<0 0 0>}
      localJointOffset=<5 0 0>
     }]
   }
   n121 = Anim Dynamics - Bone: dyn_branch_04(ComponentPose=@n120, Alpha=0.9, ExternalForce="0,0,0") {
    angularDampingOverride=1
    angularSpringConstant=1500
    bAngularSpring=true
    bGravityOverrideInSimSpace=true
    bOverrideAngularDamping=true
    gravityScale=0.35
    simulationSpace=RootRelative
    alphaBoolBlend {blendOption=Linear}
    boundBone {boneName=dyn_branch_04}
    physicsBodyDefinitions [{
      sphereCollisionRadius=0
      boundBone {boneName=dyn_branch_04}
      boxExtents=<10 20 20>
      constraintSetup {angularLimitsMax=<0 5 5>, angularLimitsMin=<0 -5 -5>, angularTarget=<0 0 0>}
      localJointOffset=<5 0 0>
     }]
   }
   n122 = Anim Dynamics - Bone: dyn_branch_05(ComponentPose=@n121, Alpha=0.9, ExternalForce="0,0,0") {
    angularDampingOverride=1
    angularSpringConstant=1500
    bAngularSpring=true
    bGravityOverrideInSimSpace=true
    bOverrideAngularDamping=true
    gravityScale=0.35
    simulationSpace=RootRelative
    alphaBoolBlend {blendOption=Linear}
    boundBone {boneName=dyn_branch_05}
    physicsBodyDefinitions [{
      sphereCollisionRadius=0
      boundBone {boneName=dyn_branch_05}
      boxExtents=<10 20 20>
      constraintSetup {angularLimitsMax=<0 5 5>, angularLimitsMin=<0 -5 -5>, angularTarget=<0 0 0>}
      localJointOffset=<5 0 0>
     }]
   }
   n123 = Anim Dynamics - Bone: dyn_branch_01_l(ComponentPose=@n122, Alpha=0.6, ExternalForce="0,0,0") {
    angularDampingOverride=1
    angularSpringConstant=1500
    bAngularSpring=true
    bGravityOverrideInSimSpace=true
    bOverrideAngularDamping=true
    gravityScale=0.35
    simulationSpace=RootRelative
    alphaBoolBlend {blendOption=Linear}
    boundBone {boneName=dyn_branch_01_l}
    physicsBodyDefinitions [{
      sphereCollisionRadius=0
      boundBone {boneName=dyn_branch_01_l}
      boxExtents=<10 20 20>
      constraintSetup {angularLimitsMax=<0 5 5>, angularLimitsMin=<0 -5 -5>, angularTarget=<0 0 0>}
      localJointOffset=<5 0 0>
     }]
   }
   n124 = Anim Dynamics - Bone: dyn_branch_02_l(ComponentPose=@n123, Alpha=0.6, ExternalForce="0,0,0") {
    angularDampingOverride=1
    angularSpringConstant=1500
    bAngularSpring=true
    bGravityOverrideInSimSpace=true
    bOverrideAngularDamping=true
    gravityScale=0.35
    simulationSpace=RootRelative
    alphaBoolBlend {blendOption=Linear}
    boundBone {boneName=dyn_branch_02_l}
    physicsBodyDefinitions [{
      sphereCollisionRadius=0
      boundBone {boneName=dyn_branch_02_l}
      boxExtents=<10 20 20>
      constraintSetup {angularLimitsMax=<0 5 5>, angularLimitsMin=<0 -5 -5>, angularTarget=<0 0 0>}
      localJointOffset=<5 0 0>
     }]
   }
   n125 = Anim Dynamics - Bone: dyn_branch_03_l(ComponentPose=@n124, Alpha=0.6, ExternalForce="0,0,0") {
    angularDampingOverride=1
    angularSpringConstant=1500
    bAngularSpring=true
    bGravityOverrideInSimSpace=true
    bOverrideAngularDamping=true
    gravityScale=0.35
    simulationSpace=RootRelative
    alphaBoolBlend {blendOption=Linear}
    boundBone {boneName=dyn_branch_03_l}
    physicsBodyDefinitions [{
      sphereCollisionRadius=0
      boundBone {boneName=dyn_branch_03_l}
      boxExtents=<10 20 20>
      constraintSetup {angularLimitsMax=<0 5 5>, angularLimitsMin=<0 -5 -5>, angularTarget=<0 0 0>}
      localJointOffset=<5 0 0>
     }]
   }
   n126 = Anim Dynamics - Bone: dyn_head_leaf_027(ComponentPose=@n125, Alpha=0.5, ExternalForce="0,0,0") {
    angularDampingOverride=1
    angularSpringConstant=1500
    bAngularSpring=true
    bGravityOverrideInSimSpace=true
    bOverrideAngularDamping=true
    gravityScale=0.35
    simulationSpace=RootRelative
    alphaBoolBlend {blendOption=Linear}
    boundBone {boneName=dyn_head_leaf_027}
    componentLinearAccScale=<0.3 0.3 0.3>
    componentLinearVelScale=<0.3 0.3 0.3>
    physicsBodyDefinitions [{
      sphereCollisionRadius=0
      boundBone {boneName=dyn_head_leaf_027}
      boxExtents=<10 20 20>
      constraintSetup {angularLimitsMax=<0 0 3>, angularLimitsMin=<0 -10 -3>, angularTarget=<0 0 0>}
      localJointOffset=<5 0 0>
     }]
   }
   n127 = Anim Dynamics - Bone: dyn_head_leaf_035(ComponentPose=@n126, Alpha=0.5, ExternalForce="0,0,0") {
    angularDampingOverride=1
    angularSpringConstant=1500
    bAngularSpring=true
    bGravityOverrideInSimSpace=true
    bOverrideAngularDamping=true
    gravityScale=0.35
    simulationSpace=RootRelative
    alphaBoolBlend {blendOption=Linear}
    boundBone {boneName=dyn_head_leaf_035}
    componentLinearAccScale=<0.3 0.3 0.3>
    componentLinearVelScale=<0.3 0.3 0.3>
    physicsBodyDefinitions [{
      sphereCollisionRadius=0
      boundBone {boneName=dyn_head_leaf_035}
      boxExtents=<10 20 20>
      constraintSetup {angularLimitsMax=<0 0 3>, angularLimitsMin=<0 -10 -3>, angularTarget=<0 0 0>}
      localJointOffset=<5 0 0>
     }]
   }
   n128 = Anim Dynamics - Bone: dyn_head_leaf_012(ComponentPose=@n127, Alpha=0.5, ExternalForce="0,0,0") {
    angularDampingOverride=1
    angularSpringConstant=1500
    bAngularSpring=true
    bGravityOverrideInSimSpace=true
    bOverrideAngularDamping=true
    gravityScale=0.35
    simulationSpace=RootRelative
    alphaBoolBlend {blendOption=Linear}
    boundBone {boneName=dyn_head_leaf_012}
    componentLinearAccScale=<0.3 0.3 0.3>
    componentLinearVelScale=<0.3 0.3 0.3>
    physicsBodyDefinitions [{
      sphereCollisionRadius=0
      boundBone {boneName=dyn_head_leaf_012}
      boxExtents=<10 20 20>
      constraintSetup {angularLimitsMax=<0 0 3>, angularLimitsMin=<0 -10 -3>, angularTarget=<0 0 0>}
      localJointOffset=<5 0 0>
     }]
   }
   n129 = Anim Dynamics - Bone: dyn_head_leaf_07(ComponentPose=@n128, Alpha=0.5, ExternalForce="0,0,0") {
    angularDampingOverride=1
    angularSpringConstant=1500
    bAngularSpring=true
    bGravityOverrideInSimSpace=true
    bOverrideAngularDamping=true
    gravityScale=0.35
    simulationSpace=RootRelative
    alphaBoolBlend {blendOption=Linear}
    boundBone {boneName=dyn_head_leaf_07}
    componentLinearAccScale=<0.3 0.3 0.3>
    componentLinearVelScale=<0.3 0.3 0.3>
    physicsBodyDefinitions [{
      sphereCollisionRadius=0
      boundBone {boneName=dyn_head_leaf_07}
      boxExtents=<10 20 20>
      constraintSetup {angularLimitsMax=<0 0 3>, angularLimitsMin=<0 -10 -3>, angularTarget=<0 0 0>}
      localJointOffset=<5 0 0>
     }]
   }
   n130 = Anim Dynamics - Bone: dyn_head_leaf_05(ComponentPose=@n129, Alpha=0.5, ExternalForce="0,0,0") {
    angularDampingOverride=1
    angularSpringConstant=1500
    bAngularSpring=true
    bGravityOverrideInSimSpace=true
    bOverrideAngularDamping=true
    gravityScale=0.35
    simulationSpace=RootRelative
    alphaBoolBlend {blendOption=Linear}
    boundBone {boneName=dyn_head_leaf_05}
    componentLinearAccScale=<0.3 0.3 0.3>
    componentLinearVelScale=<0.3 0.3 0.3>
    physicsBodyDefinitions [{
      sphereCollisionRadius=0
      boundBone {boneName=dyn_head_leaf_05}
      boxExtents=<10 20 20>
      constraintSetup {angularLimitsMax=<0 0 3>, angularLimitsMin=<0 -10 -3>, angularTarget=<0 0 0>}
      localJointOffset=<5 0 0>
     }]
   }
   n131 = Anim Dynamics - Bone: dyn_head_leaf_03(ComponentPose=@n130, Alpha=0.5, ExternalForce="0,0,0") {
    angularDampingOverride=1
    angularSpringConstant=1500
    bAngularSpring=true
    bGravityOverrideInSimSpace=true
    bOverrideAngularDamping=true
    gravityScale=0.35
    simulationSpace=RootRelative
    alphaBoolBlend {blendOption=Linear}
    boundBone {boneName=dyn_head_leaf_03}
    componentLinearAccScale=<0.3 0.3 0.3>
    componentLinearVelScale=<0.3 0.3 0.3>
    physicsBodyDefinitions [{
      sphereCollisionRadius=0
      boundBone {boneName=dyn_head_leaf_03}
      boxExtents=<10 20 20>
      constraintSetup {angularLimitsMax=<0 0 3>, angularLimitsMin=<0 -10 -3>, angularTarget=<0 0 0>}
      localJointOffset=<5 0 0>
     }]
   }
   n132 = Anim Dynamics - Bone: dyn_head_leaf_023(ComponentPose=@n131, Alpha=0.5, ExternalForce="0,0,0") {
    angularDampingOverride=1
    angularSpringConstant=1500
    bAngularSpring=true
    bGravityOverrideInSimSpace=true
    bOverrideAngularDamping=true
    gravityScale=0.35
    simulationSpace=RootRelative
    alphaBoolBlend {blendOption=Linear}
    boundBone {boneName=dyn_head_leaf_023}
    componentLinearAccScale=<0.3 0.3 0.3>
    componentLinearVelScale=<0.3 0.3 0.3>
    physicsBodyDefinitions [{
      sphereCollisionRadius=0
      boundBone {boneName=dyn_head_leaf_023}
      boxExtents=<10 20 20>
      constraintSetup {angularLimitsMax=<0 0 3>, angularLimitsMin=<0 -10 -3>, angularTarget=<0 0 0>}
      localJointOffset=<5 0 0>
     }]
   }
   n133 = Anim Dynamics - Bone: dyn_head_leaf_015(ComponentPose=@n132, Alpha=0.5, ExternalForce="0,0,0") {
    angularDampingOverride=1
    angularSpringConstant=1500
    bAngularSpring=true
    bGravityOverrideInSimSpace=true
    bOverrideAngularDamping=true
    gravityScale=0.35
    simulationSpace=RootRelative
    alphaBoolBlend {blendOption=Linear}
    boundBone {boneName=dyn_head_leaf_015}
    componentLinearAccScale=<0.3 0.3 0.3>
    componentLinearVelScale=<0.3 0.3 0.3>
    physicsBodyDefinitions [{
      sphereCollisionRadius=0
      boundBone {boneName=dyn_head_leaf_015}
      boxExtents=<10 20 20>
      constraintSetup {angularLimitsMax=<0 0 3>, angularLimitsMin=<0 -10 -3>, angularTarget=<0 0 0>}
      localJointOffset=<5 0 0>
     }]
   }
   n134 = Anim Dynamics - Bone: dyn_head_leaf_014(ComponentPose=@n133, Alpha=0.5, ExternalForce="0,0,0") {
    angularDampingOverride=1
    angularSpringConstant=1500
    bAngularSpring=true
    bGravityOverrideInSimSpace=true
    bOverrideAngularDamping=true
    gravityScale=0.35
    simulationSpace=RootRelative
    alphaBoolBlend {blendOption=Linear}
    boundBone {boneName=dyn_head_leaf_014}
    componentLinearAccScale=<0.3 0.3 0.3>
    componentLinearVelScale=<0.3 0.3 0.3>
    physicsBodyDefinitions [{
      sphereCollisionRadius=0
      boundBone {boneName=dyn_head_leaf_014}
      boxExtents=<10 20 20>
      constraintSetup {angularLimitsMax=<0 0 3>, angularLimitsMin=<0 -10 -3>, angularTarget=<0 0 0>}
      localJointOffset=<5 0 0>
     }]
   }
   n135 = Anim Dynamics - Bone: dyn_head_leaf_06(ComponentPose=@n134, Alpha=0.5, ExternalForce="0,0,0") {
    angularDampingOverride=1
    angularSpringConstant=1500
    bAngularSpring=true
    bGravityOverrideInSimSpace=true
    bOverrideAngularDamping=true
    gravityScale=0.35
    simulationSpace=RootRelative
    alphaBoolBlend {blendOption=Linear}
    boundBone {boneName=dyn_head_leaf_06}
    componentLinearAccScale=<0.3 0.3 0.3>
    componentLinearVelScale=<0.3 0.3 0.3>
    physicsBodyDefinitions [{
      sphereCollisionRadius=0
      boundBone {boneName=dyn_head_leaf_06}
      boxExtents=<10 20 20>
      constraintSetup {angularLimitsMax=<0 0 3>, angularLimitsMin=<0 -10 -3>, angularTarget=<0 0 0>}
      localJointOffset=<5 0 0>
     }]
   }
   n136 = Anim Dynamics - Bone: dyn_head_leaf_036(ComponentPose=@n135, Alpha=0.5, ExternalForce="0,0,0") {
    angularDampingOverride=1
    angularSpringConstant=1500
    bAngularSpring=true
    bGravityOverrideInSimSpace=true
    bOverrideAngularDamping=true
    gravityScale=0.35
    simulationSpace=RootRelative
    alphaBoolBlend {blendOption=Linear}
    boundBone {boneName=dyn_head_leaf_036}
    componentLinearAccScale=<0.3 0.3 0.3>
    componentLinearVelScale=<0.3 0.3 0.3>
    physicsBodyDefinitions [{
      sphereCollisionRadius=0
      boundBone {boneName=dyn_head_leaf_036}
      boxExtents=<10 20 20>
      constraintSetup {angularLimitsMax=<0 0 3>, angularLimitsMin=<0 -10 -3>, angularTarget=<0 0 0>}
      localJointOffset=<5 0 0>
     }]
   }
   n137 = Anim Dynamics - Bone: dyn_head_leaf_034(ComponentPose=@n136, Alpha=0.5, ExternalForce="0,0,0") {
    angularDampingOverride=1
    angularSpringConstant=1500
    bAngularSpring=true
    bGravityOverrideInSimSpace=true
    bOverrideAngularDamping=true
    gravityScale=0.35
    simulationSpace=RootRelative
    alphaBoolBlend {blendOption=Linear}
    boundBone {boneName=dyn_head_leaf_034}
    componentLinearAccScale=<0.3 0.3 0.3>
    componentLinearVelScale=<0.3 0.3 0.3>
    physicsBodyDefinitions [{
      sphereCollisionRadius=0
      boundBone {boneName=dyn_head_leaf_034}
      boxExtents=<10 20 20>
      constraintSetup {angularLimitsMax=<0 0 3>, angularLimitsMin=<0 -10 -3>, angularTarget=<0 0 0>}
      localJointOffset=<5 0 0>
     }]
   }
   n138 = Anim Dynamics - Bone: dyn_head_leaf_033(ComponentPose=@n137, Alpha=0.5, ExternalForce="0,0,0") {
    angularDampingOverride=1
    angularSpringConstant=1500
    bAngularSpring=true
    bGravityOverrideInSimSpace=true
    bOverrideAngularDamping=true
    gravityScale=0.35
    simulationSpace=RootRelative
    alphaBoolBlend {blendOption=Linear}
    boundBone {boneName=dyn_head_leaf_033}
    componentLinearAccScale=<0.3 0.3 0.3>
    componentLinearVelScale=<0.3 0.3 0.3>
    physicsBodyDefinitions [{
      sphereCollisionRadius=0
      boundBone {boneName=dyn_head_leaf_033}
      boxExtents=<10 20 20>
      constraintSetup {angularLimitsMax=<0 0 3>, angularLimitsMin=<0 -10 -3>, angularTarget=<0 0 0>}
      localJointOffset=<5 0 0>
     }]
   }
   n139 = Anim Dynamics - Bone: dyn_head_leaf_09(ComponentPose=@n138, Alpha=0.5, ExternalForce="0,0,0") {
    angularDampingOverride=1
    angularSpringConstant=1500
    bAngularSpring=true
    bGravityOverrideInSimSpace=true
    bOverrideAngularDamping=true
    gravityScale=0.35
    simulationSpace=RootRelative
    alphaBoolBlend {blendOption=Linear}
    boundBone {boneName=dyn_head_leaf_09}
    componentLinearAccScale=<0.3 0.3 0.3>
    componentLinearVelScale=<0.3 0.3 0.3>
    physicsBodyDefinitions [{
      sphereCollisionRadius=0
      boundBone {boneName=dyn_head_leaf_09}
      boxExtents=<10 20 20>
      constraintSetup {angularLimitsMax=<0 0 3>, angularLimitsMin=<0 -10 -3>, angularTarget=<0 0 0>}
      localJointOffset=<5 0 0>
     }]
   }
   n140 = Anim Dynamics - Bone: dyn_head_leaf_01(ComponentPose=@n139, Alpha=0.5, ExternalForce="0,0,0") {
    angularDampingOverride=1
    angularSpringConstant=1500
    bAngularSpring=true
    bGravityOverrideInSimSpace=true
    bOverrideAngularDamping=true
    gravityScale=0.35
    simulationSpace=RootRelative
    alphaBoolBlend {blendOption=Linear}
    boundBone {boneName=dyn_head_leaf_01}
    componentLinearAccScale=<0.3 0.3 0.3>
    componentLinearVelScale=<0.3 0.3 0.3>
    physicsBodyDefinitions [{
      sphereCollisionRadius=0
      boundBone {boneName=dyn_head_leaf_01}
      boxExtents=<10 20 20>
      constraintSetup {angularLimitsMax=<0 0 3>, angularLimitsMin=<0 -10 -3>, angularTarget=<0 0 0>}
      localJointOffset=<5 0 0>
     }]
   }
   n141 = Anim Dynamics - Bone: dyn_head_leaf_020(ComponentPose=@n140, Alpha=0.5, ExternalForce="0,0,0") {
    angularDampingOverride=1
    angularSpringConstant=1500
    bAngularSpring=true
    bGravityOverrideInSimSpace=true
    bOverrideAngularDamping=true
    gravityScale=0.35
    simulationSpace=RootRelative
    alphaBoolBlend {blendOption=Linear}
    boundBone {boneName=dyn_head_leaf_020}
    componentLinearAccScale=<0.3 0.3 0.3>
    componentLinearVelScale=<0.3 0.3 0.3>
    physicsBodyDefinitions [{
      sphereCollisionRadius=0
      boundBone {boneName=dyn_head_leaf_020}
      boxExtents=<10 20 20>
      constraintSetup {angularLimitsMax=<0 0 3>, angularLimitsMin=<0 -10 -3>, angularTarget=<0 0 0>}
      localJointOffset=<5 0 0>
     }]
   }
   n142 = Anim Dynamics - Bone: dyn_head_leaf_019(ComponentPose=@n141, Alpha=0.5, ExternalForce="0,0,0") {
    angularDampingOverride=1
    angularSpringConstant=1500
    bAngularSpring=true
    bGravityOverrideInSimSpace=true
    bOverrideAngularDamping=true
    gravityScale=0.35
    simulationSpace=RootRelative
    alphaBoolBlend {blendOption=Linear}
    boundBone {boneName=dyn_head_leaf_019}
    componentLinearAccScale=<0.3 0.3 0.3>
    componentLinearVelScale=<0.3 0.3 0.3>
    physicsBodyDefinitions [{
      sphereCollisionRadius=0
      boundBone {boneName=dyn_head_leaf_019}
      boxExtents=<10 20 20>
      constraintSetup {angularLimitsMax=<0 0 3>, angularLimitsMin=<0 -10 -3>, angularTarget=<0 0 0>}
      localJointOffset=<5 0 0>
     }]
   }
   n143 = Anim Dynamics - Bone: dyn_head_leaf_018(ComponentPose=@n142, Alpha=0.5, ExternalForce="0,0,0") {
    angularDampingOverride=1
    angularSpringConstant=1500
    bAngularSpring=true
    bGravityOverrideInSimSpace=true
    bOverrideAngularDamping=true
    gravityScale=0.35
    simulationSpace=RootRelative
    alphaBoolBlend {blendOption=Linear}
    boundBone {boneName=dyn_head_leaf_018}
    componentLinearAccScale=<0.3 0.3 0.3>
    componentLinearVelScale=<0.3 0.3 0.3>
    physicsBodyDefinitions [{
      sphereCollisionRadius=0
      boundBone {boneName=dyn_head_leaf_018}
      boxExtents=<10 20 20>
      constraintSetup {angularLimitsMax=<0 0 3>, angularLimitsMin=<0 -10 -3>, angularTarget=<0 0 0>}
      localJointOffset=<5 0 0>
     }]
   }
   n144 = Anim Dynamics - Bone: dyn_head_leaf_013(ComponentPose=@n143, Alpha=0.5, ExternalForce="0,0,0") {
    angularDampingOverride=1
    angularSpringConstant=1500
    bAngularSpring=true
    bGravityOverrideInSimSpace=true
    bOverrideAngularDamping=true
    gravityScale=0.35
    simulationSpace=RootRelative
    alphaBoolBlend {blendOption=Linear}
    boundBone {boneName=dyn_head_leaf_013}
    componentLinearAccScale=<0.3 0.3 0.3>
    componentLinearVelScale=<0.3 0.3 0.3>
    physicsBodyDefinitions [{
      sphereCollisionRadius=0
      boundBone {boneName=dyn_head_leaf_013}
      boxExtents=<10 20 20>
      constraintSetup {angularLimitsMax=<0 0 3>, angularLimitsMin=<0 -10 -3>, angularTarget=<0 0 0>}
      localJointOffset=<5 0 0>
     }]
   }
   n145 = Anim Dynamics - Bone: dyn_head_leaf_041(ComponentPose=@n144, Alpha=0.5, ExternalForce="0,0,0") {
    angularDampingOverride=1
    angularSpringConstant=1500
    bAngularSpring=true
    bGravityOverrideInSimSpace=true
    bOverrideAngularDamping=true
    gravityScale=0.35
    simulationSpace=RootRelative
    alphaBoolBlend {blendOption=Linear}
    boundBone {boneName=dyn_head_leaf_041}
    componentLinearAccScale=<0.3 0.3 0.3>
    componentLinearVelScale=<0.3 0.3 0.3>
    physicsBodyDefinitions [{
      sphereCollisionRadius=0
      boundBone {boneName=dyn_head_leaf_041}
      boxExtents=<10 20 20>
      constraintSetup {angularLimitsMax=<0 0 3>, angularLimitsMin=<0 -10 -3>, angularTarget=<0 0 0>}
      localJointOffset=<5 0 0>
     }]
   }
   n146 = Anim Dynamics - Bone: dyn_head_leaf_042(ComponentPose=@n145, Alpha=0.5, ExternalForce="0,0,0") {
    angularDampingOverride=1
    angularSpringConstant=1500
    bAngularSpring=true
    bGravityOverrideInSimSpace=true
    bOverrideAngularDamping=true
    gravityScale=0.35
    simulationSpace=RootRelative
    alphaBoolBlend {blendOption=Linear}
    boundBone {boneName=dyn_head_leaf_042}
    componentLinearAccScale=<0.3 0.3 0.3>
    componentLinearVelScale=<0.3 0.3 0.3>
    physicsBodyDefinitions [{
      sphereCollisionRadius=0
      boundBone {boneName=dyn_head_leaf_042}
      boxExtents=<10 20 20>
      constraintSetup {angularLimitsMax=<0 0 3>, angularLimitsMin=<0 -10 -3>, angularTarget=<0 0 0>}
      localJointOffset=<5 0 0>
     }]
   }
   n147 = Anim Dynamics - Bone: dyn_head_leaf_040(ComponentPose=@n146, Alpha=0.5, ExternalForce="0,0,0") {
    angularDampingOverride=1
    angularSpringConstant=1500
    bAngularSpring=true
    bGravityOverrideInSimSpace=true
    bOverrideAngularDamping=true
    gravityScale=0.35
    simulationSpace=RootRelative
    alphaBoolBlend {blendOption=Linear}
    boundBone {boneName=dyn_head_leaf_040}
    componentLinearAccScale=<0.3 0.3 0.3>
    componentLinearVelScale=<0.3 0.3 0.3>
    physicsBodyDefinitions [{
      sphereCollisionRadius=0
      boundBone {boneName=dyn_head_leaf_040}
      boxExtents=<10 20 20>
      constraintSetup {angularLimitsMax=<0 0 3>, angularLimitsMin=<0 -10 -3>, angularTarget=<0 0 0>}
      localJointOffset=<5 0 0>
     }]
   }
   n148 = Anim Dynamics - Bone: dyn_head_leaf_030(ComponentPose=@n147, Alpha=0.5, ExternalForce="0,0,0") {
    angularDampingOverride=1
    angularSpringConstant=1500
    bAngularSpring=true
    bGravityOverrideInSimSpace=true
    bOverrideAngularDamping=true
    gravityScale=0.35
    simulationSpace=RootRelative
    alphaBoolBlend {blendOption=Linear}
    boundBone {boneName=dyn_head_leaf_030}
    componentLinearAccScale=<0.3 0.3 0.3>
    componentLinearVelScale=<0.3 0.3 0.3>
    physicsBodyDefinitions [{
      sphereCollisionRadius=0
      boundBone {boneName=dyn_head_leaf_030}
      boxExtents=<10 20 20>
      constraintSetup {angularLimitsMax=<0 0 3>, angularLimitsMin=<0 -10 -3>, angularTarget=<0 0 0>}
      localJointOffset=<5 0 0>
     }]
   }
   n149 = Anim Dynamics - Bone: dyn_head_leaf_039(ComponentPose=@n148, Alpha=0.5, ExternalForce="0,0,0") {
    angularDampingOverride=1
    angularSpringConstant=1500
    bAngularSpring=true
    bGravityOverrideInSimSpace=true
    bOverrideAngularDamping=true
    gravityScale=0.35
    simulationSpace=RootRelative
    alphaBoolBlend {blendOption=Linear}
    boundBone {boneName=dyn_head_leaf_039}
    componentLinearAccScale=<0.3 0.3 0.3>
    componentLinearVelScale=<0.3 0.3 0.3>
    physicsBodyDefinitions [{
      sphereCollisionRadius=0
      boundBone {boneName=dyn_head_leaf_039}
      boxExtents=<10 20 20>
      constraintSetup {angularLimitsMax=<0 0 3>, angularLimitsMin=<0 -10 -3>, angularTarget=<0 0 0>}
      localJointOffset=<5 0 0>
     }]
   }
   n150 = Anim Dynamics - Bone: dyn_head_leaf_08(ComponentPose=@n149, Alpha=0.5, ExternalForce="0,0,0") {
    angularDampingOverride=1
    angularSpringConstant=1500
    bAngularSpring=true
    bGravityOverrideInSimSpace=true
    bOverrideAngularDamping=true
    gravityScale=0.35
    simulationSpace=RootRelative
    alphaBoolBlend {blendOption=Linear}
    boundBone {boneName=dyn_head_leaf_08}
    componentLinearAccScale=<0.3 0.3 0.3>
    componentLinearVelScale=<0.3 0.3 0.3>
    physicsBodyDefinitions [{
      sphereCollisionRadius=0
      boundBone {boneName=dyn_head_leaf_08}
      boxExtents=<10 20 20>
      constraintSetup {angularLimitsMax=<0 0 3>, angularLimitsMin=<0 -10 -3>, angularTarget=<0 0 0>}
      localJointOffset=<5 0 0>
     }]
   }
   n151 = Anim Dynamics - Bone: dyn_head_leaf_016(ComponentPose=@n150, Alpha=0.5, ExternalForce="0,0,0") {
    angularDampingOverride=1
    angularSpringConstant=1500
    bAngularSpring=true
    bGravityOverrideInSimSpace=true
    bOverrideAngularDamping=true
    gravityScale=0.35
    simulationSpace=RootRelative
    alphaBoolBlend {blendOption=Linear}
    boundBone {boneName=dyn_head_leaf_016}
    componentLinearAccScale=<0.3 0.3 0.3>
    componentLinearVelScale=<0.3 0.3 0.3>
    physicsBodyDefinitions [{
      sphereCollisionRadius=0
      boundBone {boneName=dyn_head_leaf_016}
      boxExtents=<10 20 20>
      constraintSetup {angularLimitsMax=<0 0 3>, angularLimitsMin=<0 -10 -3>, angularTarget=<0 0 0>}
      localJointOffset=<5 0 0>
     }]
   }
   n152 = Anim Dynamics - Bone: dyn_head_leaf_028(ComponentPose=@n151, Alpha=0.5, ExternalForce="0,0,0") {
    angularDampingOverride=1
    angularSpringConstant=1500
    bAngularSpring=true
    bGravityOverrideInSimSpace=true
    bOverrideAngularDamping=true
    gravityScale=0.35
    simulationSpace=RootRelative
    alphaBoolBlend {blendOption=Linear}
    boundBone {boneName=dyn_head_leaf_028}
    componentLinearAccScale=<0.3 0.3 0.3>
    componentLinearVelScale=<0.3 0.3 0.3>
    physicsBodyDefinitions [{
      sphereCollisionRadius=0
      boundBone {boneName=dyn_head_leaf_028}
      boxExtents=<10 20 20>
      constraintSetup {angularLimitsMax=<0 0 3>, angularLimitsMin=<0 -10 -3>, angularTarget=<0 0 0>}
      localJointOffset=<5 0 0>
     }]
   }
   n153 = Get headLeavesAimingAlpha()
   n154 = "Constraint - dyn_headleaf8_01_r (P)"(ComponentPose=@n152, Alpha=@n153, ConstraintWeights_0=1) {
    alphaBoolBlend {blendOption=Linear}
    boneToModify {boneName=dyn_headleaf8_01_r}
    constraintSetup [{targetBone {boneName=head}}]
    constraintWeights [1]
   }
   n155 = "Constraint - dyn_headleaf8_02_r (P)"(ComponentPose=@n154, Alpha=@n153, ConstraintWeights_0=1) {
    alphaBoolBlend {blendOption=Linear}
    boneToModify {boneName=dyn_headleaf8_02_r}
    constraintSetup [{targetBone {boneName=head}}]
    constraintWeights [1]
   }
   n156 = "Constraint - dyn_headleaf9_01_r (P)"(ComponentPose=@n155, Alpha=@n153, ConstraintWeights_0=1) {
    alphaBoolBlend {blendOption=Linear}
    boneToModify {boneName=dyn_headleaf9_01_r}
    constraintSetup [{targetBone {boneName=head}}]
    constraintWeights [1]
   }
   n157 = "Constraint - dyn_headleaf9_02_r (P)"(ComponentPose=@n156, Alpha=@n153, ConstraintWeights_0=1) {
    alphaBoolBlend {blendOption=Linear}
    boneToModify {boneName=dyn_headleaf9_02_r}
    constraintSetup [{targetBone {boneName=head}}]
    constraintWeights [1]
   }
   n158 = "Constraint - dyn_head_leaf_040 (P)"(ComponentPose=@n157, Alpha=@n153, ConstraintWeights_0=1) {
    alphaBoolBlend {blendOption=Linear}
    boneToModify {boneName=dyn_head_leaf_040}
    constraintSetup [{targetBone {boneName=head}}]
    constraintWeights [1]
   }
   n159 = "Constraint - dyn_head_leaf_020 (P)"(ComponentPose=@n158, Alpha=@n153, ConstraintWeights_0=1) {
    alphaBoolBlend {blendOption=Linear}
    boneToModify {boneName=dyn_head_leaf_020}
    constraintSetup [{targetBone {boneName=head}}]
    constraintWeights [1]
   }
   n160 = Look At - Bone: dyn_head_leaf_027_MSTR(ComponentPose=@n159, Alpha=1) {
    alphaBoolBlend {blendOption=Linear}
    boneToModify {boneName=dyn_head_leaf_027_MSTR}
    lookAtLocation {x=0}
    lookAtTarget {boneReference {boneName=dyn_head_leaf_aim_027}}
    lookAt_Axis {axis=<1 0>}
    lookUp_Axis {axis {x=0, z=1}}
   }
   n161 = Look At - Bone: dyn_head_leaf_019_MSTR(ComponentPose=@n160, Alpha=1) {
    alphaBoolBlend {blendOption=Linear}
    boneToModify {boneName=dyn_head_leaf_019_MSTR}
    lookAtLocation {x=0}
    lookAtTarget {boneReference {boneName=dyn_head_leaf_aim_019}}
    lookAt_Axis {axis=<1 0>}
    lookUp_Axis {axis {x=0, z=1}}
   }
   n162 = "Constraint - dyn_head_leaf_08_MSTR (RR)"(ComponentPose=@n161, Alpha=1, ConstraintWeights_0=0.5, ConstraintWeights_1=0.5) {
    alphaBoolBlend {blendOption=Linear}
    boneToModify {boneName=dyn_head_leaf_08_MSTR}
    constraintSetup [{transformType=Rotation, targetBone {boneName=neck_backhelper_01}}, {transformType=Rotation, targetBone {boneName=head}}]
    constraintWeights [1,1]
   }
   n163 = "Constraint - dyn_head_leaf_029_MSTR (RR)"!(ComponentPose=@n162, Alpha=1, ConstraintWeights_0=0.7, ConstraintWeights_1:float=0.3) -> (Pose) {
    alphaBoolBlend {blendOption=Linear}
    boneToModify {boneName=dyn_head_leaf_029_MSTR}
    constraintSetup [{transformType=Rotation, targetBone {boneName=neck_backhelper_01}}, {transformType=Rotation, targetBone {boneName=head}}]
    constraintWeights [1,1]
   }
   n164 = "Constraint - dyn_head_leaf_032_MSTR (RR)"!(ComponentPose=@n163, Alpha=1, ConstraintWeights_0=0.5, ConstraintWeights_1:float=0.5) -> (Pose) {
    alphaBoolBlend {blendOption=Linear}
    boneToModify {boneName=dyn_head_leaf_032_MSTR}
    constraintSetup [{transformType=Rotation, targetBone {boneName=neck_backhelper_01}}, {transformType=Rotation, targetBone {boneName=head}}]
    constraintWeights [1,1]
   }
   n165 = "Constraint - dyn_head_leaf_026_MSTR (RR)"!(ComponentPose=@n164, Alpha=1, ConstraintWeights_0=0.5, ConstraintWeights_1:float=0.5) -> (Pose) {
    alphaBoolBlend {blendOption=Linear}
    boneToModify {boneName=dyn_head_leaf_026_MSTR}
    constraintSetup [{transformType=Rotation, targetBone {boneName=neck_backhelper_01}}, {transformType=Rotation, targetBone {boneName=head}}]
    constraintWeights [1,1]
   }
   n166 = "Constraint - dyn_head_leaf_02_MSTR (RR)"!(ComponentPose=@n165, Alpha=1, ConstraintWeights_0=0.5, ConstraintWeights_1:float=0.5) -> (Pose) {
    alphaBoolBlend {blendOption=Linear}
    boneToModify {boneName=dyn_head_leaf_02_MSTR}
    constraintSetup [{transformType=Rotation, targetBone {boneName=neck_backhelper_01}}, {transformType=Rotation, targetBone {boneName=head}}]
    constraintWeights [1,1]
   }
   n167 = "Constraint - dyn_head_leaf_017_MSTR (RR)"!(ComponentPose=@n166, Alpha=1, ConstraintWeights_0=0.5, ConstraintWeights_1:float=0.5) -> (Pose) {
    alphaBoolBlend {blendOption=Linear}
    boneToModify {boneName=dyn_head_leaf_017_MSTR}
    constraintSetup [{transformType=Rotation, targetBone {boneName=neck_backhelper_01}}, {transformType=Rotation, targetBone {boneName=head}}]
    constraintWeights [1,1]
   }
   n168 = "Constraint - dyn_head_leaf_027_MSTR (RR)"!(ComponentPose=@n167, Alpha=0.5, ConstraintWeights_0=0.5, ConstraintWeights_1:float=0.5) -> (Pose) {
    alphaBoolBlend {blendOption=Linear}
    boneToModify {boneName=dyn_head_leaf_027_MSTR}
    constraintSetup [{transformType=Rotation, targetBone {boneName=neck_backhelper_01}}, {transformType=Rotation, targetBone {boneName=head}}]
    constraintWeights [1,1]
   }
   n169 = "Constraint - dyn_head_leaf_037_MSTR (PP)"!(ComponentPose=@n168, Alpha=1, ConstraintWeights_0=1, ConstraintWeights_1:float=0) -> (Pose) {
    alphaBoolBlend {blendOption=Linear}
    boneToModify {boneName=dyn_head_leaf_037_MSTR}
    constraintSetup [{targetBone {boneName=neck_backhelper_01}}, {targetBone {boneName=head}}]
    constraintWeights [1,1]
   }
   n170 = "Constraint - dyn_head_leaf_037_MSTR (RR)"!(ComponentPose=@n169, Alpha=1, ConstraintWeights_0=0.5, ConstraintWeights_1:float=0.5) -> (Pose) {
    alphaBoolBlend {blendOption=Linear}
    boneToModify {boneName=dyn_head_leaf_037_MSTR}
    constraintSetup [{transformType=Rotation, targetBone {boneName=neck_backhelper_01}}, {transformType=Rotation, targetBone {boneName=head}}]
    constraintWeights [1,1]
   }
   n171 = "Constraint - dyn_head_leaf_09_MSTR (RR)"!(ComponentPose=@n170, Alpha=1, ConstraintWeights_0=0.5, ConstraintWeights_1:float=0.5) -> (Pose) {
    alphaBoolBlend {blendOption=Linear}
    boneToModify {boneName=dyn_head_leaf_09_MSTR}
    constraintSetup [{transformType=Rotation, targetBone {boneName=neck_backhelper_01}}, {transformType=Rotation, targetBone {boneName=head}}]
    constraintWeights [1,1]
   }
   n172 = "Constraint - dyn_head_leaf_033_MSTR (RR)"!(ComponentPose=@n171, Alpha=1, ConstraintWeights_0=0.5, ConstraintWeights_1:float=0.5) -> (Pose) {
    alphaBoolBlend {blendOption=Linear}
    boneToModify {boneName=dyn_head_leaf_033_MSTR}
    constraintSetup [{transformType=Rotation, targetBone {boneName=neck_backhelper_01}}, {transformType=Rotation, targetBone {boneName=head}}]
    constraintWeights [1,1]
   }
   n173 = "Constraint - dyn_head_leaf_012_MSTR (RR)"!(ComponentPose=@n172, Alpha=1, ConstraintWeights_0=0.5, ConstraintWeights_1:float=0.5) -> (Pose) {
    alphaBoolBlend {blendOption=Linear}
    boneToModify {boneName=dyn_head_leaf_012_MSTR}
    constraintSetup [{transformType=Rotation, targetBone {boneName=neck_backhelper_01}}, {transformType=Rotation, targetBone {boneName=head}}]
    constraintWeights [1,1]
   }
   n174 = "Constraint - dyn_head_leaf_018_MSTR (RR)"!(ComponentPose=@n173, Alpha=1, ConstraintWeights_0=0.5, ConstraintWeights_1:float=0.5) -> (Pose) {
    alphaBoolBlend {blendOption=Linear}
    boneToModify {boneName=dyn_head_leaf_018_MSTR}
    constraintSetup [{transformType=Rotation, targetBone {boneName=neck_backhelper_01}}, {transformType=Rotation, targetBone {boneName=head}}]
    constraintWeights [1,1]
   }
   n175 = "Constraint - dyn_head_leaf_039_MSTR (PP)"!(ComponentPose=@n174, Alpha=1, ConstraintWeights_0=0.5, ConstraintWeights_1:float=0.5) -> (Pose) {
    alphaBoolBlend {blendOption=Linear}
    boneToModify {boneName=dyn_head_leaf_039_MSTR}
    constraintSetup [{targetBone {boneName=neck_backhelper_01}}, {targetBone {boneName=head}}]
    constraintWeights [1,1]
   }
   n176 = "Constraint - dyn_head_leaf_07_MSTR (RR)"!(ComponentPose=@n175, Alpha=1, ConstraintWeights_0=0.5, ConstraintWeights_1:float=0.5) -> (Pose) {
    alphaBoolBlend {blendOption=Linear}
    boneToModify {boneName=dyn_head_leaf_07_MSTR}
    constraintSetup [{transformType=Rotation, targetBone {boneName=neck_backhelper_01}}, {transformType=Rotation, targetBone {boneName=head}}]
    constraintWeights [1,1]
   }
   n177 = Get headSwingAngleBackLeavesDuringADS() -> (headSwingAngleBackLeavesDuringADS:double)
   n178 = "Transform (Modify) Bone - Bone: dyn_head_leaf_014_MSTR"(ComponentPose=@n176, Alpha=@n177, Translation="0,0,0", Rotation="-8,0.000,0.000", Scale="1,1,1") {
    rotationMode=BMM_Additive
    rotationSpace=BCS_BoneSpace
    scaleMode=BMM_Additive
    scaleSpace=BCS_BoneSpace
    translationMode=BMM_Additive
    translationSpace=BCS_BoneSpace
    alphaBoolBlend {blendOption=Linear}
    boneToModify {boneName=dyn_head_leaf_014_MSTR}
    rotation {pitch=-8}
   }
   n179 = Get headSwingAngleBackLeavesDuringADS_inv() -> (headSwingAngleBackLeavesDuringADS_inv:double)
   n180 = "Transform (Modify) Bone - Bone: dyn_head_leaf_012_MSTR"(ComponentPose=@n178, Alpha=@n179, Translation="0,0,0", Rotation="-8,0.000,0.000", Scale="1,1,1") {
    rotationMode=BMM_Additive
    rotationSpace=BCS_BoneSpace
    scaleMode=BMM_Additive
    scaleSpace=BCS_BoneSpace
    translationMode=BMM_Additive
    translationSpace=BCS_BoneSpace
    alphaBoolBlend {blendOption=Linear}
    boneToModify {boneName=dyn_head_leaf_012_MSTR}
    rotation {pitch=-8}
   }
   n181 = "Transform (Modify) Bone - Bone: dyn_head_leaf_041_MSTR"(ComponentPose=@n180, Alpha=@n179, Translation="0,0,0", Rotation="-5.000,0.000,0.000", Scale="1,1,1") {
    rotationMode=BMM_Additive
    rotationSpace=BCS_BoneSpace
    scaleMode=BMM_Additive
    scaleSpace=BCS_BoneSpace
    translationMode=BMM_Additive
    translationSpace=BCS_BoneSpace
    alphaBoolBlend {blendOption=Linear}
    boneToModify {boneName=dyn_head_leaf_041_MSTR}
    rotation {pitch=-5}
   }
   n182 = "Transform (Modify) Bone - Bone: dyn_head_leaf_039_MSTR"(ComponentPose=@n181, Alpha=@n179, Translation="0,0,0", Rotation="-5,0.000,0.000", Scale="1,1,1") {
    rotationMode=BMM_Additive
    rotationSpace=BCS_BoneSpace
    scaleMode=BMM_Additive
    scaleSpace=BCS_BoneSpace
    translationMode=BMM_Additive
    translationSpace=BCS_BoneSpace
    alphaBoolBlend {blendOption=Linear}
    boneToModify {boneName=dyn_head_leaf_039_MSTR}
    rotation {pitch=-5}
   }
   n183 = "Transform (Modify) Bone - Bone: dyn_head_leaf_07_MSTR"(ComponentPose=@n182, Alpha=@n179, Translation="0,0,0", Rotation="-5.000,0.000,0.000", Scale="1,1,1") {
    rotationMode=BMM_Additive
    rotationSpace=BCS_BoneSpace
    scaleMode=BMM_Additive
    scaleSpace=BCS_BoneSpace
    translationMode=BMM_Additive
    translationSpace=BCS_BoneSpace
    alphaBoolBlend {blendOption=Linear}
    boneToModify {boneName=dyn_head_leaf_07_MSTR}
    rotation {pitch=-5}
   }
   n184 = "Transform (Modify) Bone - Bone: dyn_head_leaf_039_MSTR"(ComponentPose=@n183, Alpha=@n179, Translation="0,0,0", Rotation="15,0.000,0.000", Scale="1,1,1") {
    rotationMode=BMM_Additive
    rotationSpace=BCS_BoneSpace
    scaleMode=BMM_Additive
    scaleSpace=BCS_BoneSpace
    translationMode=BMM_Additive
    translationSpace=BCS_BoneSpace
    alphaBoolBlend {blendOption=Linear}
    boneToModify {boneName=dyn_head_leaf_039_MSTR}
    rotation {pitch=15}
   }
   n185 = "Transform (Modify) Bone - Bone: dyn_head_leaf_07_MSTR"(ComponentPose=@n184, Alpha=@n179, Translation="0,0,0", Rotation="5,0.000,0.000", Scale="1,1,1") {
    rotationMode=BMM_Additive
    rotationSpace=BCS_BoneSpace
    scaleMode=BMM_Additive
    scaleSpace=BCS_BoneSpace
    translationMode=BMM_Additive
    translationSpace=BCS_BoneSpace
    alphaBoolBlend {blendOption=Linear}
    boneToModify {boneName=dyn_head_leaf_07_MSTR}
    rotation {pitch=5}
   }
   n186 = Get headSwingAngleBackLeavesDuringADS_top() -> (headSwingAngleBackLeavesDuringADS_top:double)
   n187 = "Transform (Modify) Bone - Bone: dyn_head_leaf_032_MSTR"(ComponentPose=@n185, Alpha=@n186, Translation="-0.117,0,-0.432", Rotation="-15.000,0.000,0.000", Scale="1,1,1") {
    rotationMode=BMM_Additive
    rotationSpace=BCS_BoneSpace
    scaleMode=BMM_Additive
    scaleSpace=BCS_BoneSpace
    translationMode=BMM_Additive
    translationSpace=BCS_BoneSpace
    alphaBoolBlend {blendOption=Linear}
    boneToModify {boneName=dyn_head_leaf_032_MSTR}
    rotation {pitch=-15}
    translation {x=-0.117, z=-0.432}
   }
   n188 = "Transform (Modify) Bone - Bone: dyn_head_leaf_026_MSTR"(ComponentPose=@n187, Alpha=@n186, Translation="-0.117,0,-0.432", Rotation="-15.000,0.000,0.000", Scale="1,1,1") {
    rotationMode=BMM_Additive
    rotationSpace=BCS_BoneSpace
    scaleMode=BMM_Additive
    scaleSpace=BCS_BoneSpace
    translationMode=BMM_Additive
    translationSpace=BCS_BoneSpace
    alphaBoolBlend {blendOption=Linear}
    boneToModify {boneName=dyn_head_leaf_026_MSTR}
    rotation {pitch=-15}
    translation {x=-0.117, z=-0.432}
   }
   n189 = "Transform (Modify) Bone - Bone: dyn_head_leaf_02_MSTR"(ComponentPose=@n188, Alpha=@n186, Translation="-0.117,-0.000,-0.432", Rotation="-15.000,0.000,0.000", Scale="1,1,1") {
    rotationMode=BMM_Additive
    rotationSpace=BCS_BoneSpace
    scaleMode=BMM_Additive
    scaleSpace=BCS_BoneSpace
    translationMode=BMM_Additive
    translationSpace=BCS_BoneSpace
    alphaBoolBlend {blendOption=Linear}
    boneToModify {boneName=dyn_head_leaf_02_MSTR}
    rotation {pitch=-15}
    translation {x=-0.117, z=-0.432}
   }
   n190 = "Transform (Modify) Bone - Bone: dyn_head_leaf_08_MSTR"(ComponentPose=@n189, Alpha=@n177, Translation="0,0,0", Rotation="-15,0.000,0.000", Scale="1,1,1") {
    rotationMode=BMM_Additive
    rotationSpace=BCS_BoneSpace
    scaleMode=BMM_Additive
    scaleSpace=BCS_BoneSpace
    translationMode=BMM_Additive
    translationSpace=BCS_BoneSpace
    alphaBoolBlend {blendOption=Linear}
    boneToModify {boneName=dyn_head_leaf_08_MSTR}
    rotation {pitch=-15}
   }
   n191 = "Constraint - dyn_head_leaf_017 (PP)"!(ComponentPose=@n190, Alpha=1, ConstraintWeights_0=0.5, ConstraintWeights_1:float=0.5) -> (Pose) {
    alphaBoolBlend {blendOption=Linear}
    boneToModify {boneName=dyn_head_leaf_017}
    constraintSetup [{targetBone {boneName=dyn_head_leaf_029}}, {targetBone {boneName=dyn_head_leaf_037}}]
    constraintWeights [1,1]
   }
   n192 = "Constraint - dyn_head_leaf_041_MSTR (RP)"!(ComponentPose=@n191, Alpha=1, ConstraintWeights_0=1, ConstraintWeights_1:float=0) -> (Pose) {
    alphaBoolBlend {blendOption=Linear}
    boneToModify {boneName=dyn_head_leaf_041_MSTR}
    constraintSetup [{transformType=Rotation, targetBone {boneName=dyn_head_leaf_012}}, {targetBone {boneName=dyn_head_leaf_037}}]
    constraintWeights [1,1]
   }
   n193 = "Constraint - dyn_head_leaf_032_MSTR (P)"(ComponentPose=@n192, Alpha=0.3, ConstraintWeights_0=1) {
    alphaBoolBlend {blendOption=Linear}
    boneToModify {boneName=dyn_head_leaf_032_MSTR}
    constraintSetup [{targetBone {boneName=dyn_neck_bk_02}}]
    constraintWeights [1]
   }
   n194 = "Constraint - dyn_head_leaf_08_MSTR (P)"(ComponentPose=@n193, Alpha=0.5, ConstraintWeights_0=1) {
    alphaBoolBlend {blendOption=Linear}
    boneToModify {boneName=dyn_head_leaf_08_MSTR}
    constraintSetup [{targetBone {boneName=dyn_neck_bk_02}}]
    constraintWeights [1]
   }
   n195 = "Constraint - dyn_head_leaf_029_MSTR (P)"(ComponentPose=@n194, Alpha=0.5, ConstraintWeights_0=1) {
    alphaBoolBlend {blendOption=Linear}
    boneToModify {boneName=dyn_head_leaf_029_MSTR}
    constraintSetup [{targetBone {boneName=dyn_neck_bk_02}}]
    constraintWeights [1]
   }
   n196 = Get isFrontendAlpha() -> (isFrontendAlpha:double)
   n197 = "Transform (Modify) Bone - Bone: dyn_headleaf2_01_l"(ComponentPose=@n195, Alpha=@n196, Translation="0,0,0", Rotation="5.000,0.000,0.000", Scale="1,1,1") {
    rotationMode=BMM_Additive
    rotationSpace=BCS_BoneSpace
    translationSpace=BCS_BoneSpace
    alphaBoolBlend {blendOption=Linear}
    boneToModify {boneName=dyn_headleaf2_01_l}
    rotation {pitch=5}
   }
   n198 = "Transform (Modify) Bone - Bone: dyn_headleaf3_01_l"(ComponentPose=@n197, Alpha=@n196, Translation="0,0,0", Rotation="23.000,0.000,0.000", Scale="1,1,1") {
    rotationMode=BMM_Additive
    rotationSpace=BCS_BoneSpace
    translationSpace=BCS_BoneSpace
    alphaBoolBlend {blendOption=Linear}
    boneToModify {boneName=dyn_headleaf3_01_l}
    rotation {pitch=23}
   }
   n199 = "Transform (Modify) Bone - Bone: dyn_headleaf4_01_l"(ComponentPose=@n198, Alpha=@n196, Translation="0,0,0", Rotation="28.000,0.000,0.000", Scale="1,1,1") {
    rotationMode=BMM_Additive
    rotationSpace=BCS_BoneSpace
    translationSpace=BCS_BoneSpace
    alphaBoolBlend {blendOption=Linear}
    boneToModify {boneName=dyn_headleaf4_01_l}
    rotation {pitch=28}
   }
   n200 = "Transform (Modify) Bone - Bone: neck_backhelper_01"(ComponentPose=@n199, Alpha=Get skydiveNeckFix() -> (skydiveNeckFix:double), Translation="-3,0,4", Rotation="10,0,0", Scale="1,1,1") {
    rotationMode=BMM_Additive
    rotationSpace=BCS_BoneSpace
    translationMode=BMM_Additive
    translationSpace=BCS_BoneSpace
    alphaBoolBlend {blendInTime=0.5, blendOption=Linear, blendOutTime=0.5}
    boneToModify {boneName=neck_backhelper_01}
    rotation {pitch=10}
    translation {x=-3, z=4}
   }
   n201 = Get bIsCrouching()
   n202 = "Constraint - neck_backhelper_01 (P)"!(ComponentPose=@n200, bAlphaBoolEnabled:bool=@n201, ConstraintWeights_0=1) -> (Pose) {
    alphaInputType=Bool
    alphaBoolBlend {blendOption=Linear}
    boneToModify {boneName=neck_backhelper_01}
    constraintSetup [{targetBone {boneName=spine_05}}]
    constraintWeights [1]
   }
   n203 = "Constraint - dyn_head_leaf_037 (P)"!(ComponentPose=@n202, bAlphaBoolEnabled:bool=@n201, ConstraintWeights_0=1) -> (Pose) {
    alphaInputType=Bool
    alphaBoolBlend {blendOption=Linear}
    boneToModify {boneName=dyn_head_leaf_037}
    constraintSetup [{targetBone {boneName=spine_05}}]
    constraintWeights [1]
   }
   n204 = "Constraint - dyn_neck_bk_01_r (P)"!(ComponentPose=@n203, bAlphaBoolEnabled:bool=True, ConstraintWeights_0=1) -> (Pose) {
    alphaInputType=Bool
    alphaBoolBlend {blendOption=Linear}
    boneToModify {boneName=dyn_neck_bk_01_r}
    constraintSetup [{targetBone {boneName=spine_05}}]
    constraintWeights [1]
   }
   Output Pose(Result=Component To Local(ComponentPose=@n204)) {layerGroup=None}
  ]}
 EventGraph {nodes [
   Event BlueprintUpdateAnimation()
   > n1 = Sequence!(execute) -> (then_0, then_1, then_2:Exec, then_3:Exec, then_4:Exec)
   n2 = Break Fort Anim Input Skydiving External Force(FortAnimInput_SkydivingExternalForce=Get SkydivingExternalForceInput()) -> (
    bUseSkydivingVectorForce=~
    SkydivingVectorForce=~
    HeadToPelvisDirection=~
    FloatingMultiplier=~
    FloatingAdditive=~
    DivingMultiplier=~
    DivingAdditive=~
    ParachutingMultiplier=~
    ParachutingAdditive=~
   )
   > Set windVector(windVector:Vector=SelectVector(A=vector * float(A=@n2.SkydivingVectorForce, B=0.15), B="0, 0,-980", bPickA=Get bIsSkyDiving())) -> (Output_Get:Vector)
   n3 = GetOwnerSkeletalMeshComponent()
   n4 = BreakRotator(InRot=BreakTransform(InTransform=GetDeltaTransformFromRefPose(self=@n3, BoneName=neck_02, BaseName=head)).Rotation)
   Set skydiveNeckFix(
    execute=@n1.then_1
    skydiveNeckFix:double=SelectFloat(A=MapRangeClamped(Value=@n4.Yaw, InRangeA=14, InRangeB=10), B=0, bPickA=Get bIsSkydiveFloating())
   ) -> (Output_Get:double)
   n5 = GetTwistAndSwingAngleOfDeltaRotationFromRefPose(self=@n3, BoneName=head)
   > Set headSwingAngleAlpha(headSwingAngleAlpha:double=SelectFloat(
     A=MapRangeClamped(Value=@n4.Yaw, InRangeA=0, InRangeB=25)
     B=MapRangeClamped(Value=@n5.OutSwingAngle, InRangeA=5, InRangeB=30)
     bPickA=Get bIsTargeting()
    )) -> (Output_Get:double)
   > Set headSwingAngleAlphaFrontBeardLeaves(headSwingAngleAlphaFrontBeardLeaves:double=MapRangeClamped(Value=@n5.OutSwingAngle, InRangeA=5, InRangeB=15, OutRangeA=1, OutRangeB=0)) -> (Output_Get:double)
   > n6 = Set headSwingAngleBackLeavesDuringADS(headSwingAngleBackLeavesDuringADS:double=SelectFloat(A=MapRangeClamped(Value=@n5.OutSwingAngle, InRangeA=20, InRangeB=30, OutRangeA=1, OutRangeB=0), B=0, bPickA=Get bIsTargeting())) -> (Output_Get:double)
   > Set headSwingAngleBackLeavesDuringADS_inv(headSwingAngleBackLeavesDuringADS_inv:double="Clamp (Float)"(Value=float - float(B=@n6) {is=title K2Node_CallFunction}, Min=0.5, Max=1)) -> (Output_Get:double)
   > Set headSwingAngleBackLeavesDuringADS_top(headSwingAngleBackLeavesDuringADS_top:double=0.2) -> (Output_Get:double)
   n7 = Get bIsTargeting()
   n8 = Get bIsCrouching()
   Set isCrouchingAlphaAndTargeting(execute=@n1.then_2, isCrouchingAlphaAndTargeting:double=SelectFloat(A=1, B=0, bPickA=AND Boolean(A=@n8, B=@n7))) -> (Output_Get:double)
   n9 = BreakTransform(InTransform=GetDeltaTransformFromRefPose(self=@n3, BoneName=head, BaseName=spine_05))
   Set headTwistAngle(execute=@n1.then_3, headTwistAngle:double=MapRangeClamped(Value=BreakRotator(InRot=@n9.Rotation).Pitch, InRangeA=5, InRangeB=40)) -> (Output_Get:double)
   Set isFrontendAlpha(execute=@n1.then_4, isFrontendAlpha:double="To Float (Boolean)"(InBool=Get bIsInFrontEnd())) -> (Output_Get:double)
   Event OnBeginTargeting() {is=title K2Node_Event}
   > Set headLeavesAimingAlpha(headLeavesAimingAlpha:double=0.7) -> (Output_Get:double)
   > Set isTargetingAlpha(isTargetingAlpha:double=1) -> (Output_Get:double)
   Event OnEndTargeting() {is=title K2Node_Event}
   > Set headLeavesAimingAlpha(headLeavesAimingAlpha=0)
   > Set isTargetingAlpha(isTargetingAlpha=0)
   Event OnBeginCrouching()
   > Set isCrouchingAlpha(isCrouchingAlpha:double=1) -> (Output_Get:double)
   > Set isNotCrouchingAlpha(isNotCrouchingAlpha:double=0) -> (Output_Get:double)
   Event OnEndCrouching()
   > Set isCrouchingAlpha(isCrouchingAlpha=0)
   > Set isNotCrouchingAlpha(isNotCrouchingAlpha=1)
   "float >= float"(A=Get PawnSpeed2D(), B=50) {is=title K2Node_CallFunction}
   "To Float (Boolean)"(InBool=Get bIsTargeting())
   TryGetPawnOwner()
  ]}
]
variables [
 headLeavesAimingAlpha {type=number}
 headSwingAngleAlpha {type=number}
 headSwingAngleAlphaFrontBeardLeaves {type=number}
 headSwingAngleBackLeavesDuringADS {type=number}
 headSwingAngleBackLeavesDuringADS_inv {type=number}
 headSwingAngleBackLeavesDuringADS_top {type=number}
 headSwingZRot {type=number}
 headTwistAngle {type=number}
 isCrouchingAlpha {type=number}
 isCrouchingAlphaAndTargeting {type=number}
 isFrontendAlpha {type=number}
 isNotCrouchingAlpha {type=number}
 isTargetingAlpha {type=number}
 skydiveNeckFix {type=number}
 windVector {type=Vector}
]
