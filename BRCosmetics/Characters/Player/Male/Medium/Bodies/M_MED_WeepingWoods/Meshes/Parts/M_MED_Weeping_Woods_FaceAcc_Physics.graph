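# Physics asset for the Weeping Woods male medium face accessory mesh, providing secondary-motion bodies for its branches, head leaves, neck and shoulder pieces.
# Tiny MSTR capsules anchor box and capsule chains to head or spine_05 through angularly free constraints, alongside larger head and spine capsules.
# Chain joints have tight swing and twist limits, except the wide-swinging center back-neck chain, whose tip also links to both side tips.
# Many head_leaf MSTR anchors have no chained bodies.
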PhysicsAsset /BRCosmetics/Characters/Player/Male/Medium/Bodies/M_MED_WeepingWoods/Meshes/Parts/M_MED_Weeping_Woods_FaceAcc_Physics
bodies {
 dyn_branch_01 [{shapeType=Capsule, length=6.098, radius=1.25, center=<3.6799 -0.0392 -0.0742>, rotation=<rot 0 -90.6104 -88.8448>}]
 dyn_branch_01_MSTR [{shapeType=Capsule, length=0.01, radius=0.01}]
 dyn_branch_01_l [{shapeType=Capsule, length=6.098, radius=1.25, center=<2.6888 0 0>, rotation=<rot 90 0.0791 0.0791>}]
 dyn_branch_01_l_MSTR [{shapeType=Capsule, length=0.01, radius=0.01}]
 dyn_branch_01_r [{shapeType=Capsule, length=6.098, radius=1.25, center=<2.6888 0 0>, rotation=<rot 90 0.0791 0.0791>}]
 dyn_branch_01_r_MSTR [{shapeType=Capsule, length=0.01, radius=0.01}]
 dyn_branch_02 [{shapeType=Capsule, length=6.098, radius=1.25, center=<3.6799 -0.0392 -0.0742>, rotation=<rot 0 -90.6104 -88.8448>}]
 dyn_branch_02_l [{shapeType=Box, extentX=4.5, extentY=2.5, extentZ=0.5, center=<2.577 0 0>}]
 dyn_branch_02_r [{shapeType=Box, extentX=4.5, extentY=2.5, extentZ=0.5, center=<2.577 0 0>}]
 dyn_branch_03 [{shapeType=Box, extentX=4.5, extentY=2.5, extentZ=0.5, center=<2.577 0 0>}]
 dyn_branch_03_l [{shapeType=Box, extentX=4.5, extentY=2.5, extentZ=0.5, center=<2.577 0 0>}]
 dyn_branch_04 [{shapeType=Box, extentX=4.5, extentY=2.5, extentZ=0.5, center=<2.577 0 0>, rotation=<rot 0 0 -89.9994>}]
 dyn_branch_05 [{shapeType=Box, extentX=2, extentY=2, extentZ=2, center=<1 0 0>}]
 dyn_head_leaf_012_MSTR [{shapeType=Capsule, length=0.01, radius=0.01}]
 dyn_head_leaf_013_MSTR [{shapeType=Capsule, length=0.01, radius=0.01}]
 dyn_head_leaf_014_MSTR [{shapeType=Capsule, length=0.01, radius=0.01}]
 dyn_head_leaf_015_MSTR [{shapeType=Capsule, length=0.01, radius=0.01}]
 dyn_head_leaf_016_MSTR [{shapeType=Capsule, length=0.01, radius=0.01}]
 dyn_head_leaf_017_MSTR [{shapeType=Capsule, length=0.01, radius=0.01}]
 dyn_head_leaf_018_MSTR [{shapeType=Capsule, length=0.01, radius=0.01}]
 dyn_head_leaf_019_MSTR [{shapeType=Capsule, length=0.01, radius=0.01}]
 dyn_head_leaf_01_MSTR [{shapeType=Capsule, length=0.01, radius=0.01}]
 dyn_head_leaf_020_MSTR [{shapeType=Capsule, length=0.01, radius=0.01}]
 dyn_head_leaf_023_MSTR [{shapeType=Capsule, length=0.01, radius=0.01}]
 dyn_head_leaf_026_MSTR [{shapeType=Capsule, length=0.01, radius=0.01}]
 dyn_head_leaf_027_MSTR [{shapeType=Capsule, length=0.01, radius=0.01}]
 dyn_head_leaf_028_MSTR [{shapeType=Capsule, length=0.01, radius=0.01}]
 dyn_head_leaf_029_MSTR [{shapeType=Capsule, length=0.01, radius=0.01}]
 dyn_head_leaf_02_MSTR [{shapeType=Capsule, length=0.01, radius=0.01}]
 dyn_head_leaf_030_MSTR [{shapeType=Capsule, length=0.01, radius=0.01}]
 dyn_head_leaf_032_MSTR [{shapeType=Capsule, length=0.01, radius=0.01}]
 dyn_head_leaf_033_MSTR [{shapeType=Capsule, length=0.01, radius=0.01}]
 dyn_head_leaf_034_MSTR [{shapeType=Capsule, length=0.01, radius=0.01}]
 dyn_head_leaf_035_MSTR [{shapeType=Capsule, length=0.01, radius=0.01}]
 dyn_head_leaf_036_MSTR [{shapeType=Capsule, length=0.01, radius=0.01}]
 dyn_head_leaf_037_MSTR [{shapeType=Capsule, length=0.01, radius=0.01}]
 dyn_head_leaf_039_MSTR [{shapeType=Capsule, length=0.01, radius=0.01}]
 dyn_head_leaf_03_MSTR [{shapeType=Capsule, length=0.01, radius=0.01}]
 dyn_head_leaf_040_MSTR [{shapeType=Capsule, length=0.01, radius=0.01}]
 dyn_head_leaf_041_MSTR [{shapeType=Capsule, length=0.01, radius=0.01}]
 dyn_head_leaf_042_MSTR [{shapeType=Capsule, length=0.01, radius=0.01}]
 dyn_head_leaf_05_MSTR [{shapeType=Capsule, length=0.01, radius=0.01}]
 dyn_head_leaf_06_MSTR [{shapeType=Capsule, length=0.01, radius=0.01}]
 dyn_head_leaf_07_MSTR [{shapeType=Capsule, length=0.01, radius=0.01}]
 dyn_head_leaf_08_MSTR [{shapeType=Capsule, length=0.01, radius=0.01}]
 dyn_head_leaf_09_MSTR [{shapeType=Capsule, length=0.01, radius=0.01}]
 dyn_headleaf10_01_l [{shapeType=Box, extentX=5.5, extentY=4, extentZ=2, center=<2.5614 0 0>}]
 dyn_headleaf10_01_l_MSTR [{shapeType=Capsule, length=0.01, radius=0.01}]
 dyn_headleaf10_01_r [{shapeType=Box, extentX=5.5, extentY=4, extentZ=2, center=<2.5614 0 0>}]
 dyn_headleaf10_01_r_MSTR [{shapeType=Capsule, length=0.01, radius=0.01}]
 dyn_headleaf10_02_l [{shapeType=Box, extentX=5.5, extentY=4, extentZ=2, center=<2.5614 0 0>}]
 dyn_headleaf10_02_r [{shapeType=Box, extentX=5.5, extentY=4, extentZ=2, center=<2.5614 0 0>}]
 dyn_headleaf11_01_l [{shapeType=Box, extentX=5.5, extentY=4, extentZ=2, center=<2.5614 0 0>}]
 dyn_headleaf11_01_l_MSTR [{shapeType=Capsule, length=0.01, radius=0.01}]
 dyn_headleaf11_02_l [{shapeType=Box, extentX=5.5, extentY=4, extentZ=2, center=<2.5614 0 0>}]
 dyn_headleaf11_03_l [{shapeType=Box, extentX=5.5, extentY=4, extentZ=2, center=<2.5614 0 0>}]
 dyn_headleaf1_01_l [{shapeType=Box, extentX=5.5, extentY=4, extentZ=2, center=<2.5614 0 0>}]
 dyn_headleaf1_01_l_MSTR [{shapeType=Capsule, length=0.01, radius=0.01}]
 dyn_headleaf1_02_l [{shapeType=Box, extentX=5.5, extentY=4, extentZ=2, center=<2.5614 0 0>}]
 dyn_headleaf1_02_r [{shapeType=Box, extentX=5.5, extentY=4, extentZ=2, center=<2.5614 0 0>}]
 dyn_headleaf1_03_l [{shapeType=Box, extentX=5.5, extentY=4, extentZ=2, center=<2.5614 0 0>}]
 dyn_headleaf2_01_l [{shapeType=Box, extentX=5.5, extentY=4, extentZ=2, center=<2.5614 0 0>}]
 dyn_headleaf2_01_l_MSTR [{shapeType=Capsule, length=0.01, radius=0.01}]
 dyn_headleaf2_01_r [{shapeType=Box, extentX=5.5, extentY=4, extentZ=2, center=<2.5614 0 0>}]
 dyn_headleaf2_01_r_MSTR [{shapeType=Capsule, length=0.01, radius=0.01}]
 dyn_headleaf2_02_l [{shapeType=Box, extentX=5.5, extentY=4, extentZ=2, center=<2.5614 0 0>}]
 dyn_headleaf2_03_l [{shapeType=Box, extentX=5.5, extentY=4, extentZ=2, center=<2.5614 0 0>}]
 dyn_headleaf3_01_l [{shapeType=Box, extentX=5.5, extentY=4, extentZ=2, center=<2.5614 0 0>}]
 dyn_headleaf3_01_l_MSTR [{shapeType=Capsule, length=0.01, radius=0.01}]
 dyn_headleaf3_01_r [{shapeType=Box, extentX=5.5, extentY=4, extentZ=2, center=<2.5614 0 0>}]
 dyn_headleaf3_01_r_MSTR [{shapeType=Capsule, length=0.01, radius=0.01}]
 dyn_headleaf4_01_l [{shapeType=Box, extentX=5.5, extentY=4, extentZ=2, center=<2.5614 0 0>}]
 dyn_headleaf4_01_l_MSTR [{shapeType=Capsule, length=0.01, radius=0.01}]
 dyn_headleaf4_01_r [{shapeType=Box, extentX=5.5, extentY=4, extentZ=2, center=<2.5614 0 0>}]
 dyn_headleaf4_01_r_MSTR [{shapeType=Capsule, length=0.01, radius=0.01}]
 dyn_headleaf4_02_l [{shapeType=Box, extentX=5.5, extentY=4, extentZ=2, center=<2.5614 0 0>}]
 dyn_headleaf4_02_r [{shapeType=Box, extentX=5.5, extentY=4, extentZ=2, center=<2.5614 0 0>}]
 dyn_headleaf5_01_l [{shapeType=Box, extentX=5.5, extentY=4, extentZ=2, center=<2.5614 0 0>}]
 dyn_headleaf5_01_l_MSTR [{shapeType=Capsule, length=0.01, radius=0.01}]
 dyn_headleaf5_01_r [{shapeType=Box, extentX=5.5, extentY=4, extentZ=2, center=<2.5614 0 0>}]
 dyn_headleaf5_01_r_MSTR [{shapeType=Capsule, length=0.01, radius=0.01}]
 dyn_headleaf5_02_r [{shapeType=Box, extentX=5.5, extentY=4, extentZ=2, center=<2.5614 0 0>}]
 dyn_headleaf6_01_l [{shapeType=Box, extentX=5.5, extentY=4, extentZ=2, center=<2.5614 0 0>}]
 dyn_headleaf6_01_l_MSTR [{shapeType=Capsule, length=0.01, radius=0.01}]
 dyn_headleaf6_02_l [{shapeType=Box, extentX=5.5, extentY=4, extentZ=2, center=<2.5614 0 0>}]
 dyn_headleaf6_03_l [{shapeType=Box, extentX=5.5, extentY=4, extentZ=2, center=<2.5614 0 0>}]
 dyn_headleaf7_01_l [{shapeType=Box, extentX=5.5, extentY=4, extentZ=2, center=<2.5614 0 0>}]
 dyn_headleaf7_01_l_MSTR [{shapeType=Capsule, length=0.01, radius=0.01}]
 dyn_headleaf7_01_r [{shapeType=Box, extentX=5.5, extentY=4, extentZ=2, center=<2.5614 0 0>}]
 dyn_headleaf7_01_r_MSTR [{shapeType=Capsule, length=0.01, radius=0.01}]
 dyn_headleaf7_02_l [{shapeType=Box, extentX=5.5, extentY=4, extentZ=2, center=<2.5614 0 0>}]
 dyn_headleaf7_03_l [{shapeType=Box, extentX=5.5, extentY=4, extentZ=2, center=<2.5614 0 0>}]
 dyn_headleaf8_01_l [{shapeType=Box, extentX=5.5, extentY=4, extentZ=2, center=<2.5614 0 0>}]
 dyn_headleaf8_01_l_MSTR [{shapeType=Capsule, length=0.01, radius=0.01}]
 dyn_headleaf8_01_r [{shapeType=Box, extentX=5.5, extentY=4, extentZ=2, center=<2.5614 0 0>}]
 dyn_headleaf8_01_r_MSTR [{shapeType=Capsule, length=0.01, radius=0.01}]
 dyn_headleaf8_02_l [{shapeType=Box, extentX=5.5, extentY=4, extentZ=2, center=<2.5614 0 0>}]
 dyn_headleaf8_02_r [{shapeType=Box, extentX=5.5, extentY=4, extentZ=2, center=<2.5614 0 0>}]
 dyn_headleaf8_03_l [{shapeType=Box, extentX=5.5, extentY=4, extentZ=2, center=<2.5614 0 0>}]
 dyn_headleaf9_01_l [{shapeType=Box, extentX=5.5, extentY=4, extentZ=2, center=<2.5614 0 0>}]
 dyn_headleaf9_01_l_MSTR [{shapeType=Capsule, length=0.01, radius=0.01}]
 dyn_headleaf9_01_r [{shapeType=Box, extentX=5.5, extentY=4, extentZ=2, center=<2.5614 0 0>}]
 dyn_headleaf9_01_r_MSTR [{shapeType=Capsule, length=0.01, radius=0.01}]
 dyn_headleaf9_02_l [{shapeType=Box, extentX=5.5, extentY=4, extentZ=2, center=<2.5614 0 0>}]
 dyn_headleaf9_02_r [{shapeType=Box, extentX=5.5, extentY=4, extentZ=2, center=<2.5614 0 0>}]
 dyn_headleaf9_03_r [{shapeType=Box, extentX=5.5, extentY=4, extentZ=2, center=<2.5614 0 0>}]
 dyn_neck_bk_01 [{shapeType=Box, extentX=5.5, extentY=9.5, extentZ=2.5, center=<2.5614 0 0>}]
 dyn_neck_bk_01_MSTR [{shapeType=Capsule, length=0.01, radius=0.01}]
 dyn_neck_bk_01_l [{shapeType=Box, extentX=5.5, extentY=9.5, extentZ=2, center=<2.5614 0 0>}]
 dyn_neck_bk_01_l_MSTR [{shapeType=Capsule, length=0.01, radius=0.01}]
 dyn_neck_bk_01_r [{shapeType=Box, extentX=5.5, extentY=9.5, extentZ=2, center=<2.5614 0 0>}]
 dyn_neck_bk_01_r_MSTR [{shapeType=Capsule, length=0.01, radius=0.01}]
 dyn_neck_bk_02 [{shapeType=Box, extentX=12, extentY=15, extentZ=2.5, center=<6.2886 0 0>}]
 dyn_neck_bk_02_l [{shapeType=Box, extentX=5.5, extentY=9.5, extentZ=2, center=<2.5614 0 0>}]
 dyn_neck_bk_02_r [{shapeType=Box, extentX=5.5, extentY=9.5, extentZ=2, center=<2.5614 0 0>}]
 dyn_neck_fr_01 [{shapeType=Box, extentX=5.5, extentY=9.5, extentZ=2, center=<2.5614 0 0>}]
 dyn_neck_fr_01_MSTR [{shapeType=Capsule, length=0.01, radius=0.01}]
 dyn_neck_fr_01_l [{shapeType=Box, extentX=5.5, extentY=9.5, extentZ=2, center=<2.5614 0 0>}]
 dyn_neck_fr_01_l_MSTR [{shapeType=Capsule, length=0.01, radius=0.01}]
 dyn_neck_fr_01_r [{shapeType=Box, extentX=5.5, extentY=9.5, extentZ=2, center=<2.5614 0 0>}]
 dyn_neck_fr_01_r_MSTR [{shapeType=Capsule, length=0.01, radius=0.01}]
 dyn_neck_fr_02 [{shapeType=Box, extentX=5.5, extentY=9.5, extentZ=2, center=<2.5614 0 0>}]
 dyn_neck_fr_02_l [{shapeType=Box, extentX=5.5, extentY=9.5, extentZ=2, center=<2.5614 0 0>}]
 dyn_neck_fr_02_r [{shapeType=Box, extentX=5.5, extentY=9.5, extentZ=2, center=<2.5614 0 0>}]
 dyn_shoulder_01_l [{shapeType=Box, extentX=2, extentY=2, extentZ=2, center=<1 0 0>}]
 dyn_shoulder_01_l_MSTR [{shapeType=Capsule, length=0.01, radius=0.01}]
 dyn_shoulder_01_r [{shapeType=Box, extentX=2, extentY=2, extentZ=2, center=<1 0 0>}]
 dyn_shoulder_01_r_MSTR [{shapeType=Capsule, length=0.01, radius=0.01}]
 head [{shapeType=Capsule, length=14.7278, radius=8.0113, rotation=<rot -90 -1.2659 1.2659>}]
 spine_04 [{shapeType=Capsule, length=7.02, radius=14.2494, center=<4.243 -3.3415 0.8424>, rotation=<rot 0 90 0>}]
 spine_05 [{shapeType=Capsule, length=9.77, radius=11.4994, center=<4.243 -3.3415 0.8424>, rotation=<rot 0 90 0>}]
}
constraints [
 {bone1Name=dyn_shoulder_01_l, bone2Name=dyn_shoulder_01_l_MSTR, swing1LimitDegrees=10, swing2LimitDegrees=10, twistLimitDegrees=5}
 {bone1Name=dyn_shoulder_01_r, bone2Name=dyn_shoulder_01_r_MSTR, swing1LimitDegrees=10, swing2LimitDegrees=10, twistLimitDegrees=5}
 {bone1Name=dyn_headleaf10_01_r, bone2Name=dyn_headleaf10_01_r_MSTR, swing1LimitDegrees=5, swing2LimitDegrees=10, twistLimitDegrees=5}
 {bone1Name=dyn_headleaf10_02_r, bone2Name=dyn_headleaf10_01_r, swing1LimitDegrees=5, swing2LimitDegrees=10, twistLimitDegrees=5}
 {bone1Name=dyn_headleaf9_01_r, bone2Name=dyn_headleaf9_01_r_MSTR, swing1LimitDegrees=5, swing2LimitDegrees=15, twistLimitDegrees=5}
 {bone1Name=dyn_headleaf9_02_r, bone2Name=dyn_headleaf9_01_r, swing1LimitDegrees=5, swing2LimitDegrees=15, twistLimitDegrees=5}
 {bone1Name=dyn_headleaf9_03_r, bone2Name=dyn_headleaf9_02_r, swing1LimitDegrees=5, swing2LimitDegrees=15, twistLimitDegrees=5}
 {bone1Name=dyn_headleaf8_01_r, bone2Name=dyn_headleaf8_01_r_MSTR, swing1LimitDegrees=5, swing2LimitDegrees=10, twistLimitDegrees=5}
 {bone1Name=dyn_headleaf8_02_r, bone2Name=dyn_headleaf8_01_r, swing1LimitDegrees=5, swing2LimitDegrees=10, twistLimitDegrees=5}
 {bone1Name=dyn_headleaf7_01_r, bone2Name=dyn_headleaf7_01_r_MSTR, swing1LimitDegrees=5, swing2LimitDegrees=10, twistLimitDegrees=5}
 {bone1Name=dyn_headleaf5_01_r, bone2Name=dyn_headleaf5_01_r_MSTR, swing1LimitDegrees=5, swing2LimitDegrees=10, twistLimitDegrees=5}
 {bone1Name=dyn_headleaf5_02_r, bone2Name=dyn_headleaf5_01_r, swing1LimitDegrees=5, swing2LimitDegrees=10, twistLimitDegrees=5}
 {bone1Name=dyn_headleaf4_01_r, bone2Name=dyn_headleaf4_01_r_MSTR, swing1LimitDegrees=5, swing2LimitDegrees=10, twistLimitDegrees=5}
 {bone1Name=dyn_headleaf4_02_r, bone2Name=dyn_headleaf4_01_r, swing1LimitDegrees=5, swing2LimitDegrees=10, twistLimitDegrees=5}
 {bone1Name=dyn_headleaf3_01_r, bone2Name=dyn_headleaf3_01_r_MSTR, swing1LimitDegrees=5, swing2LimitDegrees=10, twistLimitDegrees=5}
 {bone1Name=dyn_headleaf2_01_r, bone2Name=dyn_headleaf2_01_r_MSTR, swing1LimitDegrees=5, swing2LimitDegrees=10, twistLimitDegrees=5}
 {bone1Name=dyn_headleaf1_02_r, bone2Name=dyn_headleaf2_01_r, swing1LimitDegrees=5, swing2LimitDegrees=10, twistLimitDegrees=5}
 {bone1Name=dyn_headleaf4_01_l, bone2Name=dyn_headleaf4_01_l_MSTR, swing1LimitDegrees=5, swing2LimitDegrees=10, twistLimitDegrees=5}
 {bone1Name=dyn_headleaf4_02_l, bone2Name=dyn_headleaf4_01_l, swing1LimitDegrees=5, swing2LimitDegrees=10, twistLimitDegrees=5}
 {bone1Name=dyn_headleaf3_01_l, bone2Name=dyn_headleaf3_01_l_MSTR, swing1LimitDegrees=5, swing2LimitDegrees=10, twistLimitDegrees=5}
 {bone1Name=dyn_headleaf1_01_l, bone2Name=dyn_headleaf1_01_l_MSTR, swing1LimitDegrees=5, swing2LimitDegrees=10, twistLimitDegrees=5}
 {bone1Name=dyn_headleaf2_02_l, bone2Name=dyn_headleaf1_01_l, swing1LimitDegrees=5, swing2LimitDegrees=10, twistLimitDegrees=5}
 {bone1Name=dyn_headleaf2_03_l, bone2Name=dyn_headleaf2_02_l, swing1LimitDegrees=5, swing2LimitDegrees=10, twistLimitDegrees=5}
 {bone1Name=dyn_headleaf2_01_l, bone2Name=dyn_headleaf2_01_l_MSTR, swing1LimitDegrees=5, swing2LimitDegrees=10, twistLimitDegrees=5}
 {bone1Name=dyn_headleaf1_02_l, bone2Name=dyn_headleaf2_01_l, swing1LimitDegrees=5, swing2LimitDegrees=10, twistLimitDegrees=5}
 {bone1Name=dyn_headleaf1_03_l, bone2Name=dyn_headleaf1_02_l, swing1LimitDegrees=5, swing2LimitDegrees=10, twistLimitDegrees=5}
 {bone1Name=dyn_headleaf5_01_l, bone2Name=dyn_headleaf5_01_l_MSTR, swing1LimitDegrees=5, swing2LimitDegrees=10, twistLimitDegrees=5}
 {bone1Name=dyn_headleaf6_01_l, bone2Name=dyn_headleaf6_01_l_MSTR, swing1LimitDegrees=5, swing2LimitDegrees=10, twistLimitDegrees=5}
 {bone1Name=dyn_headleaf6_02_l, bone2Name=dyn_headleaf6_01_l, swing1LimitDegrees=5, swing2LimitDegrees=10, twistLimitDegrees=5}
 {bone1Name=dyn_headleaf6_03_l, bone2Name=dyn_headleaf6_02_l, swing1LimitDegrees=5, swing2LimitDegrees=10, twistLimitDegrees=5}
 {bone1Name=dyn_headleaf10_01_l, bone2Name=dyn_headleaf10_01_l_MSTR, swing1LimitDegrees=5, swing2LimitDegrees=10, twistLimitDegrees=5}
 {bone1Name=dyn_headleaf10_02_l, bone2Name=dyn_headleaf10_01_l, swing1LimitDegrees=5, swing2LimitDegrees=10, twistLimitDegrees=5}
 {bone1Name=dyn_headleaf11_01_l, bone2Name=dyn_headleaf11_01_l_MSTR, swing1LimitDegrees=5, swing2LimitDegrees=10, twistLimitDegrees=5}
 {bone1Name=dyn_headleaf11_02_l, bone2Name=dyn_headleaf11_01_l, swing1LimitDegrees=5, swing2LimitDegrees=10, twistLimitDegrees=5}
 {bone1Name=dyn_headleaf11_03_l, bone2Name=dyn_headleaf11_02_l, swing1LimitDegrees=5, swing2LimitDegrees=10, twistLimitDegrees=5}
 {bone1Name=dyn_headleaf9_01_l, bone2Name=dyn_headleaf9_01_l_MSTR, swing1LimitDegrees=5, swing2LimitDegrees=10, twistLimitDegrees=5}
 {bone1Name=dyn_headleaf9_02_l, bone2Name=dyn_headleaf9_01_l, swing1LimitDegrees=5, swing2LimitDegrees=10, twistLimitDegrees=5}
 {bone1Name=dyn_headleaf8_01_l, bone2Name=dyn_headleaf8_01_l_MSTR, swing1LimitDegrees=5, swing2LimitDegrees=10, twistLimitDegrees=5}
 {bone1Name=dyn_headleaf8_02_l, bone2Name=dyn_headleaf8_01_l, swing1LimitDegrees=5, swing2LimitDegrees=10, twistLimitDegrees=5}
 {bone1Name=dyn_headleaf8_03_l, bone2Name=dyn_headleaf8_02_l, swing1LimitDegrees=5, swing2LimitDegrees=10, twistLimitDegrees=5}
 {bone1Name=dyn_headleaf7_01_l, bone2Name=dyn_headleaf7_01_l_MSTR, swing1LimitDegrees=5, swing2LimitDegrees=10, twistLimitDegrees=5}
 {bone1Name=dyn_headleaf7_02_l, bone2Name=dyn_headleaf7_01_l, swing1LimitDegrees=5, swing2LimitDegrees=10, twistLimitDegrees=5}
 {bone1Name=dyn_headleaf7_03_l, bone2Name=dyn_headleaf7_02_l, swing1LimitDegrees=5, swing2LimitDegrees=10, twistLimitDegrees=5}
 {bone1Name=dyn_branch_01_r, bone2Name=dyn_branch_01_r_MSTR, swing1LimitDegrees=10, swing2LimitDegrees=10, twistLimitDegrees=5}
 {bone1Name=dyn_branch_03_l, bone2Name=dyn_branch_01_l, swing1LimitDegrees=10, swing2LimitDegrees=10, twistLimitDegrees=5}
 {bone1Name=dyn_branch_02_l, bone2Name=dyn_branch_01_l, swing1LimitDegrees=10, swing2LimitDegrees=10, twistLimitDegrees=5}
 {bone1Name=dyn_branch_02_r, bone2Name=dyn_branch_01_r, swing1LimitDegrees=10, swing2LimitDegrees=10, twistLimitDegrees=5}
 {bone1Name=dyn_branch_05, bone2Name=dyn_branch_01, swing1LimitDegrees=10, swing2LimitDegrees=10, twistLimitDegrees=5}
 {bone1Name=dyn_branch_04, bone2Name=dyn_branch_02, swing1LimitDegrees=10, swing2LimitDegrees=10, twistLimitDegrees=5}
 {bone1Name=dyn_branch_03, bone2Name=dyn_branch_02, swing1LimitDegrees=10, swing2LimitDegrees=10, twistLimitDegrees=5}
 {bone1Name=dyn_branch_01, bone2Name=dyn_branch_01_MSTR, swing1LimitDegrees=10, swing2LimitDegrees=10, twistLimitDegrees=5}
 {bone1Name=dyn_branch_02, bone2Name=dyn_branch_01, swing1LimitDegrees=10, swing2LimitDegrees=10, twistLimitDegrees=5}
 {bone1Name=dyn_neck_fr_01_l, bone2Name=dyn_neck_fr_01_l_MSTR, swing1LimitDegrees=10, swing2LimitDegrees=10, twistLimitDegrees=5}
 {bone1Name=dyn_neck_fr_02_l, bone2Name=dyn_neck_fr_01_l, swing1LimitDegrees=10, swing2LimitDegrees=10, twistLimitDegrees=5}
 {bone1Name=dyn_neck_fr_01_r, bone2Name=dyn_neck_fr_01_r_MSTR, swing1LimitDegrees=10, swing2LimitDegrees=10, twistLimitDegrees=5}
 {bone1Name=dyn_neck_fr_02_r, bone2Name=dyn_neck_fr_01_r, swing1LimitDegrees=10, swing2LimitDegrees=10, twistLimitDegrees=5}
 {bone1Name=dyn_neck_bk_01_r, bone2Name=dyn_neck_bk_01_r_MSTR, swing1LimitDegrees=10, swing2LimitDegrees=10, twistLimitDegrees=5}
 {bone1Name=dyn_neck_bk_02_r, bone2Name=dyn_neck_bk_01_r, swing1LimitDegrees=10, swing2LimitDegrees=10, twistLimitDegrees=5}
 {bone1Name=dyn_neck_bk_01_l, bone2Name=dyn_neck_bk_01_l_MSTR, swing1LimitDegrees=10, swing2LimitDegrees=10, twistLimitDegrees=5}
 {bone1Name=dyn_neck_bk_02_l, bone2Name=dyn_neck_bk_01_l, swing1LimitDegrees=10, swing2LimitDegrees=10, twistLimitDegrees=5}
 {bone1Name=dyn_neck_bk_01, bone2Name=dyn_neck_bk_01_MSTR, swing1LimitDegrees=5, swing2LimitDegrees=75, twistLimitDegrees=5}
 {bone1Name=dyn_neck_bk_02, bone2Name=dyn_neck_bk_01, swing1LimitDegrees=5, swing2LimitDegrees=75, twistLimitDegrees=5}
 {bone1Name=dyn_neck_fr_01, bone2Name=dyn_neck_fr_01_MSTR, swing1LimitDegrees=10, swing2LimitDegrees=10, twistLimitDegrees=5}
 {bone1Name=dyn_neck_fr_02, bone2Name=dyn_neck_fr_01, swing1LimitDegrees=10, swing2LimitDegrees=10, twistLimitDegrees=5}
 {bone1Name=dyn_branch_01_l, bone2Name=dyn_branch_01_l_MSTR, swing1LimitDegrees=10, swing2LimitDegrees=10, twistLimitDegrees=5}
 {
  bone1Name=dyn_headleaf10_01_r_MSTR
  bone2Name=head
  swing1LimitDegrees=45
  swing1Motion=Free
  swing2LimitDegrees=45
  swing2Motion=Free
  twistLimitDegrees=45
  twistMotion=Free
 }
 {
  bone1Name=dyn_headleaf9_01_r_MSTR
  bone2Name=head
  swing1LimitDegrees=45
  swing1Motion=Free
  swing2LimitDegrees=45
  swing2Motion=Free
  twistLimitDegrees=45
  twistMotion=Free
 }
 {
  bone1Name=dyn_headleaf8_01_r_MSTR
  bone2Name=head
  swing1LimitDegrees=45
  swing1Motion=Free
  swing2LimitDegrees=45
  swing2Motion=Free
  twistLimitDegrees=45
  twistMotion=Free
 }
 {
  bone1Name=dyn_headleaf7_01_r_MSTR
  bone2Name=head
  swing1LimitDegrees=45
  swing1Motion=Free
  swing2LimitDegrees=45
  swing2Motion=Free
  twistLimitDegrees=45
  twistMotion=Free
 }
 {
  bone1Name=dyn_headleaf5_01_r_MSTR
  bone2Name=head
  swing1LimitDegrees=45
  swing1Motion=Free
  swing2LimitDegrees=45
  swing2Motion=Free
  twistLimitDegrees=45
  twistMotion=Free
 }
 {
  bone1Name=dyn_headleaf4_01_r_MSTR
  bone2Name=head
  swing1LimitDegrees=45
  swing1Motion=Free
  swing2LimitDegrees=45
  swing2Motion=Free
  twistLimitDegrees=45
  twistMotion=Free
 }
 {
  bone1Name=dyn_headleaf3_01_r_MSTR
  bone2Name=head
  swing1LimitDegrees=45
  swing1Motion=Free
  swing2LimitDegrees=45
  swing2Motion=Free
  twistLimitDegrees=45
  twistMotion=Free
 }
 {
  bone1Name=dyn_headleaf2_01_r_MSTR
  bone2Name=head
  swing1LimitDegrees=45
  swing1Motion=Free
  swing2LimitDegrees=45
  swing2Motion=Free
  twistLimitDegrees=45
  twistMotion=Free
 }
 {
  bone1Name=dyn_headleaf4_01_l_MSTR
  bone2Name=head
  swing1LimitDegrees=45
  swing1Motion=Free
  swing2LimitDegrees=45
  swing2Motion=Free
  twistLimitDegrees=45
  twistMotion=Free
 }
 {
  bone1Name=dyn_headleaf3_01_l_MSTR
  bone2Name=head
  swing1LimitDegrees=45
  swing1Motion=Free
  swing2LimitDegrees=45
  swing2Motion=Free
  twistLimitDegrees=45
  twistMotion=Free
 }
 {
  bone1Name=dyn_headleaf1_01_l_MSTR
  bone2Name=head
  swing1LimitDegrees=45
  swing1Motion=Free
  swing2LimitDegrees=45
  swing2Motion=Free
  twistLimitDegrees=45
  twistMotion=Free
 }
 {
  bone1Name=dyn_headleaf2_01_l_MSTR
  bone2Name=head
  swing1LimitDegrees=45
  swing1Motion=Free
  swing2LimitDegrees=45
  swing2Motion=Free
  twistLimitDegrees=45
  twistMotion=Free
 }
 {
  bone1Name=dyn_headleaf5_01_l_MSTR
  bone2Name=head
  swing1LimitDegrees=45
  swing1Motion=Free
  swing2LimitDegrees=45
  swing2Motion=Free
  twistLimitDegrees=45
  twistMotion=Free
 }
 {
  bone1Name=dyn_headleaf6_01_l_MSTR
  bone2Name=head
  swing1LimitDegrees=45
  swing1Motion=Free
  swing2LimitDegrees=45
  swing2Motion=Free
  twistLimitDegrees=45
  twistMotion=Free
 }
 {
  bone1Name=dyn_headleaf10_01_l_MSTR
  bone2Name=head
  swing1LimitDegrees=45
  swing1Motion=Free
  swing2LimitDegrees=45
  swing2Motion=Free
  twistLimitDegrees=45
  twistMotion=Free
 }
 {
  bone1Name=dyn_headleaf11_01_l_MSTR
  bone2Name=head
  swing1LimitDegrees=45
  swing1Motion=Free
  swing2LimitDegrees=45
  swing2Motion=Free
  twistLimitDegrees=45
  twistMotion=Free
 }
 {
  bone1Name=dyn_headleaf9_01_l_MSTR
  bone2Name=head
  swing1LimitDegrees=45
  swing1Motion=Free
  swing2LimitDegrees=45
  swing2Motion=Free
  twistLimitDegrees=45
  twistMotion=Free
 }
 {
  bone1Name=dyn_headleaf8_01_l_MSTR
  bone2Name=head
  swing1LimitDegrees=45
  swing1Motion=Free
  swing2LimitDegrees=45
  swing2Motion=Free
  twistLimitDegrees=45
  twistMotion=Free
 }
 {
  bone1Name=dyn_headleaf7_01_l_MSTR
  bone2Name=head
  swing1LimitDegrees=45
  swing1Motion=Free
  swing2LimitDegrees=45
  swing2Motion=Free
  twistLimitDegrees=45
  twistMotion=Free
 }
 {
  bone1Name=dyn_branch_01_r_MSTR
  bone2Name=head
  swing1LimitDegrees=45
  swing1Motion=Free
  swing2LimitDegrees=45
  swing2Motion=Free
  twistLimitDegrees=45
  twistMotion=Free
 }
 {
  bone1Name=dyn_branch_01_MSTR
  bone2Name=head
  swing1LimitDegrees=45
  swing1Motion=Free
  swing2LimitDegrees=45
  swing2Motion=Free
  twistLimitDegrees=45
  twistMotion=Free
 }
 {
  bone1Name=dyn_branch_01_l_MSTR
  bone2Name=head
  swing1LimitDegrees=45
  swing1Motion=Free
  swing2LimitDegrees=45
  swing2Motion=Free
  twistLimitDegrees=45
  twistMotion=Free
 }
 {
  bone1Name=dyn_head_leaf_039_MSTR
  bone2Name=head
  swing1LimitDegrees=45
  swing1Motion=Free
  swing2LimitDegrees=45
  swing2Motion=Free
  twistLimitDegrees=45
  twistMotion=Free
 }
 {
  bone1Name=dyn_head_leaf_026_MSTR
  bone2Name=head
  swing1LimitDegrees=45
  swing1Motion=Free
  swing2LimitDegrees=45
  swing2Motion=Free
  twistLimitDegrees=45
  twistMotion=Free
 }
 {
  bone1Name=dyn_head_leaf_02_MSTR
  bone2Name=head
  swing1LimitDegrees=45
  swing1Motion=Free
  swing2LimitDegrees=45
  swing2Motion=Free
  twistLimitDegrees=45
  twistMotion=Free
 }
 {
  bone1Name=dyn_head_leaf_032_MSTR
  bone2Name=head
  swing1LimitDegrees=45
  swing1Motion=Free
  swing2LimitDegrees=45
  swing2Motion=Free
  twistLimitDegrees=45
  twistMotion=Free
 }
 {
  bone1Name=dyn_head_leaf_016_MSTR
  bone2Name=head
  swing1LimitDegrees=45
  swing1Motion=Free
  swing2LimitDegrees=45
  swing2Motion=Free
  twistLimitDegrees=45
  twistMotion=Free
 }
 {
  bone1Name=dyn_head_leaf_037_MSTR
  bone2Name=head
  swing1LimitDegrees=45
  swing1Motion=Free
  swing2LimitDegrees=45
  swing2Motion=Free
  twistLimitDegrees=45
  twistMotion=Free
 }
 {
  bone1Name=dyn_head_leaf_017_MSTR
  bone2Name=head
  swing1LimitDegrees=45
  swing1Motion=Free
  swing2LimitDegrees=45
  swing2Motion=Free
  twistLimitDegrees=45
  twistMotion=Free
 }
 {
  bone1Name=dyn_head_leaf_029_MSTR
  bone2Name=head
  swing1LimitDegrees=45
  swing1Motion=Free
  swing2LimitDegrees=45
  swing2Motion=Free
  twistLimitDegrees=45
  twistMotion=Free
 }
 {
  bone1Name=dyn_head_leaf_08_MSTR
  bone2Name=head
  swing1LimitDegrees=45
  swing1Motion=Free
  swing2LimitDegrees=45
  swing2Motion=Free
  twistLimitDegrees=45
  twistMotion=Free
 }
 {
  bone1Name=dyn_head_leaf_027_MSTR
  bone2Name=head
  swing1LimitDegrees=45
  swing1Motion=Free
  swing2LimitDegrees=45
  swing2Motion=Free
  twistLimitDegrees=45
  twistMotion=Free
 }
 {
  bone1Name=dyn_head_leaf_028_MSTR
  bone2Name=head
  swing1LimitDegrees=45
  swing1Motion=Free
  swing2LimitDegrees=45
  swing2Motion=Free
  twistLimitDegrees=45
  twistMotion=Free
 }
 {
  bone1Name=dyn_head_leaf_035_MSTR
  bone2Name=head
  swing1LimitDegrees=45
  swing1Motion=Free
  swing2LimitDegrees=45
  swing2Motion=Free
  twistLimitDegrees=45
  twistMotion=Free
 }
 {
  bone1Name=dyn_head_leaf_012_MSTR
  bone2Name=head
  swing1LimitDegrees=45
  swing1Motion=Free
  swing2LimitDegrees=45
  swing2Motion=Free
  twistLimitDegrees=45
  twistMotion=Free
 }
 {
  bone1Name=dyn_head_leaf_07_MSTR
  bone2Name=head
  swing1LimitDegrees=45
  swing1Motion=Free
  swing2LimitDegrees=45
  swing2Motion=Free
  twistLimitDegrees=45
  twistMotion=Free
 }
 {
  bone1Name=dyn_head_leaf_05_MSTR
  bone2Name=head
  swing1LimitDegrees=45
  swing1Motion=Free
  swing2LimitDegrees=45
  swing2Motion=Free
  twistLimitDegrees=45
  twistMotion=Free
 }
 {
  bone1Name=dyn_head_leaf_03_MSTR
  bone2Name=head
  swing1LimitDegrees=45
  swing1Motion=Free
  swing2LimitDegrees=45
  swing2Motion=Free
  twistLimitDegrees=45
  twistMotion=Free
 }
 {
  bone1Name=dyn_head_leaf_023_MSTR
  bone2Name=head
  swing1LimitDegrees=45
  swing1Motion=Free
  swing2LimitDegrees=45
  swing2Motion=Free
  twistLimitDegrees=45
  twistMotion=Free
 }
 {
  bone1Name=dyn_head_leaf_015_MSTR
  bone2Name=head
  swing1LimitDegrees=45
  swing1Motion=Free
  swing2LimitDegrees=45
  swing2Motion=Free
  twistLimitDegrees=45
  twistMotion=Free
 }
 {
  bone1Name=dyn_head_leaf_014_MSTR
  bone2Name=head
  swing1LimitDegrees=45
  swing1Motion=Free
  swing2LimitDegrees=45
  swing2Motion=Free
  twistLimitDegrees=45
  twistMotion=Free
 }
 {
  bone1Name=dyn_head_leaf_06_MSTR
  bone2Name=head
  swing1LimitDegrees=45
  swing1Motion=Free
  swing2LimitDegrees=45
  swing2Motion=Free
  twistLimitDegrees=45
  twistMotion=Free
 }
 {
  bone1Name=dyn_head_leaf_036_MSTR
  bone2Name=head
  swing1LimitDegrees=45
  swing1Motion=Free
  swing2LimitDegrees=45
  swing2Motion=Free
  twistLimitDegrees=45
  twistMotion=Free
 }
 {
  bone1Name=dyn_head_leaf_034_MSTR
  bone2Name=head
  swing1LimitDegrees=45
  swing1Motion=Free
  swing2LimitDegrees=45
  swing2Motion=Free
  twistLimitDegrees=45
  twistMotion=Free
 }
 {
  bone1Name=dyn_head_leaf_033_MSTR
  bone2Name=head
  swing1LimitDegrees=45
  swing1Motion=Free
  swing2LimitDegrees=45
  swing2Motion=Free
  twistLimitDegrees=45
  twistMotion=Free
 }
 {
  bone1Name=dyn_head_leaf_09_MSTR
  bone2Name=head
  swing1LimitDegrees=45
  swing1Motion=Free
  swing2LimitDegrees=45
  swing2Motion=Free
  twistLimitDegrees=45
  twistMotion=Free
 }
 {
  bone1Name=dyn_head_leaf_01_MSTR
  bone2Name=head
  swing1LimitDegrees=45
  swing1Motion=Free
  swing2LimitDegrees=45
  swing2Motion=Free
  twistLimitDegrees=45
  twistMotion=Free
 }
 {
  bone1Name=dyn_head_leaf_020_MSTR
  bone2Name=head
  swing1LimitDegrees=45
  swing1Motion=Free
  swing2LimitDegrees=45
  swing2Motion=Free
  twistLimitDegrees=45
  twistMotion=Free
 }
 {
  bone1Name=dyn_head_leaf_019_MSTR
  bone2Name=head
  swing1LimitDegrees=45
  swing1Motion=Free
  swing2LimitDegrees=45
  swing2Motion=Free
  twistLimitDegrees=45
  twistMotion=Free
 }
 {
  bone1Name=dyn_head_leaf_018_MSTR
  bone2Name=head
  swing1LimitDegrees=45
  swing1Motion=Free
  swing2LimitDegrees=45
  swing2Motion=Free
  twistLimitDegrees=45
  twistMotion=Free
 }
 {
  bone1Name=dyn_head_leaf_013_MSTR
  bone2Name=head
  swing1LimitDegrees=45
  swing1Motion=Free
  swing2LimitDegrees=45
  swing2Motion=Free
  twistLimitDegrees=45
  twistMotion=Free
 }
 {
  bone1Name=dyn_head_leaf_041_MSTR
  bone2Name=head
  swing1LimitDegrees=45
  swing1Motion=Free
  swing2LimitDegrees=45
  swing2Motion=Free
  twistLimitDegrees=45
  twistMotion=Free
 }
 {
  bone1Name=dyn_head_leaf_042_MSTR
  bone2Name=head
  swing1LimitDegrees=45
  swing1Motion=Free
  swing2LimitDegrees=45
  swing2Motion=Free
  twistLimitDegrees=45
  twistMotion=Free
 }
 {
  bone1Name=dyn_head_leaf_040_MSTR
  bone2Name=head
  swing1LimitDegrees=45
  swing1Motion=Free
  swing2LimitDegrees=45
  swing2Motion=Free
  twistLimitDegrees=45
  twistMotion=Free
 }
 {
  bone1Name=dyn_head_leaf_030_MSTR
  bone2Name=head
  swing1LimitDegrees=45
  swing1Motion=Free
  swing2LimitDegrees=45
  swing2Motion=Free
  twistLimitDegrees=45
  twistMotion=Free
 }
 {
  bone1Name=dyn_neck_bk_02_r
  bone2Name=dyn_neck_bk_02
  swing1LimitDegrees=45
  swing1Motion=Free
  swing2LimitDegrees=45
  swing2Motion=Free
  twistLimitDegrees=45
  twistMotion=Free
 }
 {
  bone1Name=dyn_neck_bk_02_l
  bone2Name=dyn_neck_bk_02
  swing1LimitDegrees=45
  swing1Motion=Free
  swing2LimitDegrees=45
  swing2Motion=Free
  twistLimitDegrees=45
  twistMotion=Free
 }
 {
  bone1Name=dyn_shoulder_01_l_MSTR
  bone2Name=spine_05
  swing1LimitDegrees=45
  swing1Motion=Free
  swing2LimitDegrees=45
  swing2Motion=Free
  twistLimitDegrees=45
  twistMotion=Free
 }
 {
  bone1Name=dyn_shoulder_01_r_MSTR
  bone2Name=spine_05
  swing1LimitDegrees=45
  swing1Motion=Free
  swing2LimitDegrees=45
  swing2Motion=Free
  twistLimitDegrees=45
  twistMotion=Free
 }
 {
  bone1Name=head
  bone2Name=spine_05
  swing1LimitDegrees=45
  swing1Motion=Free
  swing2LimitDegrees=45
  swing2Motion=Free
  twistLimitDegrees=45
  twistMotion=Free
 }
 {
  bone1Name=dyn_neck_bk_01_MSTR
  bone2Name=spine_05
  swing1LimitDegrees=45
  swing1Motion=Free
  swing2LimitDegrees=45
  swing2Motion=Free
  twistLimitDegrees=45
  twistMotion=Free
 }
 {
  bone1Name=dyn_neck_fr_01_l_MSTR
  bone2Name=spine_05
  swing1LimitDegrees=45
  swing1Motion=Free
  swing2LimitDegrees=45
  swing2Motion=Free
  twistLimitDegrees=45
  twistMotion=Free
 }
 {
  bone1Name=dyn_neck_fr_01_MSTR
  bone2Name=spine_05
  swing1LimitDegrees=45
  swing1Motion=Free
  swing2LimitDegrees=45
  swing2Motion=Free
  twistLimitDegrees=45
  twistMotion=Free
 }
 {
  bone1Name=dyn_neck_fr_01_r_MSTR
  bone2Name=spine_05
  swing1LimitDegrees=45
  swing1Motion=Free
  swing2LimitDegrees=45
  swing2Motion=Free
  twistLimitDegrees=45
  twistMotion=Free
 }
 {
  bone1Name=dyn_neck_bk_01_l_MSTR
  bone2Name=spine_05
  swing1LimitDegrees=45
  swing1Motion=Free
  swing2LimitDegrees=45
  swing2Motion=Free
  twistLimitDegrees=45
  twistMotion=Free
 }
 {
  bone1Name=dyn_neck_bk_01_r_MSTR
  bone2Name=spine_05
  swing1LimitDegrees=45
  swing1Motion=Free
  swing2LimitDegrees=45
  swing2Motion=Free
  twistLimitDegrees=45
  twistMotion=Free
 }
 {
  bone1Name=spine_05
  bone2Name=spine_04
  swing1LimitDegrees=45
  swing1Motion=Free
  swing2LimitDegrees=45
  swing2Motion=Free
  twistLimitDegrees=45
  twistMotion=Free
 }
]
