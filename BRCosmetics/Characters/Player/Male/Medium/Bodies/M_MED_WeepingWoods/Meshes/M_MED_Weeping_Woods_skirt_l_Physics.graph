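# Physics asset for the left-side skirt of the male medium Weeping Woods cosmetic body, with capsules on pelvis and legs.
# Front, front-left, side-left and back-left chains each have three box bodies with limited angular motion, hung from tiny MSTR anchor bodies on the pelvis.
# Angularly free cross-constraints tie neighboring chains together, fully between side and front-left but only at the tips elsewhere.
# Another links the left thigh twist to the front-left tip.
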PhysicsAsset /BRCosmetics/Characters/Player/Male/Medium/Bodies/M_MED_WeepingWoods/Meshes/M_MED_Weeping_Woods_skirt_l_Physics
bodies {
 calf_l [{shapeType=Capsule, length=35.2312, radius=8.5578, center=<-18.9204 -0.0015 0>, rotation=<rot -90 -0.0198 0.0198>}]
 calf_r [{shapeType=Capsule, length=35.2312, radius=8.5578, center=<18.9204 0.0015 0>, rotation=<rot -90 0.0396 179.9604>}]
 dyn_skirt_bk_01_MSTR [{shapeType=Capsule, length=0.01, radius=0.01}]
 dyn_skirt_bk_01_l [{shapeType=Box, extentX=19.6, extentY=18.1, extentZ=6.1, center=<9.8996 0 0.0001>}]
 dyn_skirt_bk_01_l_MSTR [{shapeType=Capsule, length=0.01, radius=0.01}]
 dyn_skirt_bk_02_l [{shapeType=Box, extentX=9.1, extentY=18.1, extentZ=6.1, center=<4.6773 0 -0.0001>}]
 dyn_skirt_bk_03_l [{shapeType=Box, extentX=19.6, extentY=18.1, extentZ=6.1, center=<9.8996 0 0.0001>}]
 dyn_skirt_fr_01 [{shapeType=Box, extentX=19.6, extentY=18.1, extentZ=6.1, center=<9.8996 0 0.0001>}]
 dyn_skirt_fr_01_MSTR [{shapeType=Capsule, length=0.01, radius=0.01}]
 dyn_skirt_fr_01_l [{shapeType=Box, extentX=19.6, extentY=18.1, extentZ=6.1, center=<9.8996 0 0.0001>}]
 dyn_skirt_fr_01_l_MSTR [{shapeType=Capsule, length=0.01, radius=0.01}]
 dyn_skirt_fr_02 [{shapeType=Box, extentX=9.1, extentY=18.1, extentZ=6.1, center=<4.6773 0 -0.0001>}]
 dyn_skirt_fr_02_l [{shapeType=Box, extentX=12.1, extentY=18.1, extentZ=6.1, center=<7.4365 0 -0.0001>}]
 dyn_skirt_fr_03 [{shapeType=Box, extentX=19.6, extentY=18.1, extentZ=6.1, center=<9.8996 0 0.0001>}]
 dyn_skirt_fr_03_l [{shapeType=Box, extentX=19.6, extentY=18.1, extentZ=6.1, center=<13.7371 0 0.0001>}]
 dyn_skirt_side_01_l [{shapeType=Box, extentX=19.6, extentY=18.1, extentZ=6.1, center=<9.8996 0 0.0001>}]
 dyn_skirt_side_01_l_MSTR [{shapeType=Capsule, length=0.01, radius=0.01}]
 dyn_skirt_side_02_l [{shapeType=Box, extentX=9.1, extentY=18.1, extentZ=6.1, center=<4.6773 0 -0.0001>}]
 dyn_skirt_side_03_l [{shapeType=Box, extentX=19.6, extentY=18.1, extentZ=6.1, center=<9.8996 0 0.0001>}]
 pelvis [{shapeType=Capsule, length=10.4488, radius=9.3046, center=<0 0 83.6934>}]
 thigh_l [{shapeType=Capsule, length=37.4812, radius=8.5578, center=<-18.1884 0.0001 0>, rotation=<rot -90 -0.0148 0.0148>}]
 thigh_r [{shapeType=Capsule, length=25.4812, radius=8.5578, center=<18.1884 -0.0001 0>, rotation=<rot -90 0.0396 179.9604>}]
 thigh_twist_01_l [{shapeType=Capsule, length=12.2312, radius=8.5578, center=<-9.2293 -0.0006 -0.0072>, rotation=<rot -90 0.0396 -0.0396>}]
}
constraints [
 {
  bone1Name=dyn_skirt_fr_01_MSTR
  bone2Name=pelvis
  swing1LimitDegrees=45
  swing1Motion=Free
  swing2LimitDegrees=45
  swing2Motion=Free
  twistLimitDegrees=45
  twistMotion=Free
 }
 {
  bone1Name=dyn_skirt_side_01_l_MSTR
  bone2Name=pelvis
  swing1LimitDegrees=45
  swing1Motion=Free
  swing2LimitDegrees=45
  swing2Motion=Free
  twistLimitDegrees=45
  twistMotion=Free
 }
 {
  bone1Name=dyn_skirt_bk_01_MSTR
  bone2Name=pelvis
  swing1LimitDegrees=45
  swing1Motion=Free
  swing2LimitDegrees=45
  swing2Motion=Free
  twistLimitDegrees=45
  twistMotion=Free
 }
 {
  bone1Name=dyn_skirt_fr_01_l_MSTR
  bone2Name=pelvis
  swing1LimitDegrees=45
  swing1Motion=Free
  swing2LimitDegrees=45
  swing2Motion=Free
  twistLimitDegrees=45
  twistMotion=Free
 }
 {
  bone1Name=dyn_skirt_bk_01_l_MSTR
  bone2Name=pelvis
  swing1LimitDegrees=45
  swing1Motion=Free
  swing2LimitDegrees=45
  swing2Motion=Free
  twistLimitDegrees=45
  twistMotion=Free
 }
 {bone1Name=dyn_skirt_side_01_l, bone2Name=dyn_skirt_side_01_l_MSTR, swing1LimitDegrees=75, swing2LimitDegrees=75, twistLimitDegrees=5}
 {bone1Name=dyn_skirt_side_02_l, bone2Name=dyn_skirt_side_01_l, swing1LimitDegrees=75, swing2LimitDegrees=75, twistLimitDegrees=5}
 {bone1Name=dyn_skirt_side_03_l, bone2Name=dyn_skirt_side_02_l, swing1LimitDegrees=75, swing2LimitDegrees=75, twistLimitDegrees=5}
 {bone1Name=dyn_skirt_fr_01_l, bone2Name=dyn_skirt_fr_01_l_MSTR, swing1LimitDegrees=15, swing2LimitDegrees=45, twistLimitDegrees=5}
 {bone1Name=dyn_skirt_fr_02_l, bone2Name=dyn_skirt_fr_01_l, swing1LimitDegrees=15, swing2LimitDegrees=45, twistLimitDegrees=5}
 {bone1Name=dyn_skirt_fr_03_l, bone2Name=dyn_skirt_fr_02_l, swing1LimitDegrees=15, swing2LimitDegrees=45, twistLimitDegrees=5}
 {bone1Name=dyn_skirt_bk_01_l, bone2Name=dyn_skirt_bk_01_l_MSTR, swing1LimitDegrees=75, swing2LimitDegrees=75, twistLimitDegrees=5}
 {bone1Name=dyn_skirt_bk_02_l, bone2Name=dyn_skirt_bk_01_l, swing1LimitDegrees=75, swing2LimitDegrees=75, twistLimitDegrees=5}
 {bone1Name=dyn_skirt_bk_03_l, bone2Name=dyn_skirt_bk_02_l, swing1LimitDegrees=75, swing2LimitDegrees=75, twistLimitDegrees=5}
 {
  bone1Name=thigh_l
  bone2Name=pelvis
  swing1LimitDegrees=45
  swing1Motion=Free
  swing2LimitDegrees=45
  swing2Motion=Free
  twistLimitDegrees=45
  twistMotion=Free
 }
 {
  bone1Name=thigh_r
  bone2Name=pelvis
  swing1LimitDegrees=45
  swing1Motion=Free
  swing2LimitDegrees=45
  swing2Motion=Free
  twistLimitDegrees=45
  twistMotion=Free
 }
 {
  bone1Name=thigh_twist_01_l
  bone2Name=thigh_l
  swing1LimitDegrees=45
  swing1Motion=Free
  swing2LimitDegrees=45
  swing2Motion=Free
  twistLimitDegrees=45
  twistMotion=Free
 }
 {
  bone1Name=dyn_skirt_side_03_l
  bone2Name=dyn_skirt_fr_03_l
  swing1LimitDegrees=45
  swing1Motion=Free
  swing2LimitDegrees=45
  swing2Motion=Free
  twistLimitDegrees=45
  twistMotion=Free
 }
 {
  bone1Name=dyn_skirt_side_03_l
  bone2Name=dyn_skirt_bk_03_l
  swing1LimitDegrees=45
  swing1Motion=Free
  swing2LimitDegrees=45
  swing2Motion=Free
  twistLimitDegrees=45
  twistMotion=Free
 }
 {
  bone1Name=calf_l
  bone2Name=thigh_l
  swing1LimitDegrees=45
  swing1Motion=Free
  swing2LimitDegrees=45
  swing2Motion=Free
  twistLimitDegrees=45
  twistMotion=Free
 }
 {bone1Name=dyn_skirt_fr_01, bone2Name=dyn_skirt_fr_01_MSTR, swing1LimitDegrees=75, swing2LimitDegrees=75, twistLimitDegrees=45}
 {bone1Name=dyn_skirt_fr_02, bone2Name=dyn_skirt_fr_01, swing1LimitDegrees=75, swing2LimitDegrees=75, twistLimitDegrees=45}
 {bone1Name=dyn_skirt_fr_03, bone2Name=dyn_skirt_fr_02, swing1LimitDegrees=75, swing2LimitDegrees=75, twistLimitDegrees=45}
 {
  bone1Name=dyn_skirt_fr_03
  bone2Name=dyn_skirt_fr_03_l
  swing1LimitDegrees=45
  swing1Motion=Free
  swing2LimitDegrees=45
  swing2Motion=Free
  twistLimitDegrees=45
  twistMotion=Free
 }
 {
  bone1Name=calf_r
  bone2Name=thigh_r
  swing1LimitDegrees=45
  swing1Motion=Free
  swing2LimitDegrees=45
  swing2Motion=Free
  twistLimitDegrees=45
  twistMotion=Free
 }
 {
  bone1Name=dyn_skirt_side_02_l
  bone2Name=dyn_skirt_fr_02_l
  swing1LimitDegrees=45
  swing1Motion=Free
  swing2LimitDegrees=45
  swing2Motion=Free
  twistLimitDegrees=45
  twistMotion=Free
 }
 {
  bone1Name=dyn_skirt_side_01_l
  bone2Name=dyn_skirt_fr_01_l
  swing1LimitDegrees=45
  swing1Motion=Free
  swing2LimitDegrees=45
  swing2Motion=Free
  twistLimitDegrees=45
  twistMotion=Free
 }
 {
  bone1Name=thigh_twist_01_l
  bone2Name=dyn_skirt_fr_03_l
  swing1LimitDegrees=45
  swing1Motion=Free
  swing2LimitDegrees=45
  swing2Motion=Free
  twistLimitDegrees=45
  twistMotion=Free
 }
]
