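# Physics asset for the right-side skirt of the male medium Weeping Woods cosmetic body, simulating cloth-like motion with rigid bodies.
# Five three-box chains (front, right front, right side, back, right back) hang from tiny _MSTR anchor bodies with angle-limited joints.
# Angularly free constraints tie neighboring chains, mostly at the tips, and link the right-front tip to thigh_r.
# Capsules on both thighs, thigh twists and calves form the leg bodies.
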PhysicsAsset /BRCosmetics/Characters/Player/Male/Medium/Bodies/M_MED_WeepingWoods/Meshes/M_MED_Weeping_Woods_skirt_r_Physics
bodies {
 calf_l [{shapeType=Capsule, length=35.2312, radius=8.5578, center=<-18.9204 -0.0015 0>, rotation=<rot -90 -0.0198 0.0198>}]
 calf_r [{shapeType=Capsule, length=35.2312, radius=8.5578, center=<18.9204 0.0015 0>, rotation=<rot -90 0.0396 179.9604>}]
 dyn_skirt_bk_01 [{shapeType=Box, extentX=14.1, extentY=18.1, extentZ=6.1, center=<7.5163 0 0.0001>}]
 dyn_skirt_bk_01_MSTR [{shapeType=Capsule, length=0.01, radius=0.01}]
 dyn_skirt_bk_01_r [{shapeType=Box, extentX=19.6, extentY=18.1, extentZ=6.1, center=<9.8996 0 0.0001>}]
 dyn_skirt_bk_01_r_MSTR [{shapeType=Box, extentX=0.01, extentY=0.01, extentZ=0.01, rotation=<rot 0 -90 0>}]
 dyn_skirt_bk_02 [{shapeType=Box, extentX=9.1, extentY=18.1, extentZ=6.1, center=<4.6773 0 -0.0001>}]
 dyn_skirt_bk_02_r [{shapeType=Box, extentX=9.1, extentY=18.1, extentZ=6.1, center=<4.6773 0 -0.0001>}]
 dyn_skirt_bk_03 [{shapeType=Box, extentX=19.6, extentY=18.1, extentZ=6.1, center=<9.8996 0 0.0001>}]
 dyn_skirt_bk_03_r [{shapeType=Box, extentX=19.6, extentY=18.1, extentZ=6.1, center=<9.8996 0 0.0001>}]
 dyn_skirt_fr_01 [{shapeType=Box, extentX=19.6, extentY=18.1, extentZ=6.1, center=<9.8996 0 0.0001>}]
 dyn_skirt_fr_01_MSTR [{shapeType=Capsule, length=0.01, radius=0.01}]
 dyn_skirt_fr_01_r [{shapeType=Box, extentX=19.6, extentY=18.1, extentZ=6.1, center=<9.8996 0 0.0001>}]
 dyn_skirt_fr_01_r_MSTR [{shapeType=Box, extentX=0.01, extentY=0.01, extentZ=0.01, rotation=<rot 0 -90 0>}]
 dyn_skirt_fr_02 [{shapeType=Box, extentX=9.1, extentY=18.1, extentZ=6.1, center=<4.6773 0 -0.0001>}]
 dyn_skirt_fr_02_r [{shapeType=Box, extentX=9.1, extentY=18.1, extentZ=6.1, center=<4.6773 0 -0.0001>}]
 dyn_skirt_fr_03 [{shapeType=Box, extentX=19.6, extentY=18.1, extentZ=6.1, center=<9.8996 0 0.0001>}]
 dyn_skirt_fr_03_r [{shapeType=Box, extentX=19.6, extentY=18.1, extentZ=6.1, center=<9.8996 0 0.0001>}]
 dyn_skirt_side_01_r [{shapeType=Box, extentX=19.6, extentY=18.1, extentZ=6.1, center=<9.8996 0 0.0001>}]
 dyn_skirt_side_01_r_MSTR [{shapeType=Box, extentX=0.01, extentY=0.01, extentZ=0.01, rotation=<rot 0 -90 0>}]
 dyn_skirt_side_02_r [{shapeType=Box, extentX=9.1, extentY=18.1, extentZ=6.1, center=<4.6773 0 -0.0001>}]
 dyn_skirt_side_03_r [{shapeType=Box, extentX=19.6, extentY=18.1, extentZ=6.1, center=<9.8996 0 0.0001>}]
 thigh_l [{shapeType=Capsule, length=25.4812, radius=8.5578, center=<-18.1884 0.0001 0>, rotation=<rot -90 -0.0148 0.0148>}]
 thigh_r [{shapeType=Capsule, length=25.4812, radius=8.5578, center=<18.1884 -0.0001 0>, rotation=<rot -90 0.0396 179.9604>}]
 thigh_twist_01_l [{shapeType=Capsule, length=12.2312, radius=8.5578, center=<-9.2293 -0.0006 -0.0072>, rotation=<rot -90 0.0396 -0.0396>}]
 thigh_twist_01_r [{shapeType=Capsule, length=12.2312, radius=8.5578, center=<9.2293 0.0006 0.0072>, rotation=<rot -90 -0.0264 -179.9736>}]
}
constraints [
 {
  bone1Name=thigh_twist_01_l
  bone2Name=thigh_l
  swing1LimitDegrees=45
  swing1Motion=Free
  swing2LimitDegrees=45
  swing2Motion=Free
  twistLimitDegrees=45
  twistMotion=Free
 }
 {
  bone1Name=calf_l
  bone2Name=thigh_l
  swing1LimitDegrees=45
  swing1Motion=Free
  swing2LimitDegrees=45
  swing2Motion=Free
  twistLimitDegrees=45
  twistMotion=Free
 }
 {bone1Name=dyn_skirt_fr_01, bone2Name=dyn_skirt_fr_01_MSTR, swing1LimitDegrees=30, swing2LimitDegrees=75, twistLimitDegrees=30}
 {bone1Name=dyn_skirt_fr_02, bone2Name=dyn_skirt_fr_01, swing1LimitDegrees=30, swing2LimitDegrees=75, twistLimitDegrees=30}
 {bone1Name=dyn_skirt_fr_03, bone2Name=dyn_skirt_fr_02, swing1LimitDegrees=30, swing2LimitDegrees=75, twistLimitDegrees=30}
 {
  bone1Name=calf_r
  bone2Name=thigh_r
  swing1LimitDegrees=45
  swing1Motion=Free
  swing2LimitDegrees=45
  swing2Motion=Free
  twistLimitDegrees=45
  twistMotion=Free
 }
 {bone1Name=dyn_skirt_fr_01_r, bone2Name=dyn_skirt_fr_01_r_MSTR, swing1LimitDegrees=5, swing2LimitDegrees=90, twistLimitDegrees=30}
 {bone1Name=dyn_skirt_fr_02_r, bone2Name=dyn_skirt_fr_01_r, swing1LimitDegrees=5, swing2LimitDegrees=90, twistLimitDegrees=5}
 {bone1Name=dyn_skirt_fr_03_r, bone2Name=dyn_skirt_fr_02_r, swing1LimitDegrees=5, swing2LimitDegrees=90, twistLimitDegrees=5}
 {bone1Name=dyn_skirt_side_03_r, bone2Name=dyn_skirt_side_02_r, swing1LimitDegrees=75, swing2LimitDegrees=75, twistLimitDegrees=5}
 {bone1Name=dyn_skirt_side_02_r, bone2Name=dyn_skirt_side_01_r, swing1LimitDegrees=75, swing2LimitDegrees=75, twistLimitDegrees=5}
 {bone1Name=dyn_skirt_side_01_r, bone2Name=dyn_skirt_side_01_r_MSTR, swing1LimitDegrees=75, swing2LimitDegrees=75, twistLimitDegrees=5}
 {bone1Name=dyn_skirt_bk_01_r, bone2Name=dyn_skirt_bk_01_r_MSTR, swing1LimitDegrees=75, swing2LimitDegrees=75, twistLimitDegrees=5}
 {bone1Name=dyn_skirt_bk_02_r, bone2Name=dyn_skirt_bk_01_r, swing1LimitDegrees=75, swing2LimitDegrees=75, twistLimitDegrees=5}
 {bone1Name=dyn_skirt_bk_03_r, bone2Name=dyn_skirt_bk_02_r, swing1LimitDegrees=75, swing2LimitDegrees=75, twistLimitDegrees=5}
 {
  bone1Name=dyn_skirt_fr_03
  bone2Name=dyn_skirt_fr_03_r
  swing1LimitDegrees=45
  swing1Motion=Free
  swing2LimitDegrees=45
  swing2Motion=Free
  twistLimitDegrees=45
  twistMotion=Free
 }
 {
  bone1Name=thigh_r
  bone2Name=dyn_skirt_fr_03_r
  swing1LimitDegrees=45
  swing1Motion=Free
  swing2LimitDegrees=45
  swing2Motion=Free
  twistLimitDegrees=45
  twistMotion=Free
 }
 {
  bone1Name=thigh_twist_01_r
  bone2Name=thigh_r
  swing1LimitDegrees=45
  swing1Motion=Free
  swing2LimitDegrees=45
  swing2Motion=Free
  twistLimitDegrees=45
  twistMotion=Free
 }
 {
  bone1Name=dyn_skirt_side_03_r
  bone2Name=dyn_skirt_fr_03_r
  swing1LimitDegrees=45
  swing1Motion=Free
  swing2LimitDegrees=45
  swing2Motion=Free
  twistLimitDegrees=45
  twistMotion=Free
 }
 {
  bone1Name=dyn_skirt_side_02_r
  bone2Name=dyn_skirt_fr_02_r
  swing1LimitDegrees=45
  swing1Motion=Free
  swing2LimitDegrees=45
  swing2Motion=Free
  twistLimitDegrees=45
  twistMotion=Free
 }
 {
  bone1Name=dyn_skirt_side_01_r
  bone2Name=dyn_skirt_fr_01_r
  swing1LimitDegrees=45
  swing1Motion=Free
  swing2LimitDegrees=45
  swing2Motion=Free
  twistLimitDegrees=45
  twistMotion=Free
 }
 {
  bone1Name=dyn_skirt_side_03_r
  bone2Name=dyn_skirt_bk_03_r
  swing1LimitDegrees=45
  swing1Motion=Free
  swing2LimitDegrees=45
  swing2Motion=Free
  twistLimitDegrees=45
  twistMotion=Free
 }
 {bone1Name=dyn_skirt_bk_01, bone2Name=dyn_skirt_bk_01_MSTR, swing1LimitDegrees=5, swing2LimitDegrees=75, twistLimitDegrees=5}
 {bone1Name=dyn_skirt_bk_02, bone2Name=dyn_skirt_bk_01, swing1LimitDegrees=5, swing2LimitDegrees=75, twistLimitDegrees=5}
 {bone1Name=dyn_skirt_bk_03, bone2Name=dyn_skirt_bk_02, swing1LimitDegrees=5, swing2LimitDegrees=75, twistLimitDegrees=5}
 {
  bone1Name=dyn_skirt_bk_03_r
  bone2Name=dyn_skirt_bk_03
  swing1LimitDegrees=45
  swing1Motion=Free
  swing2LimitDegrees=45
  swing2Motion=Free
  twistLimitDegrees=45
  twistMotion=Free
 }
]
